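; VM_TournamentDetailsDebug is, by its name and folder, a debug view model for the tournament details screen in the Compete section.
; It declares a RefreshScreenData function that takes a TournamentId name.
; The function body is empty, so it loads or sets no data.
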(fn RefreshScreenData
  (params (TournamentId :type Name)))
