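# Translucent UI material that draws an animated pulse, named for the eligibility element of the Compete tournament details screen.
# A rounded-box SDF supplies shape and outline masks, and a rotating, eased two-color gradient supplies color.
# Two time-driven rounded-box outlines expand and fade on offset ramps, with the second blended in as a fainter echo.
# The combined rings are masked and saturated into a Pulse value, with parameters for speed, softness, curve, thickness and echo opacity.
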Material /Compete/Materials/M_UI_Compete_TournamentDetails_Eligibility_Pulse
graph {
 properties {bAutomaticallySetUsageInEditor=false, blendMode=BLEND_Translucent, materialDomain=MD_UI, usedShadingModels=Default Lit}
 nodes [
  MaterialFunctionCall!(Anchor, Translation, Padding, Scale=Constant() {r=1}) {materialFunction=/FortUILibrary/MaterialFunctions/Basics/MF_UI_VertexTransform}
  n1 = MaterialFunctionCall(Size=Constant() {r=0.9}, CornerRoundness=Constant4Vector() {constant=<rgba 0.6 0.6 0.6 0.6>}.RGBA) {materialFunction=&MF_UI_Box}
  NamedRerouteDeclaration(None=MaterialFunctionCall(DistanceField=@n1.SDF, Threshold=Constant() {r=-0.5}) {materialFunction=&MF_UI_AntiAliasing}.Inner) {name=ShapeMask, nodeColor=<rgba 0 0 0 1>}
  NamedRerouteDeclaration(None=SmoothStep(Value=@n1.SDF) {constMax=0.5, constMin=-0.5}) {name=SDF, nodeColor=<rgba 0 0 0 1>}
  NamedRerouteDeclaration(None=Distance(A=@n1.SDF, B=Constant())) {name=OutlineSDF, nodeColor=<rgba 0 0 0 1>}
  n2 = MaterialFunctionCall(UV Channel=Rotator() {speed=2.5}) {materialFunction=&LinearGradient}
  NamedRerouteDeclaration(None=LinearInterpolate(
    A=VectorParameter() {group=0 - Coloring, parameterName=Color1, defaultValue=<rgba 0.1615 0.4154 1 1>}.RGB
    B=VectorParameter() {group=0 - Coloring, parameterName=Color2, defaultValue=<rgba 0.6883 0.3193 1 1>}.RGB
    Alpha=MaterialFunctionCall("Value (0-1)"=@n2.UGradient) {materialFunction=&MF_UI_EaseCurves}.EaseInOutCubic
   )) {name=GradientColor, nodeColor=<rgba 0 0 0 1>}
  n3 = Time()
  n4 = ScalarParameter() {defaultValue=0.35, parameterName=Pulse_Speed}
  n5 = SmoothStep(Value=Frac(None=Multiply(A=@n3, B=@n4) {constB=0.6})) {constMax=0.9, constMin=0.6}
  n6 = ScalarParameter() {defaultValue=0.95, parameterName=Pulse_Smoothness}
  n7 = ScalarParameter() {defaultValue=0.05, parameterName=Pulse_AnimCurve}
  n8 = ScalarParameter() {defaultValue=0.02, parameterName=Pulse_OutlineThickness}
  n9 = OneMinus(None=MaterialFunctionCall(
    Size=Power(Base=@n5, Exp=@n7) {constExponent=0.5}
    CornerRoundness=Constant4Vector() {constant=<rgba 0.6 0.6 0.6 0.6>}.RGBA
    OutlineThickness=@n8
   ) {materialFunction=&MF_UI_Box}.SDFOutline)
  n10 = Time()
  n11 = SmoothStep(Value=Frac(None=Multiply(A=@n10, B=@n4) {constB=0.6})) {constMin=0.65}
  n12 = MaterialFunctionCall(
   Size=Power(Base=@n11, Exp=@n7) {constExponent=0.5}
   CornerRoundness=Constant4Vector() {constant=<rgba 0.6 0.6 0.6 0.6>}.RGBA
   OutlineThickness=@n8
  ) {materialFunction=&MF_UI_Box}
  n13 = Power(Base=SmoothStep(Min=@n6, Value=OneMinus(None=@n12.SDFOutline)) {constMin=0.99}) {constExponent=4}
  n14 = OneMinus(None=@n11)
  n15 = LinearInterpolate(
   A=Multiply(A=OneMinus(None=@n5), B=SmoothStep(Min=@n6, Value=@n9) {constMin=0.99})
   Alpha=Multiply(A=Multiply(A=@n13, B=@n14), B=ScalarParameter() {defaultValue=0.85, parameterName=Pulse_EchoOpacity}) {constB=0.35}
  )
  n16 = OneMinus(None=SmoothStep(Value=NamedRerouteUsage()) {constMax=0.95})
  NamedRerouteDeclaration(None=Saturate(None=Multiply(A=@n15, B=@n16))) {name=Pulse, nodeColor=<rgba 0 0 0 1>}
  NamedRerouteUsage()
  NamedRerouteUsage()
 ]
}
