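# Compete tournament-details tile for reward items, a WBP_UIKit_ItemCard_Base child with no widgets of its own.
# When assigned a list item, it stores it as a tournament reward item VM, falling back to a Fort item VM.
# Setter functions push image, series background, variant color, enhanced, greyed-out and Juno icon state into the card.
# The Juno check prefers the stored VMs' product over the passed tag, and selection drives the equipped state.
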WidgetBlueprint /Compete/TournamentDetails/WBP_CompeteRewardTile
graphs [
 EventGraph {nodes [
   n1 = Event PreConstruct()
   > Parent: PreConstruct(IsDesignTime=@n1.IsDesignTime)
   > SetItemIsEquipped!(execute, self:WBP UIKit Item Card Base Object Reference, Is Equipped:bool=false) -> (then)
   n2 = Event OnListItemObjectSet()
   > n3 = Cast To FortTournamentRewardItemVM!(execute, Object=@n2.ListItemObject) -> (then, CastFailed, AsTournament Reward Item VM:Tournament Reward Item VM Object Reference) {is=title K2Node_DynamicCast}
   > Set FortTournamentRewardItemVM(FortTournamentRewardItemVM:Tournament Reward Item VM Object Reference=@n3) -> (Output_Get:Tournament Reward Item VM Object Reference)
   n4 = Cast To FortItemVM!(execute=@n3.CastFailed, Object=@n2.ListItemObject) -> (then, CastFailed, AsFort Item VM:Fort Item VM Object Reference) {is=title K2Node_DynamicCast}
   > Set FortItemVM(FortItemVM:Fort Item VM Object Reference=@n4) -> (Output_Get:Fort Item VM Object Reference)
   Event On Selected()
   > Parent: On Selected() {is=title K2Node_CallParentFunction}
   > SetItemIsEquipped(Is Equipped=true)
   Event On Deselected()
   > Parent: On Deselected() {is=title K2Node_CallParentFunction}
   > SetItemIsEquipped(Is Equipped=false)
   Event Construct()
   Event Tick()
   Event On Clicked()
   Get ItemID!(self:DEBUG MVVM Rewards Tile Object Reference=Get DEBUG_MVVM_RewardsTile() -> (DEBUG_MVVM_RewardsTile:DEBUG MVVM Rewards Tile Object Reference)) -> (ItemID:Name=None)
  ]}
 SetIsEnhanced {nodes [
   n1 = SetIsEnhanced!() -> (then, Is Enhanced:bool=~) {is=title K2Node_FunctionEntry}
   > SetItemIsEnhanced!(execute, self:WBP UIKit Item Card Base Object Reference, Is Enhanced:bool=@n1) -> (then)
  ]}
 SetIsGreyedOut {nodes [
   n1 = SetIsGreyedOut!() -> (then, Is Greyed Out:bool=~) {is=title K2Node_FunctionEntry}
   n2 = NOT Boolean(A=@n1)
   n3 = SetIsEnabled(bInIsEnabled=@n2)
   SetRenderOpacity(execute=[@n3.then, @n1.then], InOpacity=SelectFloat(A=1, B=0.5, bPickA=@n2))
  ]}
 SetIsJunoProduct {nodes [
   SetIsJunoProduct!() -> (then, In Product Tag:Gameplay Tag Structure) {is=title K2Node_FunctionEntry}
   n1 = Get FortTournamentRewardItemVM() -> (FortTournamentRewardItemVM:Tournament Reward Item VM Object Reference)
   > n2 = Is Valid(InputObject=@n1)
   n3 = Get ItemVM!(self:Tournament Reward Item VM Object Reference=@n1) -> (ItemVM:Fort Item VM Object Reference)
   > n4 = Is Valid(InputObject=@n3)
   > Set Item Is Juno!(
    execute
    self:WBP UIKit Item Card Base Object Reference
    IsIconVisible:bool="Equal (GameplayTag)"(A=GetCurrentProduct!(self:Fort Item VM Object Reference=@n3) -> (ReturnValue:Gameplay Tag Structure), B=(TagName="Product.Juno"))
   ) -> (then) {is=title K2Node_CallFunction}
   n5 = Get FortItemVM() -> (FortItemVM:Fort Item VM Object Reference)
   n6 = Is Valid(exec=[@n2.Is Not Valid, @n4.Is Not Valid], InputObject=@n5)
   Set Item Is Juno(
    execute=@n6.Is Not Valid
    IsIconVisible="Equal (GameplayTag)"(A=Get In Product Tag() -> (In Product Tag:Gameplay Tag Structure), B=(TagName="Product.Juno"))
   )
   Set Item Is Juno(execute=@n6.Is Valid, IsIconVisible="Equal (GameplayTag)"(A=GetCurrentProduct(self=@n5), B=(TagName="Product.Juno")))
  ]}
 SetItemDetails {nodes [
   n1 = SetItemDetails!() -> (then, In Details:Fort Item Details VM Object Reference) {is=title K2Node_FunctionEntry}
   > SetItemImage!(
    execute
    self:WBP UIKit Item Card Base Object Reference
    Item Image:Soft Object Reference=Get SmallPreviewImage!(self:Fort Item Details VM Object Reference=@n1) -> (SmallPreviewImage:Texture 2D Soft Object Reference)
   ) -> (then)
   > SetAsyncBackground!(
    execute
    self:WBP UIKit Item Card Base Object Reference
    MaterialInterfaceValue:Material Interface Soft Object Reference=Get SeriesBackgroundMaterial!(self:Fort Item Details VM Object Reference=@n1) -> (SeriesBackgroundMaterial:Material Interface Soft Object Reference)
   ) -> (then)
  ]}
 SetVariant {nodes [
   n1 = SetVariant!() -> (then, Data:Variant Display Info Structure) {is=title K2Node_FunctionEntry}
   n2 = Break Variant Display Info!(VariantDisplayInfo:"Variant Display Info Structure (by ref)"=@n1) -> (Texture:Texture 2D Soft Object Reference, TintColor:Slate Color Structure, bApplyTexture:bool=~, bApplyTint:bool=~) {is=title K2Node_BreakStruct}
   > SetItemImage(Item Image=@n2)
   > Set Item Color!(execute, self:WBP UIKit Item Card Base Object Reference, VariantColor:Slate Color Structure=@n2.TintColor) -> (then) {is=title K2Node_CallFunction}
  ]}
]
variables []
widgets {
 info {
  inheritedWidgetCount=0
  parentClass=/FortUILibrary/Systems/Items/WBP_UIKit_ItemCard_Base.WBP_UIKit_ItemCard_Base_C
  rootWidgetClass=null
  widgetCount=0
 }
 widgets []
}
