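# Activatable modal that previews cosmetic tournament payout rewards, with an item preview, description and threshold requirement text.
# InitItems flattens payout thresholds into a horizontal item list plus an index-parallel Requirements array.
# Selection updates preview, details and header, and fires an analytics event; arrows page through items, showing only off gamepad when list ends are out of view.
# Activation plays an animated intro, or closes the modal when payouts are empty.
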WidgetBlueprint /Compete/TournamentDetails/ItemPreview/WBP_ItemPreview_Modal
animations [
 Anim_InitialStateSet {
  binding_count=5
  length_seconds=0.1
  object=&self:Anim_InitialStateSet
  sequence {bindings [
    HorizontalBox_0 {
     id=#1
     possessed_class=HorizontalBox
     tracks [
      {
       type=MovieScene2DTransformTrack
       sections [{
         type=MovieScene2DTransformSection
         channels [
          Translation.X {type=MovieSceneScriptingFloatChannel, const [-128,1]}
          Translation.Y {type=MovieSceneScriptingFloatChannel, const [0,1]}
          Angle {type=MovieSceneScriptingFloatChannel}
          Scale.X {type=MovieSceneScriptingFloatChannel}
          Scale.Y {type=MovieSceneScriptingFloatChannel}
          Shear.X {type=MovieSceneScriptingFloatChannel}
          Shear.Y {type=MovieSceneScriptingFloatChannel}
         ]
        }]
      }
      {
       type=MovieSceneFloatTrack
       display_name=Render Opacity
       sections [{type=MovieSceneFloatSection, channels [None {type=MovieSceneScriptingFloatChannel, const [0,1]}]}]
      }
     ]
    }
    ItemDescription {
     id=#2
     possessed_class=WBP_UIKit_DescriptionWithSubtitle_C
     tracks [
      {
       type=MovieSceneFloatTrack
       display_name=Render Opacity
       sections [{type=MovieSceneFloatSection, channels [None {type=MovieSceneScriptingFloatChannel, const [0,1]}]}]
      }
      {
       type=MovieScene2DTransformTrack
       sections [{
         type=MovieScene2DTransformSection
         channels [
          Translation.X {type=MovieSceneScriptingFloatChannel, const [128,1]}
          Translation.Y {type=MovieSceneScriptingFloatChannel, const [0,1]}
          Angle {type=MovieSceneScriptingFloatChannel}
          Scale.X {type=MovieSceneScriptingFloatChannel}
          Scale.Y {type=MovieSceneScriptingFloatChannel}
          Shear.X {type=MovieSceneScriptingFloatChannel}
          Shear.Y {type=MovieSceneScriptingFloatChannel}
         ]
        }]
      }
     ]
    }
    Button_LeftScrollArrow {
     id=#3
     possessed_class=CircleArrowButton_C
     tracks [{
       type=MovieSceneFloatTrack
       display_name=Render Opacity
       sections [{type=MovieSceneFloatSection, end_frame=0, start_frame=0, channels [None {type=MovieSceneScriptingFloatChannel, const [0,1]}]}]
      }]
    }
    ListView_Items {
     id=#4
     possessed_class=CommonListView
     tracks [{
       type=MovieSceneFloatTrack
       display_name=Render Opacity
       sections [{type=MovieSceneFloatSection, end_frame=0, start_frame=0, channels [None {type=MovieSceneScriptingFloatChannel, const [0,1]}]}]
      }]
    }
    Button_RightScrollArrow {
     id=#5
     possessed_class=CircleArrowButton_C
     tracks [{
       type=MovieSceneFloatTrack
       display_name=Render Opacity
       sections [{type=MovieSceneFloatSection, end_frame=0, start_frame=0, channels [None {type=MovieSceneScriptingFloatChannel, const [0,1]}]}]
      }]
    }
   ]}
 }
 Anim_Intro {
  binding_count=5
  length_seconds=0.7
  object=&self:Anim_Intro
  sequence {
   bindings [
    HorizontalBox_0 {
     id=#6
     possessed_class=HorizontalBox
     tracks [
      {
       type=MovieScene2DTransformTrack
       sections [{
         type=MovieScene2DTransformSection
         channels [
          Translation.X {type=MovieSceneScriptingFloatChannel, keys [[2,-128],[10,0]]}
          Translation.Y {type=MovieSceneScriptingFloatChannel, const [0,2]}
          Angle {type=MovieSceneScriptingFloatChannel}
          Scale.X {type=MovieSceneScriptingFloatChannel}
          Scale.Y {type=MovieSceneScriptingFloatChannel}
          Shear.X {type=MovieSceneScriptingFloatChannel, keys [[2,12],[10,0]]}
          Shear.Y {type=MovieSceneScriptingFloatChannel, const [0,2]}
         ]
        }]
      }
      {
       type=MovieSceneFloatTrack
       display_name=Render Opacity
       sections [{type=MovieSceneFloatSection, channels [None {type=MovieSceneScriptingFloatChannel, keys [[2,0],[10,1]]}]}]
      }
     ]
    }
    Button_RightScrollArrow {
     id=#7
     possessed_class=CircleArrowButton_C
     tracks [
      {
       type=MovieScene2DTransformTrack
       sections [{
         type=MovieScene2DTransformSection
         channels [
          Translation.X {type=MovieSceneScriptingFloatChannel, keys [[7,-64],[13,0]]}
          Translation.Y {type=MovieSceneScriptingFloatChannel, const [0,2]}
          Angle {type=MovieSceneScriptingFloatChannel}
          Scale.X {type=MovieSceneScriptingFloatChannel}
          Scale.Y {type=MovieSceneScriptingFloatChannel}
          Shear.X {type=MovieSceneScriptingFloatChannel}
          Shear.Y {type=MovieSceneScriptingFloatChannel}
         ]
        }]
      }
      {
       type=MovieSceneFloatTrack
       display_name=Render Opacity
       sections [{type=MovieSceneFloatSection, channels [None {type=MovieSceneScriptingFloatChannel, keys [[7,0],[13,1]]}]}]
      }
     ]
    }
    Button_LeftScrollArrow {
     id=#8
     possessed_class=CircleArrowButton_C
     tracks [
      {
       type=MovieScene2DTransformTrack
       sections [{
         type=MovieScene2DTransformSection
         channels [
          Translation.X {type=MovieSceneScriptingFloatChannel, keys [[7,64],[13,0]]}
          Translation.Y {type=MovieSceneScriptingFloatChannel, const [0,2]}
          Angle {type=MovieSceneScriptingFloatChannel}
          Scale.X {type=MovieSceneScriptingFloatChannel}
          Scale.Y {type=MovieSceneScriptingFloatChannel}
          Shear.X {type=MovieSceneScriptingFloatChannel}
          Shear.Y {type=MovieSceneScriptingFloatChannel}
         ]
        }]
      }
      {
       type=MovieSceneFloatTrack
       display_name=Render Opacity
       sections [{type=MovieSceneFloatSection, channels [None {type=MovieSceneScriptingFloatChannel, keys [[7,0],[13,1]]}]}]
      }
     ]
    }
    ItemDescription {
     id=#9
     possessed_class=WBP_UIKit_DescriptionWithSubtitle_C
     tracks [
      {
       type=MovieSceneFloatTrack
       display_name=Render Opacity
       sections [{type=MovieSceneFloatSection, channels [None {type=MovieSceneScriptingFloatChannel, keys [[2,0],[10,1]]}]}]
      }
      {
       type=MovieScene2DTransformTrack
       sections [{
         type=MovieScene2DTransformSection
         channels [
          Translation.X {type=MovieSceneScriptingFloatChannel, keys [[2,128],[10,0]]}
          Translation.Y {type=MovieSceneScriptingFloatChannel, const [0,2]}
          Angle {type=MovieSceneScriptingFloatChannel}
          Scale.X {type=MovieSceneScriptingFloatChannel}
          Scale.Y {type=MovieSceneScriptingFloatChannel}
          Shear.X {type=MovieSceneScriptingFloatChannel, keys [[2,-12],[10,0]]}
          Shear.Y {type=MovieSceneScriptingFloatChannel, const [0,2]}
         ]
        }]
      }
     ]
    }
    ListView_Items {
     id=#10
     possessed_class=CommonListView
     tracks [{
       type=MovieSceneFloatTrack
       display_name=Render Opacity
       sections [{type=MovieSceneFloatSection, end_frame=14, start_frame=0, channels [None {type=MovieSceneScriptingFloatChannel, keys [[0,0],[14,1]]}]}]
      }]
    }
   ]
   tracks [{
     type=MovieSceneEventTrack
     sections [{
       type=MovieSceneEventTriggerSection
       channels [None {type=MovieSceneScriptingEventChannel, const ["<Struct 'MovieSceneEvent' (0x466a6725280) {}>", 1]}]
      }]
    }]
  }
 }
]
graphs [
 Activate ItemPreview Screen {nodes [
   Activate ItemPreview Screen() {is=title K2Node_FunctionEntry}
   n1 = Get FortTournamentViewVM() -> (FortTournamentViewVM:Tournament View VM Object Reference)
   > n2 = Branch(Condition=Is Not Empty!(TargetArray:Array of Tournament Leaderboard Payout Table Data Structures=Get Payouts!(self:Tournament View VM Object Reference=@n1) -> (Payouts:Array of Tournament Leaderboard Payout Table Data Structures)) -> (ReturnValue:bool))
   > PlayIntro()
   > HandleItemListVisible()
   > SetSidesToPad!(
    execute
    self:Safe Zone Object Reference=Get SZ_AwardBox() -> (SZ_AwardBox:Safe Zone Object Reference)
    InPadLeft:bool=true
    InPadRight:bool=true
    InPadTop:bool=true
    InPadBottom:bool=true
   ) -> (then)
   DeactivateWidget(execute=@n2.else)
  ]}
 BP_GetDesiredFocusTarget {nodes [
   Get Desired Focus Target()
   > Return Node!(execute, ReturnValue:Widget Object Reference=Get ListView_Items() -> (ListView_Items:Common List View Object Reference))
  ]}
 EventGraph {nodes [
   "On Clicked (Button_LeftScrollArrow)"!() -> (OutputDelegate, then, Button:Common Button Base Object Reference) {is=title K2Node_ComponentBoundEvent}
   > PaginateToNextOrPrevious!(execute, self, IsForward:bool=false) -> (then)
   "On Clicked (Back_Button)"!() -> (OutputDelegate, then, Button:Common Button Base Object Reference) {is=title K2Node_ComponentBoundEvent}
   > DeactivateWidget()
   n1 = "On Item Selection Changed (ListView_Items)"!() -> (OutputDelegate, then, Item:Object Reference, bIsSelected:bool) {is=title K2Node_ComponentBoundEvent}
   > Branch(Condition=@n1.bIsSelected)
   > n2 = Cast To FortItemVM!(execute, Object=@n1.Item) -> (then, CastFailed, AsFort Item VM:Fort Item VM Object Reference) {is=title K2Node_DynamicCast}
   > UpdateItemDetails!(execute, self, InputPin:Fort Item VM Object Reference=@n2) -> (then)
   n3 = Get ListView_Items()
   > "SetText (Text)"(
    self=Get Text_Header() -> (Text_Header:Common Text Object Reference)
    InText="Get (a ref)"(
     Array:Array of Texts=Get Requirements() -> (Requirements:Array of Texts)
     Dimension 1=GetIndexForItem!(self:List View Object Reference=@n3, Item:Object Reference=@n2) -> (ReturnValue:int=0)
    ) -> (Output:"Text (by ref)")
   )
   "On Entry Released (ListView_Items)"!() -> (OutputDelegate, then, Widget:User Widget Object Reference) {is=title K2Node_ComponentBoundEvent}
   > HandleArrowsVisibility()
   "On Clicked (Button_RightScrollArrow)"!() -> (OutputDelegate, then, Button:Common Button Base Object Reference) {is=title K2Node_ComponentBoundEvent}
   > PaginateToNextOrPrevious(IsForward=true)
   n4 = HandleInputMethodChanged!() -> (OutputDelegate, then, bNewInputType:ECommonInputType Enum) {is=title K2Node_CustomEvent}
   > Set Input Type(Input Type:ECommonInputType Enum=@n4.bNewInputType) -> (Output_Get:ECommonInputType Enum)
   > HandleArrowsVisibility()
   Event On Activated()
   > ActivateWidget(self=Get FortItemPreviewWidget() -> (FortItemPreviewWidget:WBP Item Preview Object Reference))
   > HandleAnimationInitialState()
   > Activate ItemPreview Screen() {is=title K2Node_CallFunction}
   > n5 = Get!(execute:Exec) -> (then:Exec, else:Exec, InstantSelectedItem:Fort Item VM Object Reference)
   > Set Selected Item!(execute, self:List View Object Reference=Get ListView_Items(), Item:Object Reference=@n5) -> (then) {is=title K2Node_CallFunction}
   > HandleArrowsVisibility()
   Event On Deactivated()
   n6 = Get FortItemPreviewWidget()
   > SetCurrentPreviewItem!(execute, self:Fort Item Preview Widget Object Reference=@n6, ItemVM:Fort Item VM Object Reference) -> (then)
   > DeactivateWidget(self=@n6)
   > Set InstantSelectedItem(InstantSelectedItem:Fort Item VM Object Reference) -> (Output_Get:Fort Item VM Object Reference)
   > SetSidesToPad(self=Get SZ_AwardBox(), InPadLeft=false, InPadRight=false, InPadTop=false, InPadBottom=false)
   K2Node_Event_11 = OnUpdateOwnedItemText!() -> (OutputDelegate, then, bShowText:bool=~) {is=title K2Node_CustomEvent}
   > SetVisibility(
    self=Get WBP_RewardInfo_Unlocked() -> (WBP_RewardInfo_Unlocked:WBP Reward Info Unlocked Object Reference)
    InVisibility=Select(Option 0=Collapsed, Option 1=SelfHitTestInvisible, Index=@K2Node_Event_11.bShowText)
   )
   Event Construct()
   n7 = Create Event()
   n8 = Get CommonInputSubsystem()
   > Bind Event to On Input Method Changed(self=@n8, Delegate=@n7)
   > HandleInputMethodChanged!(execute, self, bNewInputType:ECommonInputType Enum=GetCurrentInputType(self=@n8)) -> (then) {is=title K2Node_CallFunction}
   > SetMaxDesiredHeight!(
    execute
    self:Size Box Object Reference=Get SizeBoxTitle!(self:WBP UIKit Item Description Base Object Reference=Get ItemDescription() -> (ItemDescription:WBP UIKit Description With Subtitle Object Reference)) -> (SizeBoxTitle:Size Box Object Reference)
    InMaxDesiredHeight:float=160
   ) -> (then)
   Event Destruct()
   > Unbind Event from On Input Method Changed(self=@n8, Delegate=@n7)
  ]}
 HandleAnimationInitialState {nodes [
   HandleAnimationInitialState() {is=title K2Node_FunctionEntry}
   > PlayAnimation(InAnimation=Get Anim_InitialStateSet() -> (Anim_InitialStateSet:Widget Animation Object Reference))
  ]}
 HandleArrowsVisibility {nodes [
   HandleArrowsVisibility() {is=title K2Node_FunctionEntry}
   n1 = Get ListView_Items()
   n2 = GetItemAt!(self:List View Object Reference=@n1, Index:int=0) -> (ReturnValue:Object Reference)
   n3 = Get Input Type() -> (Input Type:ECommonInputType Enum)
   > SetVisibility(
    self=Get Button_LeftScrollArrow() -> (Button_LeftScrollArrow:WBP UIKit Button Regular Object Reference)
    InVisibility=Select(
     Option 0=Select!(
      MouseAndKeyboard:ESlateVisibility Enum=Visible
      Gamepad:ESlateVisibility Enum=Hidden
      Touch:ESlateVisibility Enum=Visible
      Count:ESlateVisibility Enum=Visible
      Index:ECommonInputType Enum=@n3
     ) -> (ReturnValue)
     Option 1=Hidden
     Index=Is Item Visible!(self:List View Object Reference=@n1, Item:Object Reference=@n2) -> (ReturnValue:bool)
    )
   )
   n4 = Get ListView_Items()
   n5 = GetItemAt(self=@n4, Index=int - int(A=GetNumItems!(self:List View Object Reference=@n4) -> (ReturnValue:int=0)))
   n6 = Get Input Type()
   > SetVisibility(
    self=Get Button_RightScrollArrow() -> (Button_RightScrollArrow:WBP UIKit Button Regular Object Reference)
    InVisibility=Select!(
     Option 0:ESlateVisibility Enum=Select!(
      MouseAndKeyboard:ESlateVisibility Enum=Visible
      Gamepad:ESlateVisibility Enum=Hidden
      Touch:ESlateVisibility Enum=Visible
      Count:ESlateVisibility Enum=Visible
      Index:ECommonInputType Enum=@n6
     ) -> (ReturnValue)
     Option 1:ESlateVisibility Enum=Hidden
     Index:bool=Is Item Visible(self=@n4, Item=@n5)
    ) -> (ReturnValue)
   )
  ]}
 HandleItemListVisible {nodes [
   HandleItemListVisible() {is=title K2Node_FunctionEntry}
   n1 = GetNumItems(self=Get ListView_Items()) -> (ReturnValue=0)
   > SetVisibility(
    self=Get GP_ItemList() -> (GP_ItemList:Grid Panel Object Reference)
    InVisibility=Select!(Option 0:ESlateVisibility Enum=Visible, Option 1:ESlateVisibility Enum=Collapsed, Index:bool="integer <= integer"(A=@n1, B=1)) -> (ReturnValue)
   )
  ]}
 InitItems {nodes [
   n1 = InitItems!() -> (then, Payouts:Array of Tournament Leaderboard Payout Table Data Structures) {is=title K2Node_FunctionEntry}
   > Clear(TargetArray:Array of Texts=Get Requirements())
   > Clear(TargetArray:Array of Fort Item VM Object References=Get Cosmetic Payout Items VM() -> (Cosmetic Payout Items VM:Array of Fort Item VM Object References))
   > n2 = For Each Loop(Array:Array of Tournament Leaderboard Payout Table Data Structures=@n1) -> (Array Element:Tournament Leaderboard Payout Table Data Structure)
   > n3 = For Each Loop(Array:Array of Tournament Payout Threshold Data Structures=Break Tournament Leaderboard Payout Table Data!(TournamentLeaderboardPayoutTableData:"Tournament Leaderboard Payout Table Data Structure (by ref)"=@n2) -> (PayoutThresholdsData:Array of Tournament Payout Threshold Data Structures) {is=title K2Node_BreakStruct}) -> (Array Element:Tournament Payout Threshold Data Structure)
   n4 = Get ListView_Items()
   Set List Items!(execute=@n3.Completed, self:List View Object Reference=@n4, InListItems:Array of Object References=Get Cosmetic Payout Items VM()) -> (then) {is=title K2Node_CallFunction}
   > n5 = Get()
   > Set Selected Item(self=@n4, Item=@n5)
   > SetCompatibilityVisibility!(execute, self:WBP UIKit Item Description Base Object Reference=Get ItemDescription(), Show:bool=true) -> (then, IsVisible:bool)
   n6 = Break Tournament Payout Threshold Data!(TournamentPayoutThresholdData:"Tournament Payout Threshold Data Structure (by ref)"=@n3) -> (
    ScoringType:EPayoutScoringType Enum
    Threshold:double=~
    ThresholdDisplayText:Text
    ThresholdDisplayTextHeader:Text
    SourceLeaderboardDisplayName:Text
    CosmeticPayoutItemsVM:Array of Fort Item VM Object References
   ) {is=title K2Node_BreakStruct}
   Append Array!(
    execute=@n3.LoopBody
    TargetArray:Array of Fort Item VM Object References=Get Cosmetic Payout Items VM()
    SourceArray:Array of Fort Item VM Object References=@n6.CosmeticPayoutItemsVM
   ) -> (then)
   > For Each Loop(Array:Array of Fort Item VM Object References=@n6.CosmeticPayoutItemsVM) -> (Array Element:Fort Item VM Object Reference)
   > Add(TargetArray:Array of Texts=Get Requirements(), NewItem:"Text (by ref)"=@n6.ThresholdDisplayText) {is=title K2Node_CallArrayFunction}
  ]}
 PaginateToNextOrPrevious {nodes [
   n1 = Get ListView_Items()
   n2 = GetIndexForItem(self=@n1, Item=Get Selected Item!(self:List View Object Reference=@n1) -> (ReturnValue:Object Reference) {is=title K2Node_CallFunction}) -> (ReturnValue=0)
   n3 = int - int(A=@n2)
   n4 = Get ListView_Items()
   n5 = ScrollIndexIntoView!(execute, self:List View Object Reference=@n4, Index:int=@n3) -> (then)
   SetSelectedIndex!(execute=[@n5.then, @n26.then], self:List View Object Reference=@n4, Index:int=@n3) -> (then)
   n6 = int + int(A=@n2)
   n7 = ScrollIndexIntoView(self=@n4, Index=@n6)
   SetSelectedIndex(execute=[@n7.then, @n25.then], self=@n4, Index=@n6)
   PaginateToNextOrPrevious!() -> (then, IsForward:bool=~) {is=title K2Node_FunctionEntry}
   > n8 = Sequence!(execute) -> (then_0, then_1, then_2:Exec)
   n9 = Get ListView_Items()
   > Set SelectedItemIndex(SelectedItemIndex:int=GetIndexForItem(self=@n9, Item=Get Selected Item(self=@n9)) -> (ReturnValue=0)) -> (Output_Get:int=0)
   n10 = Get ListView_Items()
   > Set CountOfItemsDisplayed(CountOfItemsDisplayed:int=Length(TargetArray:Array of User Widget Object References=GetDisplayedEntryWidgets!(self:List View Base Object Reference=@n10) -> (ReturnValue:Array of User Widget Object References))) -> (Output_Get:int=0)
   > Set TotalCountOfItems(TotalCountOfItems:int=GetNumItems(self=Get ListView_Items()) -> (ReturnValue=0)) -> (Output_Get:int=0)
   > Set SelectedItemPage(SelectedItemPage:int=int / int(A=Get SelectedItemIndex() -> (SelectedItemIndex:int=0), B=Get CountOfItemsDisplayed() -> (CountOfItemsDisplayed:int=0))) -> (Output_Get:int=0)
   n11 = "Not Equal (Integer)"(A=Get LastSelectedIndex() -> (LastSelectedIndex:int=0), B=Get SelectedItemIndex() -> (SelectedItemIndex=0))
   n12 = SelectInt(A=0, B=Get Pages Moved() -> (Pages Moved:int=0), bPickA=@n11)
   n13 = SelectInt(A=1, B=-1, bPickA=Get IsForward() -> (IsForward:bool))
   n14 = int / int(A=Get TotalCountOfItems() -> (TotalCountOfItems:int=0), B=Get CountOfItemsDisplayed() -> (CountOfItemsDisplayed=0))
   n15 = Get SelectedItemPage() -> (SelectedItemPage:int=0)
   n16 = Get SelectedItemPage() -> (SelectedItemPage=0)
   Set Pages Moved(
    execute=@n8.then_1
    Pages Moved:int="Clamp (Integer)"(Value=int + int(A=@n12, B=@n13), Min=int * int(A=@n16, B=-1), Max=int - int(A=@n14, B=@n15)) -> (ReturnValue=0)
   ) -> (Output_Get:int=0)
   > Branch(Condition=@n11)
   > Set LastSelectedIndex(LastSelectedIndex:int=Get SelectedItemIndex() -> (SelectedItemIndex=0)) -> (Output_Get:int=0)
   n17 = int - int(A=Get TotalCountOfItems() -> (TotalCountOfItems=0))
   n18 = Get SelectedItemIndex() -> (SelectedItemIndex=0)
   n19 = Get CountOfItemsDisplayed() -> (CountOfItemsDisplayed=0)
   n20 = int * int(A=int + int(A=int / int(A=@n18, B=@n19), B=Get Pages Moved() -> (Pages Moved=0)), B=@n19)
   n21 = int + int!(A=@n20, B=int / int(A=Get CountOfItemsDisplayed() -> (CountOfItemsDisplayed=0), B=2), C:int=-1) -> (ReturnValue)
   n22 = int - int(A=Get TotalCountOfItems() -> (TotalCountOfItems=0))
   NavigateToIndex!(
    execute=@n8.then_2
    self:List View Object Reference=Get ListView_Items()
    Index:int=SelectInt(
     A="Clamp (Integer)"(Value=@n21, Min=0, Max=@n22) -> (ReturnValue=0)
     B="Clamp (Integer)"(Value=@n20, Min=0, Max=@n17) -> (ReturnValue=0)
     bPickA=Get IsForward()
    )
   ) -> (then)
   n23 = Branch()
   n24 = GetNumItems(self=Get ListView_Items()) -> (ReturnValue=0)
   > n25 = Branch(Condition="integer < integer"(A=@n2, B=int - int(A=@n24)))
   n26 = Branch(execute=@n23.else, Condition="integer > integer"(A=@n2))
   SelectInt(A=1, B=-1, bPickA=Get IsForward())
  ]}
 PlayIntro {nodes [
   PlayIntro() {is=title K2Node_FunctionEntry}
   > PlayAnimation(InAnimation=Get Anim_Intro() -> (Anim_Intro:Widget Animation Object Reference))
   SpawnSound2D!(
    execute
    Sound:Sound Base Object Reference=Get WhooshSound() -> (WhooshSound:Sound Base Object Reference)
    VolumeMultiplier:float=1
    PitchMultiplier:float=1
    StartTime:float=0
    ConcurrencySettings:Sound Concurrency Object Reference
    bPersistAcrossLevelTransition:bool=false
    bAutoDestroy:bool=true
   ) -> (then, ReturnValue:Audio Component Object Reference)
  ]}
 Sequencer Events {nodes [
   BeforeShowArrows() {is=title K2Node_CustomEvent}
   > HandleArrowsVisibility() {is=title K2Node_CallFunction}
  ]}
 UpdateItemDetails {nodes [
   n1 = UpdateItemDetails!() -> (then, InputPin:Fort Item VM Object Reference) {is=title K2Node_FunctionEntry}
   > SetCurrentPreviewItem(self=Get FortItemPreviewWidget(), ItemVM=@n1)
   > n2 = Cast To FortPlayerController(Object=GetOwningPlayer())
   > "Fire Event Element Interacted (Non-Localized)"!(
    execute
    FortPC:Fort Player Controller Object Reference=@n2
    ElementName:String=GetItemTemplateId!(self:Fort Item VM Object Reference=@n1) -> (ReturnValue:String)
    AdditionalInfo:Flow Analytics Additional Info Structure
   ) -> (then)
   > Set FortItemDetailsVM!(
    execute
    FortItemDetailsVM:Fort Item Details VM Object Reference=Get ItemDetails!(self:Fort Item VM Object Reference=@n1) -> (ItemDetails:Fort Item Details VM Object Reference)
    self:WBP UIKit Description With Subtitle Object Reference=Get ItemDescription()
   ) -> (then, Output_Get:Fort Item Details VM Object Reference)
  ]}
]
variables [
 Cosmetic Payout Items VM {type="array<FortItemVM>", default []}
 Input Type {type=ECommonInputType, default=MouseAndKeyboard}
 InstantSelectedItem {type=FortItemVM, default=null}
 LastSelectedIndex {type=integer}
 Pages Moved {type=integer}
 Requirements {type="array<string>", default []}
 WhooshSound {type=SoundBase, default=null}
]
widgets {
 info {
  inheritedWidgetCount=0
  namedSlotCount=1
  parentClass=&CommonActivatableWidget
  rootWidgetClass=&Overlay
  widgetCount=33
 }
 widgets [
  Root {widgetClassPath=&Overlay, widget=Root}
  SZ_AwardBox {
   widgetClassPath=&SafeZone
   bIsVariable=true
   parent=Root
   widget=SZ_AwardBox
   properties {padBottom=false, padLeft=false, padRight=false, padTop=false}
   slot {class=OverlaySlot, properties {horizontalAlignment=HAlign_Fill, verticalAlignment=VAlign_Fill}}
  }
  Content {widgetClassPath=&Overlay, bIsVariable=true, parent=SZ_AwardBox, widget=Content, slot {class=SafeZoneSlot}}
  FortItemPreviewBackground {
   widgetClassPath=/Script/FortniteUI.FortItemPreviewBackground
   bIsVariable=true
   parent=Content
   widget=FortItemPreviewBackground
   properties {previewSceneChangerTag {tagName=UI.ItemPreview.SceneChanger.MPItemShop}}
   slot {class=OverlaySlot, properties {horizontalAlignment=HAlign_Fill, verticalAlignment=VAlign_Fill}}
  }
  SizeBox_CameraFrame {
   widgetClassPath=&SizeBox
   parent=Content
   widget=SizeBox_CameraFrame
   properties {bOverride_WidthOverride=true, heightOverride=864, maxDesiredWidth=1200, minDesiredWidth=1024, widthOverride=800}
   slot {class=OverlaySlot, properties {horizontalAlignment=HAlign_Center, verticalAlignment=VAlign_Fill, padding=<ltrb 0 60 0 160>}}
  }
  FortItemPreviewWidget {
   widgetClassPath=/FortniteCoreUI/Frontend/ItemPreview/WBP_ItemPreview
   bIsVariable=true
   parent=SizeBox_CameraFrame
   widget=FortItemPreviewWidget
   properties {
    bAllowPawnOverrideOnExclusiveItems=true
    bCanDisplayUserAccessories=false
    bShowFloor=false
    bSupportsActivationFocus=false
    itemViewObject=&self:WidgetTree.FortItemPreviewWidget.ItemViewObject
    previewActionButtons [null]
    previewSceneChangerTag {tagName=UI.ItemPreview.SceneChanger.MPItemShop}
   }
   slot {class=SizeBoxSlot}
  }
  VB_Content {
   widgetClassPath=&VerticalBox
   parent=Content
   widget=VB_Content
   slot {class=OverlaySlot, properties {horizontalAlignment=HAlign_Fill, verticalAlignment=VAlign_Fill, padding=<ltrb 0 32 0 64>}}
  }
  HB_Panels {
   widgetClassPath=&HorizontalBox
   parent=VB_Content
   widget=HB_Panels
   slot {class=VerticalBoxSlot, properties {padding=<ltrb 80 0 80 0>, size {sizeRule=Fill, value=1}}}
  }
  LeftsideContentSwitcher {
   widgetClassPath=&WidgetSwitcher
   bIsVariable=true
   parent=HB_Panels
   widget=LeftsideContentSwitcher
   slot {class=HorizontalBoxSlot, properties {size {sizeRule=Fill, value=0.9}}}
  }
  DefaultContent {
   widgetClassPath=&Overlay
   bIsVariable=true
   parent=LeftsideContentSwitcher
   widget=DefaultContent
   slot {class=WidgetSwitcherSlot, properties {verticalAlignment=VAlign_Top, padding=<ltrb 0 260 0 0>}}
  }
  HorizontalBox_0 {
   widgetClassPath=&HorizontalBox
   parent=DefaultContent
   widget=HorizontalBox_0
   properties {bOverride_Cursor=true}
   slot {class=OverlaySlot, properties {horizontalAlignment=HAlign_Fill, verticalAlignment=VAlign_Center}}
  }
  VerticalBox_1 {
   widgetClassPath=&VerticalBox
   parent=HorizontalBox_0
   widget=VerticalBox_1
   slot {class=HorizontalBoxSlot, properties {verticalAlignment=VAlign_Center, size {sizeRule=Fill, value=1}}}
  }
  Text_Body {
   widgetClassPath=&CommonTextBlock
   bIsVariable=true
   parent=VerticalBox_1
   widget=Text_Body
   properties {
    autoWrapText=true
    flowDirectionPreference=LeftToRight
    lineHeightPercentage=1.29
    text=The award will be given if the conditions are met:
    colorAndOpacity {colorUseRule=UseColor_Specified, specifiedColor=<rgba 0.5776 0.5776 0.5776 1>}
    font {
     bForceMonospaced=false
     bMaterialIsStencil=false
     fontMaterial=null
     fontObject=&HeadingNow
     letterSpacing=0
     monospacedWidth=1
     size=18
     skewAmount=0
     typefaceFontName=74 Regular
     outlineSettings {
      bApplyOutlineToDropShadows=false
      bMiteredCorners=false
      bSeparateFillAlpha=false
      outlineMaterial=null
      outlineSize=0
      outlineColor=<rgba 0 0 0 1>
     }
    }
    shapedTextOptions {bOverride_TextFlowDirection=true, bOverride_TextShapingMethod=false, textFlowDirection=LeftToRight, textShapingMethod=Auto}
   }
   slot {class=VerticalBoxSlot, properties {verticalAlignment=VAlign_Top, padding=<ltrb 0 8 0 0>}}
  }
  Text_Header {
   widgetClassPath=&CommonTextBlock
   bIsVariable=true
   parent=VerticalBox_1
   widget=Text_Header
   properties {
    autoWrapText=true
    lineHeightPercentage=0.85
    text=TOP 1
    textTransformPolicy=ToUpper
    font {
     bForceMonospaced=false
     bMaterialIsStencil=false
     fontMaterial=null
     fontObject=&HeadingNow
     letterSpacing=0
     monospacedWidth=1
     size=32.25
     skewAmount=0
     typefaceFontName=76 Bold Italic
     outlineSettings {
      bApplyOutlineToDropShadows=false
      bMiteredCorners=false
      bSeparateFillAlpha=false
      outlineMaterial=null
      outlineSize=0
      outlineColor=<rgba 0 0 0 1>
     }
    }
   }
   slot {class=VerticalBoxSlot, properties {verticalAlignment=VAlign_Top}}
  }
  Overlay_CustomContent {
   widgetClassPath=&Overlay
   bIsVariable=true
   parent=LeftsideContentSwitcher
   widget=Overlay_CustomContent
   slot {class=WidgetSwitcherSlot, properties {verticalAlignment=VAlign_Top, padding=<ltrb 0 260 0 0>}}
  }
  CustomContent {
   widgetClassPath=&NamedSlot
   bIsVariable=true
   parent=Overlay_CustomContent
   widget=CustomContent
   properties {bExposeOnInstanceOnly=true}
   slot {class=OverlaySlot, properties {horizontalAlignment=HAlign_Fill, verticalAlignment=VAlign_Center}}
  }
  Spacer_ItemPreview {
   widgetClassPath=&Spacer
   parent=HB_Panels
   widget=Spacer_ItemPreview
   slot {class=HorizontalBoxSlot, properties {size {sizeRule=Fill, value=1}}}
  }
  VB_RightPanel {
   widgetClassPath=&VerticalBox
   parent=HB_Panels
   widget=VB_RightPanel
   slot {class=HorizontalBoxSlot, properties {horizontalAlignment=HAlign_Right, size {sizeRule=Fill, value=0.8}}}
  }
  VB_ItemDescription {
   widgetClassPath=&VerticalBox
   parent=VB_RightPanel
   widget=VB_ItemDescription
   slot {class=VerticalBoxSlot, properties {verticalAlignment=VAlign_Top, padding=<ltrb 0 260 0 0>, size {sizeRule=Fill, value=1}}}
  }
  ItemDescription {
   widgetClassPath=/FortUILibrary/UIKit/Items/WBP_UIKit_DescriptionWithSubtitle
   bIsVariable=true
   parent=VB_ItemDescription
   widget=ItemDescription
   properties {
    bodyTextCustomTextStyle=/Game/Athena/Items/Cosmetics/Metadata/ItemDescriptionTextStyle
    bodyTextRow=3
    bottomSlotRow=5
    compatibilityRow=4
    hideBodyText=false
    hideCompatibility=true
    hideGameplayTagsText=false
    hideSetText=false
    hideSubtitle=true
    tagsOrientation=Orient_Vertical
    use Rich Description=false
    compatibilityPadding=<ltrb 0 0 0 0>
   }
   slot {class=VerticalBoxSlot}
  }
  WBP_RewardInfo_Unlocked {
   widgetClassPath=&/WBP_RewardInfo_Unlocked
   bIsVariable=true
   parent=VB_RightPanel
   widget=WBP_RewardInfo_Unlocked
   properties {visibility=Collapsed}
   slot {class=VerticalBoxSlot, properties {padding=<ltrb 0 16 0 0>}}
  }
  GP_ItemList {
   widgetClassPath=&GridPanel
   bIsVariable=true
   parent=VB_Content
   widget=GP_ItemList
   properties {columnFill [0,1,0], rowFill [1]}
   slot {class=VerticalBoxSlot, properties {horizontalAlignment=HAlign_Center, verticalAlignment=VAlign_Top, padding=<ltrb 80 0 80 0>}}
  }
  O_ScrollingListContainer {
   widgetClassPath=&Overlay
   parent=GP_ItemList
   widget=O_ScrollingListContainer
   properties {clipping=ClipToBounds}
   slot {class=GridSlot, properties {column=1, padding=<ltrb -80 0 -80 0>}}
  }
  PostBufferUpdate_NoBlur {
   widgetClassPath=/Script/UMG.PostBufferUpdate
   bIsVariable=true
   parent=O_ScrollingListContainer
   widget=PostBufferUpdate_NoBlur
   properties {
    visibility=HitTestInvisible
    buffersToUpdate [ESlatePostRT_0]
    updateBufferInfos [{
      bufferToUpdate=ESlatePostRT_0
      postParamUpdater=&self:WidgetTree.PostBufferUpdate_NoBlur.PostBufferBlurUpdater_0
     }]
   }
   slot {class=OverlaySlot, properties {horizontalAlignment=HAlign_Fill, verticalAlignment=VAlign_Fill}}
  }
  ListView_Items {
   widgetClassPath=&CommonListView
   bIsVariable=true
   parent=O_ScrollingListContainer
   widget=ListView_Items
   properties {
    bInEnableTouchAnimatedScrolling=true
    bReturnFocusToSelection=true
    clipping=Inherit
    entryWidgetClass=/Compete/TournamentDetails/WBP_CompeteRewardTile.WBP_CompeteRewardTile_C
    horizontalEntrySpacing=16
    numDesignerPreviewEntries=20
    orientation=Orient_Horizontal
    scrollBarStyle {
     thickness=0
     draggedThumbImage {
      bIsDynamicallyLoaded=false
      drawAs=Box
      imageType=FullColor
      mirroring=NoMirror
      resourceName=../../../Engine/Content/Slate/Common/Scrollbar_Thumb.png
      resourceObject=null
      tiling=NoTile
      imageSize=<0 0>
      margin=<ltrb 0.25 0.25 0.25 0.25>
      outlineSettings {
       bUseBrushTransparency=false
       roundingType=HalfHeightRadius
       width=0
       color {colorUseRule=UseColor_Specified, specifiedColor=<rgba 0 0 0 0>}
       cornerRadii=<0 0 0 0>
      }
      tintColor {colorUseRule=UseColor_Specified, specifiedColor=<rgba 1 1 1 1>}
      uVRegion {bIsValid=false, max=<0 0>, min=<0 0>}
     }
     horizontalBackgroundImage {
      bIsDynamicallyLoaded=false
      drawAs=NoDrawType
      imageType=NoImage
      mirroring=NoMirror
      resourceName=None
      resourceObject=null
      tiling=NoTile
      imageSize=<0 0>
      margin=<ltrb 0 0 0 0>
      outlineSettings {
       bUseBrushTransparency=false
       roundingType=HalfHeightRadius
       width=0
       color {colorUseRule=UseColor_Specified, specifiedColor=<rgba 0 0 0 0>}
       cornerRadii=<0 0 0 0>
      }
      tintColor {colorUseRule=UseColor_Specified, specifiedColor=<rgba 1 1 1 1>}
      uVRegion {bIsValid=false, max=<0 0>, min=<0 0>}
     }
     horizontalBottomSlotImage {
      bIsDynamicallyLoaded=false
      drawAs=Image
      imageType=FullColor
      mirroring=NoMirror
      resourceName=../../../Engine/Content/Slate/Common/Scrollbar_Background_Horizontal.png
      resourceObject=null
      tiling=NoTile
      imageSize=<8 8>
      margin=<ltrb 0 0 0 0>
      outlineSettings {
       bUseBrushTransparency=false
       roundingType=HalfHeightRadius
       width=0
       color {colorUseRule=UseColor_Specified, specifiedColor=<rgba 0 0 0 0>}
       cornerRadii=<0 0 0 0>
      }
      tintColor {colorUseRule=UseColor_Specified, specifiedColor=<rgba 1 1 1 1>}
      uVRegion {bIsValid=false, max=<0 0>, min=<0 0>}
     }
     horizontalTopSlotImage {
      bIsDynamicallyLoaded=false
      drawAs=Image
      imageType=FullColor
      mirroring=NoMirror
      resourceName=../../../Engine/Content/Slate/Common/Scrollbar_Background_Horizontal.png
      resourceObject=null
      tiling=NoTile
      imageSize=<8 8>
      margin=<ltrb 0 0 0 0>
      outlineSettings {
       bUseBrushTransparency=false
       roundingType=HalfHeightRadius
       width=0
       color {colorUseRule=UseColor_Specified, specifiedColor=<rgba 0 0 0 0>}
       cornerRadii=<0 0 0 0>
      }
      tintColor {colorUseRule=UseColor_Specified, specifiedColor=<rgba 1 1 1 1>}
      uVRegion {bIsValid=false, max=<0 0>, min=<0 0>}
     }
     hoveredThumbImage {
      bIsDynamicallyLoaded=false
      drawAs=Box
      imageType=FullColor
      mirroring=NoMirror
      resourceName=../../../Engine/Content/Slate/Common/Scrollbar_Thumb.png
      resourceObject=null
      tiling=NoTile
      imageSize=<0 0>
      margin=<ltrb 0.25 0.25 0.25 0.25>
      outlineSettings {
       bUseBrushTransparency=false
       roundingType=HalfHeightRadius
       width=0
       color {colorUseRule=UseColor_Specified, specifiedColor=<rgba 0 0 0 0>}
       cornerRadii=<0 0 0 0>
      }
      tintColor {colorUseRule=UseColor_Specified, specifiedColor=<rgba 1 1 1 1>}
      uVRegion {bIsValid=false, max=<0 0>, min=<0 0>}
     }
     normalThumbImage {
      bIsDynamicallyLoaded=false
      drawAs=Box
      imageType=FullColor
      mirroring=NoMirror
      resourceName=../../../Engine/Content/Slate/Common/Scrollbar_Thumb.png
      resourceObject=null
      tiling=NoTile
      imageSize=<0 0>
      margin=<ltrb 0.25 0.25 0.25 0.25>
      outlineSettings {
       bUseBrushTransparency=false
       roundingType=HalfHeightRadius
       width=0
       color {colorUseRule=UseColor_Specified, specifiedColor=<rgba 0 0 0 0>}
       cornerRadii=<0 0 0 0>
      }
      tintColor {colorUseRule=UseColor_Specified, specifiedColor=<rgba 1 1 1 1>}
      uVRegion {bIsValid=false, max=<0 0>, min=<0 0>}
     }
     verticalBackgroundImage {
      bIsDynamicallyLoaded=false
      drawAs=NoDrawType
      imageType=NoImage
      mirroring=NoMirror
      resourceName=None
      resourceObject=null
      tiling=NoTile
      imageSize=<0 0>
      margin=<ltrb 0 0 0 0>
      outlineSettings {
       bUseBrushTransparency=false
       roundingType=HalfHeightRadius
       width=0
       color {colorUseRule=UseColor_Specified, specifiedColor=<rgba 0 0 0 0>}
       cornerRadii=<0 0 0 0>
      }
      tintColor {colorUseRule=UseColor_Specified, specifiedColor=<rgba 1 1 1 1>}
      uVRegion {bIsValid=false, max=<0 0>, min=<0 0>}
     }
     verticalBottomSlotImage {
      bIsDynamicallyLoaded=false
      drawAs=Image
      imageType=FullColor
      mirroring=NoMirror
      resourceName=../../../Engine/Content/Slate/Common/Scrollbar_Background_Vertical.png
      resourceObject=null
      tiling=NoTile
      imageSize=<8 8>
      margin=<ltrb 0 0 0 0>
      outlineSettings {
       bUseBrushTransparency=false
       roundingType=HalfHeightRadius
       width=0
       color {colorUseRule=UseColor_Specified, specifiedColor=<rgba 0 0 0 0>}
       cornerRadii=<0 0 0 0>
      }
      tintColor {colorUseRule=UseColor_Specified, specifiedColor=<rgba 1 1 1 1>}
      uVRegion {bIsValid=false, max=<0 0>, min=<0 0>}
     }
     verticalTopSlotImage {
      bIsDynamicallyLoaded=false
      drawAs=Image
      imageType=FullColor
      mirroring=NoMirror
      resourceName=../../../Engine/Content/Slate/Common/Scrollbar_Background_Vertical.png
      resourceObject=null
      tiling=NoTile
      imageSize=<8 8>
      margin=<ltrb 0 0 0 0>
      outlineSettings {
       bUseBrushTransparency=false
       roundingType=HalfHeightRadius
       width=0
       color {colorUseRule=UseColor_Specified, specifiedColor=<rgba 0 0 0 0>}
       cornerRadii=<0 0 0 0>
      }
      tintColor {colorUseRule=UseColor_Specified, specifiedColor=<rgba 1 1 1 1>}
      uVRegion {bIsValid=false, max=<0 0>, min=<0 0>}
     }
    }
   }
   slot {class=OverlaySlot, properties {horizontalAlignment=HAlign_Fill, verticalAlignment=VAlign_Fill, padding=<ltrb 80 32 80 24>}}
  }
  LeftFade {
   widgetClassPath=&Image
   bIsVariable=true
   parent=O_ScrollingListContainer
   widget=LeftFade
   properties {
    visibility=HitTestInvisible
    brush {
     bIsDynamicallyLoaded=false
     drawAs=Image
     imageType=FullColor
     mirroring=NoMirror
     resourceName=None
     resourceObject=/Compete/MaterialInstances/MI_UI_ListFade_Left
     tiling=NoTile
     imageSize=<80 32>
     margin=<ltrb 0 0 0 0>
     outlineSettings {
      bUseBrushTransparency=false
      roundingType=HalfHeightRadius
      width=0
      color {colorUseRule=UseColor_Specified, specifiedColor=<rgba 0 0 0 0>}
      cornerRadii=<0 0 0 0>
     }
     tintColor {colorUseRule=UseColor_Specified, specifiedColor=<rgba 1 1 1 1>}
     uVRegion {bIsValid=false, max=<0 0>, min=<0 0>}
    }
   }
   slot {class=OverlaySlot, properties {verticalAlignment=VAlign_Fill}}
  }
  RightFade {
   widgetClassPath=&Image
   bIsVariable=true
   parent=O_ScrollingListContainer
   widget=RightFade
   properties {
    visibility=HitTestInvisible
    brush {
     bIsDynamicallyLoaded=false
     drawAs=Image
     imageType=FullColor
     mirroring=NoMirror
     resourceName=None
     resourceObject=/Compete/MaterialInstances/MI_UI_ListFade_Right
     tiling=NoTile
     imageSize=<80 32>
     margin=<ltrb 0 0 0 0>
     outlineSettings {
      bUseBrushTransparency=false
      roundingType=HalfHeightRadius
      width=0
      color {colorUseRule=UseColor_Specified, specifiedColor=<rgba 0 0 0 0>}
      cornerRadii=<0 0 0 0>
     }
     tintColor {colorUseRule=UseColor_Specified, specifiedColor=<rgba 1 1 1 1>}
     uVRegion {bIsValid=false, max=<0 0>, min=<0 0>}
    }
   }
   slot {class=OverlaySlot, properties {horizontalAlignment=HAlign_Right, verticalAlignment=VAlign_Fill}}
  }
  PostBufferUpdate_ResetBlur {
   widgetClassPath=&PostBufferUpdate
   bIsVariable=true
   parent=O_ScrollingListContainer
   widget=PostBufferUpdate_ResetBlur
   properties {
    visibility=HitTestInvisible
    buffersToUpdate [ESlatePostRT_0]
    updateBufferInfos [{
      bufferToUpdate=ESlatePostRT_0
      postParamUpdater=&self:WidgetTree.PostBufferUpdate_ResetBlur.PostBufferBlurUpdater_0
     }]
   }
   slot {class=OverlaySlot, properties {horizontalAlignment=HAlign_Fill, verticalAlignment=VAlign_Fill}}
  }
  Button_LeftScrollArrow {
   widgetClassPath=&WBP_UIKit_Button_Regular
   bIsVariable=true
   parent=GP_ItemList
   widget=Button_LeftScrollArrow
   properties {
    bHideInputAction=true
    bIsFocusable=false
    clipping=OnDemand
    iconMaterial=/FortUILibrary/Foundation/Buttons/Materials/Arrows/MI_UIKit_Icon_Button_Previous
    text=""
    visibility=Visible
   }
   slot {class=GridSlot, properties {horizontalAlignment=HAlign_Left, layer=10, verticalAlignment=VAlign_Center, padding=<ltrb 0 0 24 0>}}
  }
  Button_RightScrollArrow {
   widgetClassPath=&WBP_UIKit_Button_Regular
   bIsVariable=true
   parent=GP_ItemList
   widget=Button_RightScrollArrow
   properties {
    bHideInputAction=true
    bIsFocusable=false
    iconMaterial=/FortUILibrary/Foundation/Buttons/Materials/Arrows/MI_UIKit_Icon_Button_Next
    text=""
    visibility=Visible
   }
   slot {class=GridSlot, properties {column=2, horizontalAlignment=HAlign_Right, layer=10, verticalAlignment=VAlign_Center, padding=<ltrb 24 0 0 0>}}
  }
  BottomBar {
   widgetClassPath=/Game/UI/Frontend/BottomBar/WBP_AthenaBottomBarWidget
   bIsVariable=true
   parent=Root
   widget=BottomBar
   slot {class=OverlaySlot, properties {horizontalAlignment=HAlign_Right, verticalAlignment=VAlign_Bottom}}
  }
  BottomButtons {
   widgetClassPath=&WrapBox
   namedSlotHost=BottomBar
   widget=BottomButtons
   properties {bExplicitWrapSize=true, horizontalAlignment=HAlign_Right, wrapSize=1500, innerSlotPadding=<24 0>}
  }
  Back_Button {
   widgetClassPath=/FortUILibrary/UIKit/Buttons/WBP_UIKit_Button_Quiet
   bIsVariable=true
   parent=BottomButtons
   widget=Back_Button
   properties {
    bIsFocusable=false
    minWidth=120
    mobileMinWidth=200
    text=Back
    triggeringInputAction {dataTable=&InputActionDataTable, rowName=Back}
   }
   slot {class=WrapBoxSlot, properties {horizontalAlignment=HAlign_Right}}
  }
 ]
}
