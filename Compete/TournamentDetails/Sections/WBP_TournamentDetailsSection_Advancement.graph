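# Tournament details "Advancement" section widget that lists token payouts by placement threshold, labeled by division or item text.
# Set Advance Tokens Data counts Token and Commerce rewards to set section visibility, then adds a scoring row per token.
# UpdateTableColumns measures the first row to choose two or three columns, then adjusts grid span and background corner-rounding material parameters.
# PreConstruct sets the intro animation and builds placeholder rows at design time.
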WidgetBlueprint /Compete/TournamentDetails/Sections/WBP_TournamentDetailsSection_Advancement
animations [Anim_Intro {
  binding_count=2
  length_seconds=1
  object=&self:Anim_Intro
  sequence {bindings [
    GridSlot_2 {
     id=#1
     parent_id=#2
     possessed_class=GridSlot
     tracks [{
       type=MovieSceneMarginTrack
       display_name=Padding
       sections [{
         type=MovieSceneMarginSection
         end_frame=20
         start_frame=0
         channels [
          Left {type=MovieSceneScriptingFloatChannel, const [10,2]}
          Top {type=MovieSceneScriptingFloatChannel, keys [[0,10],[20,0]]}
          Right {type=MovieSceneScriptingFloatChannel, const [10,2]}
          Bottom {type=MovieSceneScriptingFloatChannel, const [10,2]}
         ]
        }]
      }]
    }
    SectionSlot {id=#2, possessed_class=NamedSlot, tracks []}
   ]}
 }]
graphs [
 AddItem {nodes [
   AddItem!() -> (then, Index:int) {is=title K2Node_FunctionEntry}
   > n1 = Create Widget!(
    execute:Exec
    Class:User Widget Class Reference=&/WBP_TournamentDetailsSectionItem_Scoring.WBP_TournamentDetailsSectionItem_Scoring_C
    OwningPlayer:Player Controller Object Reference
    VM_TournamentDetailsSectionItem_PrizeDebug:VM Tournament Details Section Item Prize Debug Object Reference
   ) -> (then:Exec, ReturnValue:WBP Tournament Details Section Item Scoring Object Reference)
   n2 = Get Index() -> (Index:int=0)
   n3 = Get MaxRowsPerColumn() -> (MaxRowsPerColumn:int=0)
   > n4 = AddChildToUniformGrid!(
    execute
    self:Uniform Grid Panel Object Reference=Get Table() -> (Table:Uniform Grid Panel Object Reference)
    Content:Widget Object Reference=@n1
    InRow:int="% (Integer)"(A=@n2, B=@n3) -> (ReturnValue=0)
    InColumn:int=int / int(A=@n2, B=@n3)
   ) -> (then, ReturnValue:Uniform Grid Slot Object Reference)
   > SetHorizontalAlignment(self:Uniform Grid Slot Object Reference=@n4, InHorizontalAlignment=HAlign_Fill)
   > SetVerticalAlignment(self:Uniform Grid Slot Object Reference=@n4, InVerticalAlignment=VAlign_Center)
   > Return Node!(execute, Item Widget:WBP Tournament Details Section Item Scoring Object Reference=@n1)
  ]}
 EventGraph {nodes [
   n1 = Event PreConstruct()
   > Parent: PreConstruct(IsDesignTime=@n1.IsDesignTime)
   > SetFocusAnim!(
    execute
    self:WBP Tournament Details Section Base Object Reference
    Anim_Focus:Widget Animation Object Reference=Get Anim_Intro() -> (Anim_Intro:Widget Animation Object Reference)
   ) -> (then)
   > UpdateBackground()
   > Branch(Condition=@n1.IsDesignTime)
   n2 = MakeLiteralInt!(Value:int=11) -> (ReturnValue:int=0)
   > n3 = For Loop(FirstIndex=1, LastIndex=@n2)
   > n4 = AddItem!(execute, self, Index:int=int - int(A=@n3)) -> (then, Item Widget:WBP Tournament Details Section Item Scoring Object Reference) {is=title K2Node_CallFunction}
   n5 = "To String (Integer)"(InInt=@n3)
   n6 = "To String (Integer)"(InInt=int * int(A=@n3, B=11))
   > Set Prize Data!(
    execute
    self:WBP Tournament Details Section Item Scoring Object Reference=@n4
    In Threshold Text:Text="To Text (String)"(InString=Append(A="Top ", B=@n6))
    In Value Text:Text="To Text (String)"(InString=Append(A="Division ", B=@n5))
   ) -> (then) {is=title K2Node_CallFunction}
   > Branch(Condition="Equal (Integer)"(A=@n3, B=1))
   > UpdateTableColumns!(execute, self, TableFirstElement:WBP Tournament Details Section Item Scoring Object Reference=@n4, In Num Items:int=@n2) -> (then)
  ]}
 Set Advance Tokens Data {nodes [
   Set Advance Tokens Data!() -> (then, In Prizes Data:Array of Tournament Leaderboard Payout Table Data Structures) {is=title K2Node_FunctionEntry}
   > ClearChildren(self=Get Table())
   > Set Filtered Prizes Num(Filtered Prizes Num:int=0) -> (Output_Get:int=0)
   > Set Prize Item Index(Prize Item Index:int=0) -> (Output_Get:int=0)
   > n1 = Sequence()
   > n2 = For Each Loop(Array:Array of Tournament Leaderboard Payout Table Data Structures=Get In Prizes Data() -> (In Prizes Data:Array of Tournament Leaderboard Payout Table Data Structures)) -> (Array Element:Tournament Leaderboard Payout Table Data Structure)
   > n3 = For Each Loop(Array:Array of Tournament Payout Threshold Data Structures=Break Tournament Leaderboard Payout Table Data!(TournamentLeaderboardPayoutTableData:"Tournament Leaderboard Payout Table Data Structure (by ref)"=@n2) -> (bIsMainWindowLeaderboard:bool=~, PayoutThresholdsData:Array of Tournament Payout Threshold Data Structures) {is=title K2Node_BreakStruct}.PayoutThresholdsData) -> (Array Element:Tournament Payout Threshold Data Structure)
   > n4 = For Each Loop(Array:Array of Tournament Payout Item Data Structures=Break Tournament Payout Threshold Data!(TournamentPayoutThresholdData:"Tournament Payout Threshold Data Structure (by ref)"=@n3) -> (
     ScoringType:EPayoutScoringType Enum
     Threshold:double=~
     ThresholdDisplayText:Text
     ThresholdDisplayTextHeader:Text
     SourceLeaderboardDisplayName:Text
     PayoutItemsData:Array of Tournament Payout Item Data Structures
     CosmeticPayoutItemsVM:Array of Fort Item VM Object References
    ) {is=title K2Node_BreakStruct}.PayoutItemsData) -> (Array Element:Tournament Payout Item Data Structure)
   > n5 = Switch on EPayoutRewardType!(
    execute
    Selection:EPayoutRewardType Enum=Break Tournament Payout Item Data!(TournamentPayoutItemData:"Tournament Payout Item Data Structure (by ref)"=@n4) -> (RewardType:EPayoutRewardType Enum, DisplayText:Text) {is=title K2Node_BreakStruct}
   ) -> (Invalid:Exec, Commerce:Exec, GameRelated:Exec, Token:Exec, Score:Exec) {is=title K2Node_SwitchEnum}
   Increment Int(" "=@n5.Commerce, Value=Get Cash Prizes Num() -> (Cash Prizes Num:int=0))
   Increment Int(" "=@n5.Token, Value=Get Filtered Prizes Num() -> (Filtered Prizes Num:int=0))
   n6 = Get Cash Prizes Num() -> (Cash Prizes Num=0)
   n7 = Get Filtered Prizes Num() -> (Filtered Prizes Num=0)
   n8 = OR Boolean(A="integer >= integer"(A=@n7, B=2), B=AND Boolean(A="Equal (Integer)"(A=@n7, B=1), B="integer >= integer"(A=@n6, B=1)))
   SetVisibility(execute=@n1.then_1, InVisibility=Select(Option 0=Collapsed, Option 1=SelfHitTestInvisible, Index=@n8))
   > Branch(Condition=@n8)
   > n9 = For Each Loop(Array:Array of Tournament Leaderboard Payout Table Data Structures=Get In Prizes Data()) -> (Array Element:Tournament Leaderboard Payout Table Data Structure)
   n10 = Get Filtered Prizes Num() -> (Filtered Prizes Num=0)
   SetVisibility(
    execute=@n9.Completed
    self=Get TableHeader() -> (TableHeader:Grid Panel Object Reference)
    InVisibility=Select(Option 0=SelfHitTestInvisible, Option 1=Collapsed, Index="Equal (Integer)"(A=@n10, B=1))
   )
   n11 = For Each Loop(
    Exec=@n9.LoopBody
    Array:Array of Tournament Payout Threshold Data Structures=Break Tournament Leaderboard Payout Table Data(TournamentLeaderboardPayoutTableData=@n9) -> (bIsMainWindowLeaderboard=~).PayoutThresholdsData
   ) -> (Array Element:Tournament Payout Threshold Data Structure)
   n12 = Break Tournament Payout Threshold Data(TournamentPayoutThresholdData=@n11) -> (Threshold=~)
   > n13 = For Each Loop(Array:Array of Tournament Payout Item Data Structures=@n12.PayoutItemsData) -> (Array Element:Tournament Payout Item Data Structure)
   n14 = Break Tournament Payout Item Data(TournamentPayoutItemData=@n13)
   > Branch(Condition="Equal (Enum)"(A:EPayoutRewardType Enum=@n14, B:EPayoutRewardType Enum=Token))
   n15 = Get Prize Item Index() -> (Prize Item Index:int=0)
   > n16 = AddItem(Index=@n15)
   n17 = Get FortCompeteUIManager() -> (ReturnValue:Fort Compete UIManager Object Reference) {is=title K2Node_GetSubsystem}
   n18 = int + int(A=@n15)
   > Set Prize Data(
    self=@n16
    In Threshold Text=@n12.ThresholdDisplayText
    In Value Text=SelectText(
     A=Format Text!(Format="LOCTEXT(\"Division {DivisionNumber}\")", DivisionNumber:int=@n18) -> (Result)
     B=@n14.DisplayText
     bPickA=ShowNextRoundTokensByDivisionLabels!(self:Fort Compete UIManager Object Reference=@n17) -> (ReturnValue:bool)
    )
   )
   > n19 = Increment Int(Value=Get Prize Item Index() -> (Prize Item Index=0))
   > Branch(Condition="Equal (Integer)"(A=@n19, B=1))
   > UpdateTableColumns(TableFirstElement=@n16, In Num Items=Get Filtered Prizes Num() -> (Filtered Prizes Num=0))
  ]}
 Set Table Grid Slot {nodes [
   n1 = Set Table Grid Slot!() -> (then, Num Items:int) {is=title K2Node_FunctionEntry}
   n2 = Get MaxRowsPerColumn() -> (MaxRowsPerColumn=0)
   > SetGridSlotColumnAndRow!(
    execute
    Widget:Widget Object Reference=Get TableSizeBox() -> (TableSizeBox:Size Box Object Reference)
    Row:int=2
    Row Span:int=0
    Column:int=0
    Column Span:int=Ceil(A=float / float(A:int=@n1, B:int=@n2)) {is=title K2Node_CallFunction}
   ) -> (then)
  ]}
 SetCornerRoundness {nodes [
   n1 = SetCornerRoundness!() -> (then, Widget:Image Object Reference, Value:Linear Color Structure) {is=title K2Node_FunctionEntry}
   > n2 = GetDynamicMaterial(self=@n1)
   > SetVectorParameterValue(self=@n2, ParameterName=Roundness, Value=@n1.Value)
  ]}
 SetTableVisibility {nodes [
   n1 = SetTableVisibility!() -> (then, ItemsAmount:int) {is=title K2Node_FunctionEntry}
   n2 = "Equal (Integer)"(A=@n1, B=1)
   > SetVisibility(
    self=[Get Center() -> (Center:Image Object Reference), Get BackgroundDark_Right() -> (BackgroundDark_Right:Image Object Reference)]
    InVisibility=Select(Option 0=SelfHitTestInvisible, Option 1=Collapsed, Index=@n2)
   )
   > SetCornerRoundness!(
    execute
    self
    Widget:Image Object Reference=Get BackgroundDark_Left() -> (BackgroundDark_Left:Image Object Reference)
    Value:Linear Color Structure=Select!(
     Option 0_R:float=25
     Option 0_G:float=25
     Option 0_B:float=0
     Option 0_A:float=0
     Option 1_R:float=25
     Option 1_G:float=25
     Option 1_B:float=25
     Option 1_A:float=25
     Index=@n2
    ) -> (ReturnValue:Linear Color Structure)
   ) -> (then) {is=title K2Node_CallFunction}
  ]}
 UpdateBackground {nodes [
   UpdateBackground() {is=title K2Node_FunctionEntry}
   > SetCornerRoundness!(execute, self, Widget=Get BackgroundDark_Left(), Value_R:float=25, Value_G:float=25, Value_B:float=0, Value_A:float=0) -> (then)
   n1 = Get BackgroundDark_Right()
   > SetCornerRoundness(Widget=@n1, Value_R=0, Value_G=0, Value_B=25, Value_A=25)
   n2 = Select!(Option 0:float=25, Option 1:float=0, Index=IsVisible(self=@n1)) -> (ReturnValue:float)
   > SetCornerRoundness(Widget=Get Center(), Value_R=0, Value_G=0, Value_B=@n2, Value_A=@n2)
  ]}
 UpdateTableColumns {nodes [
   n1 = UpdateTableColumns!() -> (then, TableFirstElement:WBP Tournament Details Section Item Scoring Object Reference, In Num Items:int) {is=title K2Node_FunctionEntry}
   > ForceLayoutPrepass(self:Widget Object Reference=@n1)
   n2 = IsMobileGame()
   n3 = Get TableFirstElement() -> (TableFirstElement:WBP Tournament Details Section Item Scoring Object Reference)
   n4 = "float > float"(
    A=GetDesiredSize!(self:Widget Object Reference=Get CommonText_Place!(self:WBP Tournament Details Section Item Scoring Object Reference=@n3) -> (CommonText_Place:Common Text Object Reference)) -> (ReturnValue_X:double, ReturnValue_Y:double)
    B=150
   )
   n5 = "float > float"(
    A=GetDesiredSize(self=Get CommonText_Points!(self:WBP Tournament Details Section Item Scoring Object Reference=@n3) -> (CommonText_Points:Common Text Object Reference))
    B=150
   )
   n6 = "Equal (Integer)"(A=Get In Num Items() -> (In Num Items:int=0), B=4)
   > Set IsContentSizeLargerThenCell(IsContentSizeLargerThenCell:bool=AND Boolean(A=OR Boolean!(A=@n4, B=@n5, C:bool=@n6) -> (ReturnValue), B=NOT Boolean(A=@n2))) -> (Output_Get:bool)
   > Set Num Columns(Num Columns:int=Select(Option 0:int=3, Option 1:int=2, Index=Get IsContentSizeLargerThenCell() -> (IsContentSizeLargerThenCell:bool)) -> (ReturnValue:int)) -> (Output_Get:int=0)
   n7 = Get In Num Items() -> (In Num Items=0)
   n8 = "To Float (Integer)"(InInt=@n7)
   n9 = "To Float (Integer)"(InInt=Get Num Columns() -> (Num Columns:int=0))
   > Set MaxRowsPerColumn(MaxRowsPerColumn:int=Ceil(A=float / float(A:double=@n8, B:double=@n9))) -> (Output_Get:int=0)
   > Set Table Grid Slot!(execute, self, Num Items:int=@n7) -> (then) {is=title K2Node_CallFunction}
   > SetColumnFill!(
    execute
    self:Grid Panel Object Reference=Get Content() -> (Content:Grid Panel Object Reference)
    ColumnIndex:int=2
    Coefficient:float=Select(Option 0:float=1, Option 1:float=0, Index=Get IsContentSizeLargerThenCell()) -> (ReturnValue:float)
   ) -> (then)
   > SetVisibility(
    self=Get BackgroundDark_Right()
    InVisibility=Select(Option 0:ESlateVisibility Enum=SelfHitTestInvisible, Option 1:ESlateVisibility Enum=Collapsed, Index=Get IsContentSizeLargerThenCell()) -> (ReturnValue:ESlateVisibility Enum)
   )
   > UpdateBackground() {is=title K2Node_CallFunction}
  ]}
]
variables [
 MaxRowsPerColumn {type=integer, default=5}
 Prize Item Index {type=integer}
]
widgets {
 info {
  inheritedWidgetCount=0
  namedSlotCount=1
  parentClass=&/WBP_TournamentDetailsSection_Base.WBP_TournamentDetailsSection_Base_C
  rootWidgetClass=null
  widgetCount=11
 }
 widgets [
  Content {widgetClassPath=&GridPanel, bIsVariable=true, widget=Content, properties {columnFill [1,1,1], rowFill [0,0,1]}}
  Title {
   widgetClassPath=&GridPanel
   parent=Content
   widget=Title
   properties {columnFill [0,0,0], rowFill [1]}
   slot {class=GridSlot, properties {columnSpan=2, verticalAlignment=VAlign_Top, padding=<ltrb 0 0 0 8>}}
  }
  CommonText_Title {
   widgetClassPath=&CommonTextBlock
   parent=Title
   widget=CommonText_Title
   properties {
    scrollStyle=&TextScrollStyle-Base_C
    text=Advancement
    font {
     bForceMonospaced=false
     bMaterialIsStencil=false
     fontMaterial=null
     fontObject=&NotoSans
     letterSpacing=0
     monospacedWidth=1
     size=24
     skewAmount=0
     typefaceFontName=NotoSans-Regular
     outlineSettings {
      bApplyOutlineToDropShadows=false
      bMiteredCorners=false
      bSeparateFillAlpha=false
      outlineMaterial=null
      outlineSize=0
      outlineColor=<rgba 0 0 0 1>
     }
    }
    shadowOffset=<0 0>
   }
   slot {class=GridSlot, properties {horizontalAlignment=HAlign_Left, verticalAlignment=VAlign_Top}}
  }
  BackgroundDark_Left {
   widgetClassPath=&Image
   bIsVariable=true
   parent=Content
   widget=BackgroundDark_Left
   properties {brush {
     bIsDynamicallyLoaded=false
     drawAs=Image
     imageType=FullColor
     mirroring=NoMirror
     resourceName=None
     resourceObject=/Compete/MaterialInstances/MI_UI_CompeteLeaderboardBackground
     tiling=NoTile
     imageSize=<32 32>
     margin=<ltrb 0 0 0 0>
     outlineSettings {
      bUseBrushTransparency=false
      roundingType=HalfHeightRadius
      width=0
      color {colorUseRule=UseColor_Specified, specifiedColor=<rgba 0 0 0 0>}
      cornerRadii=<0 0 0 0>
     }
     tintColor {colorUseRule=UseColor_Specified, specifiedColor=<rgba 1 1 1 1>}
     uVRegion {bIsValid=false, max=<0 0>, min=<0 0>}
    }}
   slot {class=GridSlot, properties {row=2}}
  }
  Center {
   widgetClassPath=&Image
   bIsVariable=true
   parent=Content
   widget=Center
   properties {brush {
     bIsDynamicallyLoaded=false
     drawAs=Image
     imageType=FullColor
     mirroring=NoMirror
     resourceName=None
     resourceObject=/Compete/MaterialInstances/MI_UI_CompeteLeaderboardTileBackground
     tiling=NoTile
     imageSize=<32 32>
     margin=<ltrb 0 0 0 0>
     outlineSettings {
      bUseBrushTransparency=false
      roundingType=HalfHeightRadius
      width=0
      color {colorUseRule=UseColor_Specified, specifiedColor=<rgba 0 0 0 0>}
      cornerRadii=<0 0 0 0>
     }
     tintColor {colorUseRule=UseColor_Specified, specifiedColor=<rgba 1 1 1 1>}
     uVRegion {bIsValid=false, max=<0 0>, min=<0 0>}
    }}
   slot {class=GridSlot, properties {column=1, layer=1, row=2}}
  }
  BackgroundDark_Right {
   widgetClassPath=&Image
   bIsVariable=true
   parent=Content
   widget=BackgroundDark_Right
   properties {brush {
     bIsDynamicallyLoaded=false
     drawAs=Image
     imageType=FullColor
     mirroring=NoMirror
     resourceName=None
     resourceObject=&MI_UI_CompeteLeaderboardBackground
     tiling=NoTile
     imageSize=<32 32>
     margin=<ltrb 0 0 0 0>
     outlineSettings {
      bUseBrushTransparency=false
      roundingType=HalfHeightRadius
      width=0
      color {colorUseRule=UseColor_Specified, specifiedColor=<rgba 0 0 0 0>}
      cornerRadii=<0 0 0 0>
     }
     tintColor {colorUseRule=UseColor_Specified, specifiedColor=<rgba 1 1 1 1>}
     uVRegion {bIsValid=false, max=<0 0>, min=<0 0>}
    }}
   slot {class=GridSlot, properties {column=2, row=2}}
  }
  TableHeader {
   widgetClassPath=&GridPanel
   bIsVariable=true
   parent=Content
   widget=TableHeader
   properties {columnFill [1,1], rowFill [1]}
   slot {class=GridSlot, properties {row=1, padding=<ltrb 0 11 0 11>}}
  }
  CommonTextBlock_272 {
   widgetClassPath=&CommonTextBlock
   parent=TableHeader
   widget=CommonTextBlock_272
   properties {
    justification=Center
    mobileFontSizeMultiplier=1.5
    scrollStyle=&TextScrollStyle-Base_C
    text=Place
    colorAndOpacity {colorUseRule=UseColor_Specified, specifiedColor=<rgba 0.5776 0.5776 0.5776 1>}
    font {
     bForceMonospaced=false
     bMaterialIsStencil=false
     fontMaterial=null
     fontObject=&HeadingNow
     letterSpacing=0
     monospacedWidth=1
     size=12
     skewAmount=0
     typefaceFontName=74 Regular
     outlineSettings {
      bApplyOutlineToDropShadows=false
      bMiteredCorners=false
      bSeparateFillAlpha=false
      outlineMaterial=null
      outlineSize=0
      outlineColor=<rgba 0 0 0 1>
     }
    }
    shadowOffset=<0 0>
   }
   slot {class=GridSlot, properties {verticalAlignment=VAlign_Top}}
  }
  CommonTextBlock {
   widgetClassPath=&CommonTextBlock
   parent=TableHeader
   widget=CommonTextBlock
   properties {
    justification=Center
    minDesiredWidth=110
    mobileFontSizeMultiplier=1.5
    scrollStyle=&TextScrollStyle-Base_C
    text=Division
    colorAndOpacity {colorUseRule=UseColor_Specified, specifiedColor=<rgba 0.5776 0.5776 0.5776 1>}
    font {
     bForceMonospaced=false
     bMaterialIsStencil=false
     fontMaterial=null
     fontObject=&HeadingNow
     letterSpacing=0
     monospacedWidth=1
     size=12
     skewAmount=0
     typefaceFontName=74 Regular
     outlineSettings {
      bApplyOutlineToDropShadows=false
      bMiteredCorners=false
      bSeparateFillAlpha=false
      outlineMaterial=null
      outlineSize=0
      outlineColor=<rgba 0 0 0 1>
     }
    }
    shadowOffset=<0 0>
   }
   slot {class=GridSlot, properties {column=1, verticalAlignment=VAlign_Bottom}}
  }
  TableSizeBox {
   widgetClassPath=&SizeBox
   bIsVariable=true
   parent=Content
   widget=TableSizeBox
   properties {bOverride_MinDesiredWidth=true, minDesiredWidth=1100}
   slot {class=GridSlot, properties {columnSpan=3, layer=1, row=2}}
  }
  Table {
   widgetClassPath=/Script/UMG.UniformGridPanel
   bIsVariable=true
   parent=TableSizeBox
   widget=Table
   properties {minDesiredSlotHeight=65, minDesiredSlotWidth=227}
   slot {class=SizeBoxSlot}
  }
 ]
}
