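# Details section widget for the tournament details screen, derived from WBP_TournamentDetailsSection_Base, showing a "Details" title above an event description.
# Its Content grid stacks two CommonTextBlocks, with the auto-wrapping description prefilled with sample test-event copy.
# SetDescription replaces that copy with the passed text, and the event graph is empty.
# Anim_Intro animates the left padding of SectionSlot's grid slot, but nothing here plays it.
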WidgetBlueprint /Compete/TournamentDetails/Sections/WBP_TournamentDetailsSection_Details
animations [Anim_Intro {
  binding_count=2
  length_seconds=1
  object=&self:Anim_Intro
  sequence {bindings [
    GridSlot_2 {
     id=#1
     parent_id=#2
     possessed_class=GridSlot
     tracks [{
       type=MovieSceneMarginTrack
       display_name=Padding
       sections [{
         type=MovieSceneMarginSection
         end_frame=20
         start_frame=0
         channels [
          Left {type=MovieSceneScriptingFloatChannel, keys [[0,0],[20,10]]}
          Top {type=MovieSceneScriptingFloatChannel, const [0,2]}
          Right {type=MovieSceneScriptingFloatChannel, const [0,2]}
          Bottom {type=MovieSceneScriptingFloatChannel, const [0,2]}
         ]
        }]
      }]
    }
    SectionSlot {id=#2, possessed_class=NamedSlot, tracks []}
   ]}
 }]
graphs [
 EventGraph {links [], nodes []}
 SetDescription {nodes [
   n1 = SetDescription!() -> (then, InText:Text) {is=title K2Node_FunctionEntry}
   > "SetText (Text)"(self=Get CommonText_Description() -> (CommonText_Description:Common Text Object Reference), InText=@n1)
  ]}
]
variables []
widgets {
 info {
  inheritedWidgetCount=0
  namedSlotCount=1
  parentClass=&/WBP_TournamentDetailsSection_Base.WBP_TournamentDetailsSection_Base_C
  rootWidgetClass=null
  widgetCount=3
 }
 widgets [
  Content {widgetClassPath=&GridPanel, widget=Content, properties {columnFill [1], rowFill [0,1]}}
  CommonText_Title {
   widgetClassPath=&CommonTextBlock
   parent=Content
   widget=CommonText_Title
   properties {
    scrollStyle=&TextScrollStyle-Base_C
    text=Details
    font {
     bForceMonospaced=false
     bMaterialIsStencil=false
     fontMaterial=null
     fontObject=&NotoSans
     letterSpacing=0
     monospacedWidth=1
     size=24
     skewAmount=0
     typefaceFontName=NotoSans-Regular
     outlineSettings {
      bApplyOutlineToDropShadows=false
      bMiteredCorners=false
      bSeparateFillAlpha=false
      outlineMaterial=null
      outlineSize=0
      outlineColor=<rgba 0 0 0 1>
     }
    }
    shadowOffset=<0 0>
   }
   slot {class=GridSlot, properties {horizontalAlignment=HAlign_Left, verticalAlignment=VAlign_Top, padding=<ltrb 0 0 6 10>}}
  }
  CommonText_Description {
   widgetClassPath=&CommonTextBlock
   bIsVariable=true
   parent=Content
   widget=CommonText_Description
   properties {
    autoWrapText=true
    text="This event is testing our tournament system and game performance at the highest level of competition for the latest update. As we're gathering performance data actively during this event, you may experience moments of low performance. This event will occur over two rounds, with the top 50 teams advancing to Round 2. Cash prizes will be awarded to top players at the conclusion of Round 2's test event."
    font {
     bForceMonospaced=false
     bMaterialIsStencil=false
     fontMaterial=null
     fontObject=&NotoSans
     letterSpacing=0
     monospacedWidth=1
     size=24
     skewAmount=0
     typefaceFontName=NotoSans-Regular
     outlineSettings {
      bApplyOutlineToDropShadows=false
      bMiteredCorners=false
      bSeparateFillAlpha=false
      outlineMaterial=null
      outlineSize=0
      outlineColor=<rgba 0 0 0 1>
     }
    }
    shadowOffset=<0 0>
   }
   slot {class=GridSlot, properties {horizontalAlignment=HAlign_Left, row=1, verticalAlignment=VAlign_Top, padding=<ltrb 0 0 6 6>}}
  }
 ]
}
