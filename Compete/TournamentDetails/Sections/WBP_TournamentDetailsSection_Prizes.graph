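# Tournament details section widget listing leaderboard payouts as a titled "Cash Prizes" table with a USD note.
# SetLeaderboardsPrizesData counts, then adds displayable payouts, skipping cosmetic, token and invalid items and thresholds matching a cash header prize.
# The first row's measured text or a four-item count selects two columns instead of three, except on mobile.
# Rows fill column-first over background images rounded through dynamic materials; PreConstruct builds a design-time preview.
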WidgetBlueprint /Compete/TournamentDetails/Sections/WBP_TournamentDetailsSection_Prizes
graphs [
 Add Item {nodes [
   Add Item!() -> (then, Index:int) {is=title K2Node_FunctionEntry}
   > n1 = Create Widget!(
    execute:Exec
    Class:User Widget Class Reference=&/WBP_TournamentDetailsSectionItem_Scoring.WBP_TournamentDetailsSectionItem_Scoring_C
    OwningPlayer:Player Controller Object Reference
    VM_TournamentDetailsSectionItem_PrizeDebug:VM Tournament Details Section Item Prize Debug Object Reference
   ) -> (then:Exec, ReturnValue:WBP Tournament Details Section Item Scoring Object Reference)
   n2 = Get Index() -> (Index:int=0)
   n3 = Get MaxRowsPerColumn() -> (MaxRowsPerColumn:int=0)
   > n4 = AddChildToUniformGrid!(
    execute
    self:Uniform Grid Panel Object Reference=Get Table() -> (Table:Uniform Grid Panel Object Reference)
    Content:Widget Object Reference=@n1
    InRow:int="% (Integer)"(A=@n2, B=@n3) -> (ReturnValue=0)
    InColumn:int=int / int(A=@n2, B=@n3)
   ) -> (then, ReturnValue:Uniform Grid Slot Object Reference)
   > SetHorizontalAlignment(self:Uniform Grid Slot Object Reference=@n4, InHorizontalAlignment=HAlign_Fill)
   > SetVerticalAlignment(self:Uniform Grid Slot Object Reference=@n4, InVerticalAlignment=VAlign_Center)
   > Return Node!(execute, Item Prize Widget:WBP Tournament Details Section Item Scoring Object Reference=@n1)
  ]}
 CheckIfOnlyFirstPrizeCashPrize {nodes [
   n1 = CheckIfOnlyFirstPrizeCashPrize!() -> (then, CurrentPayout:Tournament Payout Threshold Data Structure) {is=title K2Node_FunctionEntry}
   n2 = Get Tournament View VM() -> (Tournament View VM:Tournament View VM Object Reference)
   n3 = Break Tournament Payout Threshold Data!(TournamentPayoutThresholdData:"Tournament Payout Threshold Data Structure (by ref)"=Get HeaderPayoutThresholdData!(self:Tournament View VM Object Reference=@n2) -> (HeaderPayoutThresholdData:Tournament Payout Threshold Data Structure)) -> (
    ScoringType:EPayoutScoringType Enum
    Threshold:double=~
    ThresholdDisplayText:Text
    ThresholdDisplayTextHeader:Text
    SourceLeaderboardDisplayName:Text
    PayoutItemsData:Array of Tournament Payout Item Data Structures
    CosmeticPayoutItemsVM:Array of Fort Item VM Object References
   ) {is=title K2Node_BreakStruct}
   n4 = Break Tournament Payout Threshold Data(TournamentPayoutThresholdData=@n1) -> (Threshold=~)
   n5 = Break Tournament Payout Item Data!(TournamentPayoutItemData:"Tournament Payout Item Data Structure (by ref)"=Get HeaderPayoutItemData!(self:Tournament View VM Object Reference=@n2) -> (HeaderPayoutItemData:Tournament Payout Item Data Structure)) -> (RewardType:EPayoutRewardType Enum, DisplayText:Text) {is=title K2Node_BreakStruct}
   > Return Node!(
    execute
    IsFirstPrizeWithOnlyCash:bool=AND Boolean!(
     A="Equal (Enum)"(A:EPayoutScoringType Enum=@n4, B:EPayoutScoringType Enum=@n3)
     B="Equal (Float)"(A=@n4.Threshold, B=@n3.Threshold)
     C:bool="Equal (Enum)"(A:EPayoutRewardType Enum=@n5, B:EPayoutRewardType Enum=Commerce)
    ) -> (ReturnValue)
   )
  ]}
 EventGraph {nodes [
   n1 = Event PreConstruct()
   > Parent: PreConstruct(IsDesignTime=@n1.IsDesignTime)
   > Branch(Condition=@n1.IsDesignTime)
   > UpdateBackground()
   > SetTitle!(execute, self, InText:Text=Get CashPrizeText() -> (CashPrizeText:Text)) -> (then)
   n2 = MakeLiteralInt!(Value:int=7) -> (ReturnValue:int=0)
   n3 = "To Float (Integer)"(InInt=@n2)
   > Set MaxRowsPerColumn(MaxRowsPerColumn:int=Ceil(A=float / float(A=@n3, B=3)) {is=title K2Node_CallFunction}) -> (Output_Get:int=0)
   > n4 = For Loop(LastIndex=int - int(A=@n2))
   Set Table Grid Slot!(execute=@n4.Completed, self, In Num Items:int=@n2) -> (then) {is=title K2Node_CallFunction}
   Add Item!(execute=@n4.LoopBody, self, Index:int=@n4) -> (then, Item Prize Widget:WBP Tournament Details Section Item Scoring Object Reference) {is=title K2Node_CallFunction}
  ]}
 Set Table Grid Slot {nodes [
   n1 = Set Table Grid Slot!() -> (then, In Num Items:int) {is=title K2Node_FunctionEntry}
   n2 = Get MaxRowsPerColumn() -> (MaxRowsPerColumn=0)
   > SetGridSlotColumnAndRow!(
    execute
    Widget:Widget Object Reference=Get TableSizeBox() -> (TableSizeBox:Size Box Object Reference)
    Row:int=1
    Row Span:int=0
    Column:int=0
    Column Span:int=Ceil(A=float / float(A:int=@n1, B:int=@n2))
   ) -> (then)
  ]}
 SetCornerRoundness {nodes [
   n1 = SetCornerRoundness!() -> (then, Widget:Image Object Reference, Value:Linear Color Structure) {is=title K2Node_FunctionEntry}
   > n2 = GetDynamicMaterial(self=@n1)
   > SetVectorParameterValue(self=@n2, ParameterName=Roundness, Value=@n1.Value)
  ]}
 SetLeaderboardsPrizesData {nodes [
   SetLeaderboardsPrizesData!() -> (then, In Prizes Data:Array of Tournament Leaderboard Payout Table Data Structures) {is=title K2Node_FunctionEntry}
   > ClearChildren(self=Get Table())
   > Set Filtered Prizes Num(Filtered Prizes Num:int=0) -> (Output_Get:int=0)
   > Set Prize Index(Prize Index:int=0) -> (Output_Get:int=0)
   > n1 = Sequence()
   > n2 = For Each Loop(Array:Array of Tournament Leaderboard Payout Table Data Structures=Get In Prizes Data() -> (In Prizes Data:Array of Tournament Leaderboard Payout Table Data Structures)) -> (Array Element:Tournament Leaderboard Payout Table Data Structure)
   > n3 = For Each Loop(Array:Array of Tournament Payout Threshold Data Structures=Break Tournament Leaderboard Payout Table Data!(TournamentLeaderboardPayoutTableData:"Tournament Leaderboard Payout Table Data Structure (by ref)"=@n2) -> (bIsMainWindowLeaderboard:bool=~, PayoutThresholdsData:Array of Tournament Payout Threshold Data Structures) {is=title K2Node_BreakStruct}.PayoutThresholdsData) -> (Array Element:Tournament Payout Threshold Data Structure)
   > n4 = CheckIfOnlyFirstPrizeCashPrize!(execute, self, CurrentPayout:Tournament Payout Threshold Data Structure=@n3) -> (then, IsFirstPrizeWithOnlyCash:bool) {is=title K2Node_CallFunction}
   > n5 = Branch(Condition=@n4)
   > Set HasSkippedFirstPrize(HasSkippedFirstPrize:bool=true) -> (Output_Get:bool)
   n6 = For Each Loop(
    Exec=@n5.else
    Array:Array of Tournament Payout Item Data Structures=Break Tournament Payout Threshold Data(TournamentPayoutThresholdData=@n3) -> (Threshold=~).PayoutItemsData
   ) -> (Array Element:Tournament Payout Item Data Structure)
   > n7 = Should Display Payout Item!(execute, self, Payout Item Data:"Tournament Payout Item Data Structure (by ref)"=@n6) -> (then, b Display Item:bool)
   > Branch(Condition=@n7)
   > Increment Int(Value=Get Filtered Prizes Num() -> (Filtered Prizes Num:int=0))
   n8 = "integer >= integer"(A=Get Filtered Prizes Num() -> (Filtered Prizes Num=0), B=1)
   SetVisibility(execute=@n1.then_1, InVisibility=Select(Option 0=Collapsed, Option 1=SelfHitTestInvisible, Index=@n8))
   > Branch(Condition=@n8)
   > n9 = For Each Loop(Array:Array of Tournament Leaderboard Payout Table Data Structures=Get In Prizes Data()) -> (Array Element:Tournament Leaderboard Payout Table Data Structure)
   > n10 = For Each Loop(Array:Array of Tournament Payout Threshold Data Structures=Break Tournament Leaderboard Payout Table Data(TournamentLeaderboardPayoutTableData=@n9) -> (bIsMainWindowLeaderboard=~).PayoutThresholdsData) -> (Array Element:Tournament Payout Threshold Data Structure)
   > n11 = CheckIfOnlyFirstPrizeCashPrize(CurrentPayout=@n10)
   > n12 = Branch(Condition=@n11)
   n13 = Break Tournament Payout Threshold Data(TournamentPayoutThresholdData=@n10) -> (Threshold=~)
   n14 = For Each Loop(Exec=@n12.else, Array:Array of Tournament Payout Item Data Structures=@n13.PayoutItemsData) -> (Array Element:Tournament Payout Item Data Structure)
   > n15 = Should Display Payout Item(Payout Item Data=@n14)
   > Branch(Condition=@n15)
   > n16 = Add Item(Index=Get Prize Index() -> (Prize Index:int=0))
   > Set Prize Data!(
    execute
    self:WBP Tournament Details Section Item Scoring Object Reference=@n16
    In Threshold Text:Text=@n13.ThresholdDisplayText
    In Value Text:Text=Break Tournament Payout Item Data(TournamentPayoutItemData=@n14).DisplayText
   ) -> (then) {is=title K2Node_CallFunction}
   > n17 = Increment Int(Value=Get Prize Index() -> (Prize Index=0))
   > Branch(Condition="Equal (Integer)"(A=@n17, B=1))
   > UpdateTableColums!(
    execute
    self
    TableFirstElement:WBP Tournament Details Section Item Scoring Object Reference=@n16
    In Num Items:int=Get Filtered Prizes Num() -> (Filtered Prizes Num=0)
   ) -> (then)
  ]}
 SetTitle {nodes [
   n1 = SetTitle!() -> (then, InText:Text) {is=title K2Node_FunctionEntry}
   > n2 = Is Valid(InputObject=Get TitleText() -> (TitleText:Common Text Object Reference))
   n3 = Branch(execute=@n2.Is Not Valid, Condition=TextIsEmpty(InText=@n1))
   n4 = Construct Object from Class!(
    execute=@n3.else
    Class=&CommonTextBlock
    self
    bIsScrollingEnabled:bool=True
    Style:Common Text Style Class Reference
    ScrollStyle:Common Text Scroll Style Class Reference=&TextScrollStyle-Base_C
    ScrollOrientation:EOrientation Enum=Orient_Horizontal
   ) -> (then, ReturnValue:Common Text Object Reference)
   > n5 = Set TitleText(TitleText:Common Text Object Reference=@n4) -> (Output_Get:Common Text Object Reference)
   > "SetText (Text)"(self=@n5, InText=Get InText() -> (InText:Text))
   > AddChild(self=Get Content() -> (Content:Grid Panel Object Reference), Content=@n4)
   > SetGridSlotAlignment!(execute, Widget:Widget Object Reference=@n4, Horizontal:EHorizontalAlignment Enum=HAlign_Left, Vertical:EVerticalAlignment Enum=VAlign_Top) -> (then)
   > SetGridSlotPadding!(execute, Widget:Widget Object Reference=@n4, Padding_Left:float=0, Padding_Top:float=0, Padding_Right:float=0, Padding_Bottom:float=14) -> (then)
   > SetGridSlotColumnAndRow(Widget=@n4, Row=0, Row Span=0, Column=0, Column Span=3)
   > UpdateWidgetStyleTags!(
    execute
    InWidget:Widget Object Reference=@n4
    InAddedTags:Array of Gameplay Tag Structures=Make Array!("[0]"=(TagName="StyleSheet.Style.10"), "[1]":Gameplay Tag Structure=(TagName="StyleSheet.Style.Heading")) -> (Array)
    InRemovedTags:Array of Gameplay Tag Structures=Make Array!("[0]") -> (Array)
   ) -> (then, ReturnValue:bool)
   n6 = Get TitleText()
   "SetText (Text)"(execute=@n2.Is Valid, self=@n6, InText=Get InText())
   > SetGridSlotColumnAndRow(Widget=@n6, Row=0, Row Span=0, Column=0, Column Span=Get ColumsAmount() -> (ColumsAmount:int=0))
   > SetVisibility(
    self=Get TitleText()
    InVisibility=Select(Option 0=SelfHitTestInvisible, Option 1=Collapsed, Index=Get HasSkippedFirstPrize() -> (HasSkippedFirstPrize:bool))
   )
   > SetColumnSpan!(
    execute
    self:Grid Slot Object Reference=SlotAsGridSlot(Widget=Get CurrencyWarning() -> (CurrencyWarning:Common Text Object Reference))
    InColumnSpan:int=Get ColumsAmount() -> (ColumsAmount=0)
   ) -> (then)
  ]}
 Should Display Payout Item {nodes [
   n1 = Should Display Payout Item!() -> (then, Payout Item Data:"Tournament Payout Item Data Structure (by ref)") {is=title K2Node_FunctionEntry}
   > n2 = IsPayoutItemCosmetic!(execute, ItemData:"Tournament Payout Item Data Structure (by ref)"=@n1) -> (then, ReturnValue:bool)
   > n3 = Branch(Condition=@n2)
   > Return Node!(execute, b Display Item:bool=false)
   Return Node(
    execute=@n3.else
    b Display Item=Select!(
     Invalid:bool=false
     Commerce:bool=true
     GameRelated:bool=true
     Token:bool=false
     Score:bool=true
     Index:EPayoutRewardType Enum=Break Tournament Payout Item Data(TournamentPayoutItemData=@n1)
    ) -> (ReturnValue:bool)
   )
  ]}
 UpdateBackground {nodes [
   UpdateBackground() {is=title K2Node_FunctionEntry}
   > SetCornerRoundness!(
    execute
    self
    Widget:Image Object Reference=Get BackgroundDark_Left() -> (BackgroundDark_Left:Image Object Reference)
    Value_R:float=25
    Value_G:float=25
    Value_B:float=0
    Value_A:float=0
   ) -> (then) {is=title K2Node_CallFunction}
   > SetCornerRoundness(Widget=Get BackgroundDark_Right() -> (BackgroundDark_Right:Image Object Reference), Value_R=0, Value_G=0, Value_B=25, Value_A=25)
   n1 = Select!(Option 0:float=0, Option 1:float=25, Index:bool="Equal (Integer)"(A=Get ColumsAmount() -> (ColumsAmount=0), B=2)) -> (ReturnValue:float)
   > SetCornerRoundness(Widget=Get Center() -> (Center:Image Object Reference), Value_R=0, Value_G=0, Value_B=@n1, Value_A=@n1)
  ]}
 UpdateTableColums {nodes [
   n1 = UpdateTableColums!() -> (then, TableFirstElement:WBP Tournament Details Section Item Scoring Object Reference, In Num Items:int) {is=title K2Node_FunctionEntry}
   > ForceLayoutPrepass(self:Widget Object Reference=@n1)
   n2 = IsMobileGame()
   n3 = Get TableFirstElement() -> (TableFirstElement:WBP Tournament Details Section Item Scoring Object Reference)
   n4 = "float > float"(
    A=GetDesiredSize!(self:Widget Object Reference=Get CommonText_Place!(self:WBP Tournament Details Section Item Scoring Object Reference=@n3) -> (CommonText_Place:Common Text Object Reference)) -> (ReturnValue_X:double, ReturnValue_Y:double)
    B=150
   )
   n5 = "float > float"(
    A=GetDesiredSize(self=Get CommonText_Points!(self:WBP Tournament Details Section Item Scoring Object Reference=@n3) -> (CommonText_Points:Common Text Object Reference))
    B=150
   )
   n6 = "Equal (Integer)"(A=Get In Num Items() -> (In Num Items:int=0), B=4)
   > Set IsContentSizeLargerThenCell(IsContentSizeLargerThenCell:bool=AND Boolean!(A=OR Boolean!(A=@n4, B=@n5, C:bool=@n6) -> (ReturnValue), B=NOT Boolean(A=@n2)) -> (ReturnValue)) -> (Output_Get:bool)
   > Set ColumsAmount(ColumsAmount:int=Select(Option 0:int=3, Option 1:int=2, Index=Get IsContentSizeLargerThenCell() -> (IsContentSizeLargerThenCell:bool)) -> (ReturnValue:int)) -> (Output_Get:int=0)
   n7 = Get In Num Items() -> (In Num Items=0)
   n8 = "To Float (Integer)"(InInt=@n7)
   n9 = "To Float (Integer)"(InInt=Get ColumsAmount() -> (ColumsAmount=0))
   > Set MaxRowsPerColumn(MaxRowsPerColumn=Ceil(A=float / float(A:double=@n8, B:double=@n9))) -> (Output_Get=0)
   > Set Table Grid Slot(In Num Items=@n7) {is=title K2Node_CallFunction}
   > SetTitle(InText=Get CashPrizeText()) {is=title K2Node_CallFunction}
   > SetColumnFill!(
    execute
    self:Grid Panel Object Reference=Get Content()
    ColumnIndex:int=2
    Coefficient:float=Select(Option 0:float=1, Option 1:float=0, Index=Get IsContentSizeLargerThenCell()) -> (ReturnValue:float)
   ) -> (then)
   > SetVisibility(
    self=Get BackgroundDark_Right()
    InVisibility=Select(Option 0:ESlateVisibility Enum=SelfHitTestInvisible, Option 1:ESlateVisibility Enum=Collapsed, Index=Get IsContentSizeLargerThenCell()) -> (ReturnValue:ESlateVisibility Enum)
   )
   > UpdateBackground() {is=title K2Node_CallFunction}
  ]}
]
variables [
 CashPrizeText {type=string, default=Cash Prizes}
 ColumsAmount {type=integer, default=3}
 HasSkippedFirstPrize {type=boolean}
 MaxRowsPerColumn {type=integer, default=10}
 TitleText {type=CommonTextBlock, default=null}
]
widgets {
 info {
  inheritedWidgetCount=0
  namedSlotCount=1
  parentClass=&/WBP_TournamentDetailsSection_Base.WBP_TournamentDetailsSection_Base_C
  rootWidgetClass=null
  widgetCount=7
 }
 widgets [
  Content {widgetClassPath=&GridPanel, bIsVariable=true, widget=Content, properties {columnFill [1,1,1], rowFill [0,1]}}
  BackgroundDark_Left {
   widgetClassPath=&Image
   bIsVariable=true
   parent=Content
   widget=BackgroundDark_Left
   properties {brush {
     bIsDynamicallyLoaded=false
     drawAs=Image
     imageType=FullColor
     mirroring=NoMirror
     resourceName=None
     resourceObject=/Compete/MaterialInstances/MI_UI_CompeteLeaderboardBackground
     tiling=NoTile
     imageSize=<32 32>
     margin=<ltrb 0 0 0 0>
     outlineSettings {
      bUseBrushTransparency=false
      roundingType=HalfHeightRadius
      width=0
      color {colorUseRule=UseColor_Specified, specifiedColor=<rgba 0 0 0 0>}
      cornerRadii=<0 0 0 0>
     }
     tintColor {colorUseRule=UseColor_Specified, specifiedColor=<rgba 1 1 1 1>}
     uVRegion {bIsValid=false, max=<0 0>, min=<0 0>}
    }}
   slot {class=GridSlot, properties {row=1}}
  }
  Center {
   widgetClassPath=&Image
   bIsVariable=true
   parent=Content
   widget=Center
   properties {brush {
     bIsDynamicallyLoaded=false
     drawAs=Image
     imageType=FullColor
     mirroring=NoMirror
     resourceName=None
     resourceObject=/Compete/MaterialInstances/MI_UI_CompeteLeaderboardTileBackground
     tiling=NoTile
     imageSize=<32 32>
     margin=<ltrb 0 0 0 0>
     outlineSettings {
      bUseBrushTransparency=false
      roundingType=HalfHeightRadius
      width=0
      color {colorUseRule=UseColor_Specified, specifiedColor=<rgba 0 0 0 0>}
      cornerRadii=<0 0 0 0>
     }
     tintColor {colorUseRule=UseColor_Specified, specifiedColor=<rgba 1 1 1 1>}
     uVRegion {bIsValid=false, max=<0 0>, min=<0 0>}
    }}
   slot {class=GridSlot, properties {column=1, layer=1, row=1}}
  }
  BackgroundDark_Right {
   widgetClassPath=&Image
   bIsVariable=true
   parent=Content
   widget=BackgroundDark_Right
   properties {brush {
     bIsDynamicallyLoaded=false
     drawAs=Image
     imageType=FullColor
     mirroring=NoMirror
     resourceName=None
     resourceObject=&MI_UI_CompeteLeaderboardBackground
     tiling=NoTile
     imageSize=<32 32>
     margin=<ltrb 0 0 0 0>
     outlineSettings {
      bUseBrushTransparency=false
      roundingType=HalfHeightRadius
      width=0
      color {colorUseRule=UseColor_Specified, specifiedColor=<rgba 0 0 0 0>}
      cornerRadii=<0 0 0 0>
     }
     tintColor {colorUseRule=UseColor_Specified, specifiedColor=<rgba 1 1 1 1>}
     uVRegion {bIsValid=false, max=<0 0>, min=<0 0>}
    }}
   slot {class=GridSlot, properties {column=2, row=1}}
  }
  TableSizeBox {
   widgetClassPath=&SizeBox
   bIsVariable=true
   parent=Content
   widget=TableSizeBox
   properties {minDesiredWidth=1100}
   slot {class=GridSlot, properties {columnSpan=3, layer=1, row=1}}
  }
  Table {
   widgetClassPath=/Script/UMG.UniformGridPanel
   bIsVariable=true
   parent=TableSizeBox
   widget=Table
   properties {minDesiredSlotHeight=64, minDesiredSlotWidth=227}
   slot {class=SizeBoxSlot}
  }
  CurrencyWarning {
   widgetClassPath=&CommonTextBlock
   bIsVariable=true
   parent=Content
   widget=CurrencyWarning
   properties {
    text=All cash prizes are displayed in USD.
    font {
     bForceMonospaced=false
     bMaterialIsStencil=false
     fontMaterial=null
     fontObject=&HeadingNow
     letterSpacing=0
     monospacedWidth=1
     size=11.25
     skewAmount=0
     typefaceFontName=74 Regular
     outlineSettings {
      bApplyOutlineToDropShadows=false
      bMiteredCorners=false
      bSeparateFillAlpha=false
      outlineMaterial=null
      outlineSize=0
      outlineColor=<rgba 0 0 0 1>
     }
    }
    shadowOffset=<0 0>
   }
   slot {class=GridSlot, properties {columnSpan=3, row=2, padding=<ltrb 0 10 0 0>}}
  }
 ]
}
