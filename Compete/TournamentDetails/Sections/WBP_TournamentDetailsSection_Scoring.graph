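# Tournament details "Scoring" section widget that shows placement points in a table and scoring rules as a text line, collapsing when there are no rules.
# SetScoringData counts placement rules, detects Win/Loss layouts, sizes the grid, then spawns item widgets column by column using each rule's array index.
# Backgrounds adapt to single-item tables through dynamic-material corner roundness.
# PreConstruct sets the intro animation and fills sample items at design time.
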WidgetBlueprint /Compete/TournamentDetails/Sections/WBP_TournamentDetailsSection_Scoring
animations [Anim_Intro {
  binding_count=2
  length_seconds=1
  object=&self:Anim_Intro
  sequence {bindings [
    GridSlot_2 {
     id=#1
     parent_id=#2
     possessed_class=GridSlot
     tracks [{
       type=MovieSceneMarginTrack
       display_name=Padding
       sections [{
         type=MovieSceneMarginSection
         end_frame=20
         start_frame=0
         channels [
          Left {type=MovieSceneScriptingFloatChannel, const [10,2]}
          Top {type=MovieSceneScriptingFloatChannel, keys [[0,10],[20,0]]}
          Right {type=MovieSceneScriptingFloatChannel, const [10,2]}
          Bottom {type=MovieSceneScriptingFloatChannel, const [10,2]}
         ]
        }]
      }]
    }
    SectionSlot {id=#2, possessed_class=NamedSlot, tracks []}
   ]}
 }]
graphs [
 AddItem {nodes [
   AddItem!() -> (then, Item Index:int) {is=title K2Node_FunctionEntry}
   > n1 = Create Widget!(
    execute:Exec
    Class:User Widget Class Reference=&/WBP_TournamentDetailsSectionItem_Scoring.WBP_TournamentDetailsSectionItem_Scoring_C
    OwningPlayer:Player Controller Object Reference
    VM_TournamentDetailsSectionItem_PrizeDebug:VM Tournament Details Section Item Prize Debug Object Reference
   ) -> (then:Exec, ReturnValue:WBP Tournament Details Section Item Scoring Object Reference)
   n2 = Get Table Rows Num() -> (Table Rows Num:int=0)
   n3 = Get Item Index() -> (Item Index:int=0)
   > n4 = AddChildToUniformGrid!(
    execute
    self:Uniform Grid Panel Object Reference=Get Table() -> (Table:Uniform Grid Panel Object Reference)
    Content:Widget Object Reference=@n1
    InRow:int="% (Integer)"(A=@n3, B=@n2) -> (ReturnValue=0)
    InColumn:int=int / int(A=@n3, B=@n2)
   ) -> (then, ReturnValue:Uniform Grid Slot Object Reference)
   > SetHorizontalAlignment(self:Uniform Grid Slot Object Reference=@n4, InHorizontalAlignment=HAlign_Fill)
   > SetVerticalAlignment(self:Uniform Grid Slot Object Reference=@n4, InVerticalAlignment=VAlign_Center)
   > Return Node!(execute, Item Scroing Widget:WBP Tournament Details Section Item Scoring Object Reference=@n1)
  ]}
 EventGraph {nodes [
   n1 = Event PreConstruct()
   > Parent: PreConstruct(IsDesignTime=@n1.IsDesignTime)
   > SetFocusAnim!(
    execute
    self:WBP Tournament Details Section Base Object Reference
    Anim_Focus:Widget Animation Object Reference=Get Anim_Intro() -> (Anim_Intro:Widget Animation Object Reference)
   ) -> (then)
   > UpdateBackground()
   > Branch(Condition=@n1.IsDesignTime)
   > ClearChildren(self=Get Table())
   n2 = Get Debug Table Items Num() -> (Debug Table Items Num:int=0)
   > Set Table Grid Slot!(execute, self, Num Items:int=@n2) -> (then) {is=title K2Node_CallFunction}
   > n3 = For Loop(LastIndex=int - int(A=@n2))
   > n4 = AddItem!(execute, self, Item Index:int=@n3) -> (then, Item Scroing Widget:WBP Tournament Details Section Item Scoring Object Reference) {is=title K2Node_CallFunction}
   > Set Placement Item!(
    execute
    self
    Item Widget:WBP Tournament Details Section Item Scoring Object Reference=@n4
    Rank:int=int + int(A=@n3)
    Points:int=int - int(A=100, B=int * int(A=5, B=@n3))
    b Are Placements Win Loss:bool=false
   ) -> (then) {is=title K2Node_CallFunction}
  ]}
 Is Placement Scoring Rule {nodes [
   n1 = Is Placement Scoring Rule!() -> (then, Display Info Id:Name) {is=title K2Node_FunctionEntry}
   n2 = "To String (Name)"(InName=@n1)
   > Return Node!(execute, b Is Placement:bool=OR Boolean(A="Equal (Name)"(A=@n1, B=Placement), B=EndsWith(SourceString=@n2, InSuffix=_PLACEMENT_STAT_INDEX)))
  ]}
 Set Placement Item {nodes [
   n1 = Set Placement Item!() -> (then, Item Widget:WBP Tournament Details Section Item Scoring Object Reference, Rank:int, Points:int, b Are Placements Win Loss:bool=~) {is=title K2Node_FunctionEntry}
   > n2 = Branch(Condition="Equal (Integer)"(A=@n1.Rank, B=1))
   > n3 = AddPlaceIcon(self:WBP Tournament Details Section Item Scoring Object Reference=@n1)
   SetPoints!(execute=[@n3.then, @n4.then], self:WBP Tournament Details Section Item Scoring Object Reference=@n1, Points:int=@n1.Points) -> (then)
   n4 = SetPlace!(
    execute=@n2.else
    self:WBP Tournament Details Section Item Scoring Object Reference=@n1
    InText:Text=SelectText(A="LOCTEXT(\"Loss\")", B=Format Text!(Format="LOCTEXT(\"##{Rank}\")", Rank:int=@n1.Rank) -> (Result), bPickA=@n1.b Are Placements Win Loss)
   ) -> (then)
  ]}
 Set Table Grid Slot {nodes [
   n1 = Set Table Grid Slot!() -> (then, Num Items:int) {is=title K2Node_FunctionEntry}
   n2 = Get Max Table Columns Num() -> (Max Table Columns Num:int=0)
   > Set Table Rows Num(Table Rows Num:int=Ceil(A=float / float(A:int=@n1, B:int=@n2)) {is=title K2Node_CallFunction}) -> (Output_Get:int=0)
   n3 = Get Num Items() -> (Num Items:int=0)
   > SetGridSlotColumnAndRow!(
    execute
    Widget:Widget Object Reference=Get TableSizeBox() -> (TableSizeBox:Size Box Object Reference)
    Row:int=1
    Row Span:int=0
    Column:int=0
    Column Span:int=SelectInt(A=1, B=Get Max Table Columns Num() -> (Max Table Columns Num=0), bPickA="Equal (Integer)"(A=@n3, B=1))
   ) -> (then)
   > SetTableVisibility!(execute, self, ItemsAmount:int=Get Num Items() -> (Num Items=0)) -> (then)
  ]}
 SetCornerRoundness {nodes [
   n1 = SetCornerRoundness!() -> (then, Widget:Image Object Reference, Value:Linear Color Structure) {is=title K2Node_FunctionEntry}
   > n2 = GetDynamicMaterial(self=@n1)
   > SetVectorParameterValue(self=@n2, ParameterName=Roundness, Value=@n1.Value)
  ]}
 SetScoringData {nodes [
   n1 = SetScoringData!() -> (then, ScoringData:Array of Fort Showdown Scoring Rule Info Structures) {is=title K2Node_FunctionEntry}
   n2 = Is Empty!(TargetArray:Array of Fort Showdown Scoring Rule Info Structures=@n1) -> (ReturnValue:bool)
   > SetVisibility(InVisibility=Select(Option 0=SelfHitTestInvisible, Option 1=Collapsed, Index=@n2))
   > n3 = Branch(Condition=@n2)
   n4 = Sequence!(execute=@n3.else) -> (then_0, then_1, then_2:Exec, then_3:Exec)
   > ClearChildren(self=Get Table())
   > Set Placement Items Num(Placement Items Num:int=0) -> (Output_Get:int=0)
   n5 = For Each Loop(
    Exec=@n4.then_1
    Array:Array of Fort Showdown Scoring Rule Info Structures=Get ScoringData() -> (ScoringData:Array of Fort Showdown Scoring Rule Info Structures)
   ) -> (Array Element:Fort Showdown Scoring Rule Info Structure)
   n6 = Break Fort Showdown Scoring Rule Info!(FortShowdownScoringRuleInfo:"Fort Showdown Scoring Rule Info Structure (by ref)"=@n5) -> (DisplayInfoId:Name, PointValue:int, AccumulatedPointValue:int, ScoreRequirement:int, UpperBoundRequirement:int, LowerBoundRequirement:int) {is=title K2Node_BreakStruct}
   > Branch(Condition=Is Placement Scoring Rule!(Display Info Id:Name=@n6) -> (b Is Placement:bool) {is=title K2Node_CallFunction})
   > Increment Int(Value=Get Placement Items Num() -> (Placement Items Num:int=0))
   n7 = int + int(A=@n5.Array Index)
   > Branch(Condition=OR Boolean(A="Not Equal (Integer)"(A=@n7, B=@n6.ScoreRequirement), B="integer >= integer"(A=@n6.ScoreRequirement, B=3)))
   > Set b Are Placement Rules Win or Loss(b Are Placement Rules Win or Loss:bool=false) -> (Output_Get:bool)
   n8 = "integer >= integer"(A=Get Placement Items Num() -> (Placement Items Num=0), B=1)
   SetVisibility(
    execute=@n4.then_2
    self=Get ScoringContent() -> (ScoringContent:Grid Panel Object Reference)
    InVisibility=Select(Option 0=Collapsed, Option 1=SelfHitTestInvisible, Index=@n8)
   )
   > Branch(Condition=@n8)
   > Set Table Grid Slot(Num Items=Get Placement Items Num() -> (Placement Items Num=0)) {is=title K2Node_CallFunction}
   > n9 = For Each Loop(Array=Get ScoringData())
   n10 = Break Fort Showdown Scoring Rule Info(FortShowdownScoringRuleInfo=@n9)
   > Branch(Condition=Is Placement Scoring Rule(Display Info Id=@n10))
   > Increment Int(Value=Get Placement Items Num() -> (Placement Items Num=0))
   > n11 = AddItem(Item Index=@n9.Array Index)
   > Set Placement Item(
    Item Widget=@n11
    Rank=@n10.ScoreRequirement
    Points=@n10.AccumulatedPointValue
    b Are Placements Win Loss=Get b Are Placement Rules Win or Loss() -> (b Are Placement Rules Win or Loss:bool)
   ) {is=title K2Node_CallFunction}
   n12 = GetTournamentDetailsScoringRulesDisplayText!(
    self:Fort Compete UIManager Object Reference=Get FortCompeteUIManager() -> (ReturnValue:Fort Compete UIManager Object Reference) {is=title K2Node_GetSubsystem}
    InScoringRules:Array of Fort Showdown Scoring Rule Info Structures=Get ScoringData()
    RuleTextPointFormat:Text="LOCTEXT(\"{0} +{1}\")"
    RulesSeparator:Text="INVTEXT(\"    \")"
    RequirementsSeparator:Text="INVTEXT(\", \")"
    PrefixText:Text="INVTEXT(\"*\")"
   ) -> (ReturnValue:Text)
   Set Text(execute=@n4.then_3, Text=@n12, self=Get CommonText_ScoringRules() -> (CommonText_ScoringRules:Common Text Object Reference))
   > SetVisibility(self=Get CommonText_ScoringRules(), InVisibility=Select(Option 0=SelfHitTestInvisible, Option 1=Collapsed, Index=TextIsEmpty(InText=@n12)))
   Reroute Node(InputPin:Wildcard) -> (OutputPin:Wildcard)
  ]}
 SetTableVisibility {nodes [
   n1 = SetTableVisibility!() -> (then, ItemsAmount:int) {is=title K2Node_FunctionEntry}
   n2 = "Equal (Integer)"(A=@n1, B=1)
   > SetVisibility(
    self=[
     Get Center_1() -> (Center_1:Image Object Reference)
     Get BackgroundDark_Right() -> (BackgroundDark_Right:Image Object Reference)
     Get BackgroundDark_Center() -> (BackgroundDark_Center:Image Object Reference)
     Get Center_2() -> (Center_2:Image Object Reference)
    ]
    InVisibility=Select(Option 0=SelfHitTestInvisible, Option 1=Collapsed, Index=@n2)
   )
   > SetCornerRoundness!(
    execute
    self
    Widget:Image Object Reference=Get BackgroundDark_Left() -> (BackgroundDark_Left:Image Object Reference)
    Value:Linear Color Structure=Select!(
     Option 0_R:float=25
     Option 0_G:float=25
     Option 0_B:float=0
     Option 0_A:float=0
     Option 1_R:float=25
     Option 1_G:float=25
     Option 1_B:float=25
     Option 1_A:float=25
     Index=@n2
    ) -> (ReturnValue:Linear Color Structure)
   ) -> (then) {is=title K2Node_CallFunction}
  ]}
 UpdateBackground {nodes [
   UpdateBackground() {is=title K2Node_FunctionEntry}
   > SetCornerRoundness!(execute, self, Widget=Get BackgroundDark_Left(), Value_R:float=25, Value_G:float=25, Value_B:float=0, Value_A:float=0) -> (then)
   > SetCornerRoundness(Widget=Get BackgroundDark_Right(), Value_R=0, Value_G=0, Value_B=25, Value_A=25)
   > SetCornerRoundness!(execute, self, Widget=Get Center_2(), Value:Linear Color Structure) -> (then)
   > SetCornerRoundness(Widget=Get Center_1())
   > SetCornerRoundness(Widget=Get BackgroundDark_Center())
  ]}
]
variables [
 Debug Table Items Num {type=integer, default=13}
 Max Table Columns Num {type=integer, default=5}
 Table Rows Num {type=integer, default=5}
]
widgets {
 info {
  inheritedWidgetCount=0
  namedSlotCount=1
  parentClass=&/WBP_TournamentDetailsSection_Base.WBP_TournamentDetailsSection_Base_C
  rootWidgetClass=null
  widgetCount=15
 }
 widgets [
  GridPanel_0 {widgetClassPath=&GridPanel, widget=GridPanel_0, properties {bOverride_Cursor=true, columnFill [1], rowFill [0,1]}}
  Title {
   widgetClassPath=&GridPanel
   parent=GridPanel_0
   widget=Title
   properties {columnFill [0,1], rowFill [1]}
   slot {class=GridSlot, properties {verticalAlignment=VAlign_Top}}
  }
  CommonText_Title {
   widgetClassPath=&CommonTextBlock
   parent=Title
   widget=CommonText_Title
   properties {
    scrollStyle=&TextScrollStyle-Base_C
    text=Scoring
    font {
     bForceMonospaced=false
     bMaterialIsStencil=false
     fontMaterial=null
     fontObject=&NotoSans
     letterSpacing=0
     monospacedWidth=1
     size=24
     skewAmount=0
     typefaceFontName=NotoSans-Regular
     outlineSettings {
      bApplyOutlineToDropShadows=false
      bMiteredCorners=false
      bSeparateFillAlpha=false
      outlineMaterial=null
      outlineSize=0
      outlineColor=<rgba 0 0 0 1>
     }
    }
    shadowOffset=<0 0>
   }
   slot {class=GridSlot, properties {horizontalAlignment=HAlign_Left, verticalAlignment=VAlign_Top}}
  }
  CommonText_ScoringRules {
   widgetClassPath=&CommonTextBlock
   bIsVariable=true
   parent=Title
   widget=CommonText_ScoringRules
   properties {
    scrollStyle=&TextScrollStyle-Base_C
    text="*Sample Rule 1 +Y    Sample Rule 2 +Y    Sample Rule 3 +Y    Sample Rule 4 +Y    Sample Rule 5 +Y    Sample Rule 6 +Y    Sample Rule 7 +Y    Sample Rule 8 +Y    Sample Rule 9 +Y    Sample Rule 10 +Y"
    colorAndOpacity {colorUseRule=UseColor_Specified, specifiedColor=<rgba 0.65 0.65 0.65 1>}
    font {
     bForceMonospaced=false
     bMaterialIsStencil=false
     fontMaterial=null
     fontObject=&HeadingNow
     letterSpacing=10
     monospacedWidth=1
     size=15.75
     skewAmount=0
     typefaceFontName=74 Regular
     outlineSettings {
      bApplyOutlineToDropShadows=false
      bMiteredCorners=false
      bSeparateFillAlpha=false
      outlineMaterial=null
      outlineSize=2
      outlineColor=<rgba 0.04 0.04 0.04 1>
     }
    }
    shadowOffset=<0 0>
   }
   slot {class=GridSlot, properties {column=1, verticalAlignment=VAlign_Top, padding=<ltrb 16 15 0 0>}}
  }
  ScoringContent {
   widgetClassPath=&GridPanel
   bIsVariable=true
   parent=GridPanel_0
   widget=ScoringContent
   properties {columnFill [1,1,1,1,1], rowFill [0,1]}
   slot {class=GridSlot, properties {row=1, verticalAlignment=VAlign_Top}}
  }
  BackgroundDark_Left {
   widgetClassPath=&Image
   bIsVariable=true
   parent=ScoringContent
   widget=BackgroundDark_Left
   properties {brush {
     bIsDynamicallyLoaded=false
     drawAs=Image
     imageType=FullColor
     mirroring=NoMirror
     resourceName=None
     resourceObject=/Compete/MaterialInstances/MI_UI_CompeteLeaderboardBackground
     tiling=NoTile
     imageSize=<32 32>
     margin=<ltrb 0 0 0 0>
     outlineSettings {
      bUseBrushTransparency=false
      roundingType=HalfHeightRadius
      width=0
      color {colorUseRule=UseColor_Specified, specifiedColor=<rgba 0 0 0 0>}
      cornerRadii=<0 0 0 0>
     }
     tintColor {colorUseRule=UseColor_Specified, specifiedColor=<rgba 1 1 1 1>}
     uVRegion {bIsValid=false, max=<0 0>, min=<0 0>}
    }}
   slot {class=GridSlot, properties {row=1}}
  }
  BackgroundDark_Center {
   widgetClassPath=&Image
   bIsVariable=true
   parent=ScoringContent
   widget=BackgroundDark_Center
   properties {brush {
     bIsDynamicallyLoaded=false
     drawAs=Image
     imageType=FullColor
     mirroring=NoMirror
     resourceName=None
     resourceObject=&MI_UI_CompeteLeaderboardBackground
     tiling=NoTile
     imageSize=<32 32>
     margin=<ltrb 0 0 0 0>
     outlineSettings {
      bUseBrushTransparency=false
      roundingType=HalfHeightRadius
      width=0
      color {colorUseRule=UseColor_Specified, specifiedColor=<rgba 0 0 0 0>}
      cornerRadii=<0 0 0 0>
     }
     tintColor {colorUseRule=UseColor_Specified, specifiedColor=<rgba 1 1 1 1>}
     uVRegion {bIsValid=false, max=<0 0>, min=<0 0>}
    }}
   slot {class=GridSlot, properties {column=2, row=1}}
  }
  BackgroundDark_Right {
   widgetClassPath=&Image
   bIsVariable=true
   parent=ScoringContent
   widget=BackgroundDark_Right
   properties {brush {
     bIsDynamicallyLoaded=false
     drawAs=Image
     imageType=FullColor
     mirroring=NoMirror
     resourceName=None
     resourceObject=&MI_UI_CompeteLeaderboardBackground
     tiling=NoTile
     imageSize=<32 32>
     margin=<ltrb 0 0 0 0>
     outlineSettings {
      bUseBrushTransparency=false
      roundingType=HalfHeightRadius
      width=0
      color {colorUseRule=UseColor_Specified, specifiedColor=<rgba 0 0 0 0>}
      cornerRadii=<0 0 0 0>
     }
     tintColor {colorUseRule=UseColor_Specified, specifiedColor=<rgba 1 1 1 1>}
     uVRegion {bIsValid=false, max=<0 0>, min=<0 0>}
    }}
   slot {class=GridSlot, properties {column=4, row=1}}
  }
  Center_1 {
   widgetClassPath=&Image
   bIsVariable=true
   parent=ScoringContent
   widget=Center_1
   properties {brush {
     bIsDynamicallyLoaded=false
     drawAs=Image
     imageType=FullColor
     mirroring=NoMirror
     resourceName=None
     resourceObject=/Compete/MaterialInstances/MI_UI_CompeteLeaderboardTileBackground
     tiling=NoTile
     imageSize=<32 32>
     margin=<ltrb 0 0 0 0>
     outlineSettings {
      bUseBrushTransparency=false
      roundingType=HalfHeightRadius
      width=0
      color {colorUseRule=UseColor_Specified, specifiedColor=<rgba 0 0 0 0>}
      cornerRadii=<0 0 0 0>
     }
     tintColor {colorUseRule=UseColor_Specified, specifiedColor=<rgba 1 1 1 1>}
     uVRegion {bIsValid=false, max=<0 0>, min=<0 0>}
    }}
   slot {class=GridSlot, properties {column=1, layer=1, row=1}}
  }
  Center_2 {
   widgetClassPath=&Image
   bIsVariable=true
   parent=ScoringContent
   widget=Center_2
   properties {brush {
     bIsDynamicallyLoaded=false
     drawAs=Image
     imageType=FullColor
     mirroring=NoMirror
     resourceName=None
     resourceObject=&MI_UI_CompeteLeaderboardTileBackground
     tiling=NoTile
     imageSize=<32 32>
     margin=<ltrb 0 0 0 0>
     outlineSettings {
      bUseBrushTransparency=false
      roundingType=HalfHeightRadius
      width=0
      color {colorUseRule=UseColor_Specified, specifiedColor=<rgba 0 0 0 0>}
      cornerRadii=<0 0 0 0>
     }
     tintColor {colorUseRule=UseColor_Specified, specifiedColor=<rgba 1 1 1 1>}
     uVRegion {bIsValid=false, max=<0 0>, min=<0 0>}
    }}
   slot {class=GridSlot, properties {column=3, layer=1, row=1}}
  }
  TableHeader {
   widgetClassPath=&GridPanel
   parent=ScoringContent
   widget=TableHeader
   properties {columnFill [1,1], rowFill [1]}
   slot {class=GridSlot, properties {padding=<ltrb 0 19 0 11>}}
  }
  CommonTextBlock_272 {
   widgetClassPath=&CommonTextBlock
   parent=TableHeader
   widget=CommonTextBlock_272
   properties {
    justification=Center
    mobileFontSizeMultiplier=1.5
    scrollStyle=&TextScrollStyle-Base_C
    text=Place
    colorAndOpacity {colorUseRule=UseColor_Specified, specifiedColor=<rgba 0.5776 0.5776 0.5776 1>}
    font {
     bForceMonospaced=false
     bMaterialIsStencil=false
     fontMaterial=null
     fontObject=&HeadingNow
     letterSpacing=0
     monospacedWidth=1
     size=12
     skewAmount=0
     typefaceFontName=74 Regular
     outlineSettings {
      bApplyOutlineToDropShadows=false
      bMiteredCorners=false
      bSeparateFillAlpha=false
      outlineMaterial=null
      outlineSize=0
      outlineColor=<rgba 0 0 0 1>
     }
    }
    shadowOffset=<0 0>
   }
   slot {class=GridSlot, properties {verticalAlignment=VAlign_Top}}
  }
  CommonTextBlock {
   widgetClassPath=&CommonTextBlock
   parent=TableHeader
   widget=CommonTextBlock
   properties {
    justification=Center
    minDesiredWidth=110
    mobileFontSizeMultiplier=1.5
    scrollStyle=&TextScrollStyle-Base_C
    text=Points
    colorAndOpacity {colorUseRule=UseColor_Specified, specifiedColor=<rgba 0.5776 0.5776 0.5776 1>}
    font {
     bForceMonospaced=false
     bMaterialIsStencil=false
     fontMaterial=null
     fontObject=&HeadingNow
     letterSpacing=0
     monospacedWidth=1
     size=12
     skewAmount=0
     typefaceFontName=74 Regular
     outlineSettings {
      bApplyOutlineToDropShadows=false
      bMiteredCorners=false
      bSeparateFillAlpha=false
      outlineMaterial=null
      outlineSize=0
      outlineColor=<rgba 0 0 0 1>
     }
    }
    shadowOffset=<0 0>
   }
   slot {class=GridSlot, properties {column=1, verticalAlignment=VAlign_Bottom}}
  }
  TableSizeBox {
   widgetClassPath=&SizeBox
   bIsVariable=true
   parent=ScoringContent
   widget=TableSizeBox
   properties {bOverride_MinDesiredWidth=true, minDesiredWidth=1100}
   slot {class=GridSlot, properties {columnSpan=5, layer=1, row=1}}
  }
  Table {
   widgetClassPath=/Script/UMG.UniformGridPanel
   bIsVariable=true
   parent=TableSizeBox
   widget=Table
   properties {minDesiredSlotHeight=65, minDesiredSlotWidth=227}
   slot {class=SizeBoxSlot, properties {horizontalAlignment=HAlign_Left}}
  }
 ]
}
