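# Activatable modal that warns the player before switching matchmaking region for tournaments and other modes.
# PreConstruct pads the dialog's content slot and passes a new UIKit dialog view model with the title to the base dialog.
# Construct formats the warning from the CurrentRegionName and NewRegionName variables.
# Accept broadcasts DialogAccepted before deactivating, while Cancel only deactivates; both then fire an element-interacted event and bind to Confirm/Back input actions.
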WidgetBlueprint /Compete/TournamentDetails/WBP_ChangeRegionWarningModal
graphs [
 DialogAccepted {nodes [DialogAccepted() {is=title K2Node_FunctionEntry}]}
 EventGraph {nodes [
   n1 = "On Clicked (CancelButton)"!() -> (OutputDelegate, then, Button:Common Button Base Object Reference) {is=title K2Node_ComponentBoundEvent}
   > DeactivateWidget()
   > Fire Event Element Interacted(Element=@n1.Button)
   n2 = "On Clicked (AcceptButtom)"!() -> (OutputDelegate, then, Button:Common Button Base Object Reference) {is=title K2Node_ComponentBoundEvent}
   > Call Dialog Accepted!(execute:Exec, self:WBP Change Region Warning Modal Object Reference) -> (then:Exec) {is=title K2Node_CallDelegate}
   > DeactivateWidget()
   > Fire Event Element Interacted(Element=@n2.Button)
   Event PreConstruct()
   n3 = Get WBP_UIKit_Dialog_Base() -> (WBP_UIKit_Dialog_Base:WBP UIKit Dialog Internal Object Reference)
   > SetPadding!(
    execute
    self:Grid Slot Object Reference=SlotAsGridSlot(Widget=Get Slot_Content!(self:WBP UIKit Dialog Internal Object Reference=@n3) -> (Slot_Content:Named Slot Object Reference))
    InPadding_Left:float=80
    InPadding_Top:float=0
    InPadding_Right:float=80
    InPadding_Bottom:float=0
   ) -> (then)
   > n4 = Construct Object from Class(Class=/Script/FortniteUI.UIKitDialogViewModel) -> (ReturnValue:UI Kit Dialog Viewmodel Object Reference)
   > Set DialogTitle!(execute, DialogTitle:Text="LOCTEXT(\"Switching Regions\")", self:UI Kit Dialog Viewmodel Object Reference=@n4) -> (then, Output_Get:Text)
   > SetDialogData!(execute, self:WBP UIKit Dialog Internal Object Reference=Get WBP_UIKit_Dialog_Base(), Data:UI Kit Dialog Viewmodel Object Reference=@n4) -> (then)
   Event Construct()
   > Set Text(
    Text=Format Text!(
     Format="LOCTEXT(\"You are about to switch matchmaking region from {CurrentRegion} to {TargetRegion}, limiting your matchmaking to that region for tournaments and other modes. Your connection quality may change, which could affect gameplay. You can change this at any time in your settings.\")"
     CurrentRegion:Text=Get CurrentRegionName() -> (CurrentRegionName:Text)
     TargetRegion:Text=Get NewRegionName() -> (NewRegionName:Text)
    ) -> (Result)
    self=Get MessageText() -> (MessageText:Common Text Object Reference)
   )
   Event Tick()
  ]}
]
variables [
 CurrentRegionName {type=string, default=""}
 DialogAccepted {}
 NewRegionName {type=string, default=""}
]
widgets {
 info {
  inheritedWidgetCount=0
  namedSlotCount=1
  parentClass=&CommonActivatableWidget
  rootWidgetClass=/FortUILibrary/Systems/Dialog/WBP_UIKit_Dialog_Internal.WBP_UIKit_Dialog_Internal_C
  widgetCount=6
 }
 widgets [
  WBP_UIKit_Dialog_Base {
   widgetClassPath=/FortUILibrary/Systems/Dialog/WBP_UIKit_Dialog_Internal
   bIsVariable=true
   widget=WBP_UIKit_Dialog_Base
   properties {
    useText=None
    buttons {
     each Button Horizontal Fill=true
     each Button Min Width=0
     each Button Size=Medium
     focuse Button=0
     space Between Buttons=24
     cTA Buttons Types [Regular,Quiet]
     group Placement {
      horizontal Alignment=HAlign_Fill
      vertical Alignment=VAlign_Bottom
      grid {column=1, column Span=0, layer=0, row=2, row Span=0, nudge=<0 0>}
      padding=<ltrb 80 24 80 48>
     }
    }
    dialogSIze {
     height Override {enable=false, value=592}
     max Aspect Ratio {enable=false, value=0}
     max Desired Height {enable=true, value=886}
     max Desired Width {enable=false, value=0}
     min Aspect Ratio {enable=false, value=0}
     min Desired Height {enable=false, value=0}
     min Desired Width {enable=false, value=0}
     width Override {enable=true, value=850}
    }
    imageProperties {
     material=/FortUILibrary/Systems/Dialog/MaterialInstances/MI_UI_DialogImage
     image Size=<0 0>
     placement {
      horizontal Alignment=HAlign_Fill
      vertical Alignment=VAlign_Fill
      grid {column=1, column Span=0, layer=20, row=0, row Span=1, nudge=<0 0>}
      padding=<ltrb 0 0 0 0>
     }
    }
    titleProperties {
     auto Wrap Text=true
     justification=Left
     line Height Percentage=1
     mobile Font Size Multiplier=1.3
     transform Policy=ToUpper
     wrap Text At=0
     color {colorUseRule=UseColor_Specified, specifiedColor=<rgba 1 1 1 1>}
     font {
      bForceMonospaced=false
      bMaterialIsStencil=false
      fontMaterial=null
      fontObject=&HeadingNow
      letterSpacing=50
      monospacedWidth=1
      size=24
      skewAmount=0
      typefaceFontName=86 Bold Italic
      outlineSettings {
       bApplyOutlineToDropShadows=false
       bMiteredCorners=false
       bSeparateFillAlpha=false
       outlineMaterial=null
       outlineSize=0
       outlineColor=<rgba 0 0 0 1>
      }
     }
     placement {
      horizontal Alignment=HAlign_Fill
      vertical Alignment=VAlign_Top
      grid {column=1, column Span=0, layer=0, row=0, row Span=0, nudge=<0 0>}
      padding=<ltrb 80 50 80 18>
     }
    }
   }
  }
  Content {widgetClassPath=&GridPanel, namedSlotHost=WBP_UIKit_Dialog_Base, widget=Content, properties {columnFill [1], rowFill [0,0,1]}}
  MessageText {
   widgetClassPath=&CommonTextBlock
   bIsVariable=true
   parent=Content
   widget=MessageText
   properties {
    autoWrapText=true
    bIsScrollingEnabled=false
    mobileFontSizeMultiplier=1.14
    scrollStyle=&TextScrollStyle-Base_C
    text="You are about to switch matchmaking region from {CurrentRegion} to {TargetRegion}, limiting your matchmaking to that region for tournaments and other modes. Your connection quality may change, which could affect gameplay. You can change this at any time in your settings"
    colorAndOpacity {colorUseRule=UseColor_Specified, specifiedColor=<rgba 0.5776 0.5776 0.5776 1>}
    font {
     bForceMonospaced=false
     bMaterialIsStencil=false
     fontMaterial=null
     fontObject=&HeadingNow
     letterSpacing=0
     monospacedWidth=1
     size=15.75
     skewAmount=0
     typefaceFontName=74 Regular
     outlineSettings {
      bApplyOutlineToDropShadows=false
      bMiteredCorners=false
      bSeparateFillAlpha=false
      outlineMaterial=null
      outlineSize=0
      outlineColor=<rgba 0 0 0 1>
     }
    }
    shadowOffset=<0 0>
   }
   slot {class=GridSlot, properties {padding=<ltrb 0 0 0 30>}}
  }
  HorizontalBox_34 {
   widgetClassPath=&HorizontalBox
   parent=Content
   widget=HorizontalBox_34
   slot {class=GridSlot, properties {row=2, verticalAlignment=VAlign_Center}}
  }
  AcceptButtom {
   widgetClassPath=&WBP_UIKit_Button_Regular
   bIsVariable=true
   parent=HorizontalBox_34
   widget=AcceptButtom
   properties {bIsFocusable=false, minHeight=56, text=Accept, triggeringInputAction {dataTable=&InputActionDataTable, rowName=Confirm}}
   slot {class=HorizontalBoxSlot, properties {verticalAlignment=VAlign_Center, padding=<ltrb 0 0 24 0>, size {sizeRule=Fill, value=1}}}
  }
  CancelButton {
   widgetClassPath=&WBP_UIKit_Button_Regular
   bIsVariable=true
   parent=HorizontalBox_34
   widget=CancelButton
   properties {
    bIsFocusable=false
    buttonType=Quiet
    minHeight=56
    minWidth=482
    text=Cancel
    triggeringInputAction {dataTable=&InputActionDataTable, rowName=Back}
   }
   slot {class=HorizontalBoxSlot, properties {verticalAlignment=VAlign_Center, size {sizeRule=Fill, value=1}}}
  }
 ]
}
