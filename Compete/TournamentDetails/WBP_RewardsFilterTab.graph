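# Tab button widget for the rewards filter in the Compete tournament details UI, derived from WBP_UIKit_Button_Tab.
# It adds no widgets or variables of its own.
# Its only logic extends On Selected, calling the parent implementation and then requesting that the button's UIKit blocks transition to the Selected visual state.
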WidgetBlueprint /Compete/TournamentDetails/WBP_RewardsFilterTab
graphs [EventGraph {nodes [
   Event On Selected()
   > Parent: On Selected() {is=title K2Node_CallParentFunction}
   > RequestTransitionOnBlocks!(execute, self:UIKit Modular Button Object Reference, State:EUIKitBlockVisualState Enum=Selected) -> (then)
  ]}]
variables []
widgets {
 info {
  inheritedWidgetCount=0
  parentClass=/FortUILibrary/UIKit/Buttons/TabButtons/WBP_UIKit_Button_Tab.WBP_UIKit_Button_Tab_C
  rootWidgetClass=null
  widgetCount=0
 }
 widgets []
}
