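# Full-image widget that draws the tournament details background in the Compete UI using a configurable material.
# PreConstruct applies the material, while setters write two colors and a texture to its dynamic instance.
# Colors at or above a luminance cap are darkened by multiplying their RGB by that cap.
# A Tournament Background view model is set up by image type and set to an empty input on Destruct.
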WidgetBlueprint /Compete/TournamentDetails/WBP_TournamentBackground
graphs [
 EventGraph {nodes [
   Event PreConstruct()
   > SetBrushFromMaterial(
    self=Get Background_Image() -> (Background_Image:Image Object Reference)
    Material=Get BackgroundMaterial() -> (BackgroundMaterial:Material Interface Object Reference)
   )
   Event Destruct()
   > SetTournament Background VM!(execute, self, Viewmodel:Tournament Background VM Object Reference) -> (then)
   Event Construct()
   Event Tick()
  ]}
 Set Background Main Color {nodes [
   n1 = Set Background Main Color!() -> (then, New Color:Linear Color Structure) {is=title K2Node_FunctionEntry}
   > Set Material Color!(execute, self, Image:Image Object Reference=Get Background_Image(), Material Parameter:Name=Color1, Color:Linear Color Structure=@n1) -> (then) {is=title K2Node_CallFunction}
  ]}
 Set Background Second Color {nodes [
   n1 = Set Background Second Color!() -> (then, New Color:Linear Color Structure) {is=title K2Node_FunctionEntry}
   > Set Material Color(Image=Get Background_Image(), Material Parameter=Color2, Color=@n1)
  ]}
 Set Background Texture {nodes [
   n1 = Set Background Texture!() -> (then, InputPin:Texture Object Reference) {is=title K2Node_FunctionEntry}
   > n2 = GetDynamicMaterial(self=Get Background_Image())
   > Is Valid(InputObject=@n2)
   > SetTextureParameterValue(self=@n2, ParameterName=Texture, Value=@n1)
  ]}
 Set Material Color {nodes [
   n1 = Set Material Color!() -> (then, Image:Image Object Reference, Material Parameter:Name, Color:Linear Color Structure) {is=title K2Node_FunctionEntry}
   > n2 = GetDynamicMaterial(self=@n1)
   > Is Valid(InputObject=@n2)
   n3 = Get Max Color Luminance() -> (Max Color Luminance:double)
   n4 = MakeColor!(R:float=@n3, G:float=@n3, B:float=@n3, A:float=1) -> (ReturnValue:Linear Color Structure)
   n5 = "Luminance (LinearColor)"!(InColor:Linear Color Structure=@n1.Color) -> (ReturnValue:float)
   > SetVectorParameterValue(
    self=@n2
    ParameterName=@n1.Material Parameter
    Value=SelectColor(
     A=@n1.Color
     B=LinearColor * LinearColor(A:Linear Color Structure=@n1.Color, B:Linear Color Structure=@n4) -> (ReturnValue:Linear Color Structure) {is=title K2Node_PromotableOperator}
     bPickA="float < float"(A=@n5, B=@n3)
    )
   )
  ]}
 Setup View Model {nodes [
   n1 = Setup View Model!() -> (then, BackgroundType:EFortTournamentBackgroundImageType Enum=LoadingScreen) {is=title K2Node_FunctionEntry}
   > SetupViewModel!(
    execute
    self:Tournament Background VM Object Reference=Get Tournament Background VM() -> (Tournament Background VM:Tournament Background VM Object Reference)
    InType:EFortTournamentBackgroundImageType Enum=@n1
   ) -> (then)
  ]}
]
variables [
 BackgroundMaterial {type=MaterialInterface, default=/Compete/Materials/MI_UI_Compete_TournamentDetailsBackground}
 Max Color Luminance {type=number, default=0.15}
]
widgets {
 info {inheritedWidgetCount=0, parentClass=&UserWidget, rootWidgetClass=&Image, widgetCount=1}
 widgets [Background_Image {
   widgetClassPath=&Image
   bIsVariable=true
   widget=Background_Image
   properties {brush {
     bIsDynamicallyLoaded=false
     drawAs=Image
     imageType=FullColor
     mirroring=NoMirror
     resourceName=None
     resourceObject=&MI_UI_Compete_TournamentDetailsBackground
     tiling=NoTile
     imageSize=<1920 1080>
     margin=<ltrb 0 0 0 0>
     outlineSettings {
      bUseBrushTransparency=false
      roundingType=HalfHeightRadius
      width=0
      color {colorUseRule=UseColor_Specified, specifiedColor=<rgba 0 0 0 0>}
      cornerRadii=<0 0 0 0>
     }
     tintColor {colorUseRule=UseColor_Specified, specifiedColor=<rgba 1 1 1 1>}
     uVRegion {bIsValid=false, max=<0 0>, min=<0 0>}
    }}
  }]
}
