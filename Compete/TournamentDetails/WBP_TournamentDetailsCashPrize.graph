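# Tournament details widget that shows a tournament's cash prize in a single text block.
# SetData reads the header payout item from FortTournamentViewVM, not from its Header Data input, and tests whether the reward type is Commerce.
# That test drives focusability and visibility, collapsing the widget for other reward types.
# For Commerce rewards, the payout's display text replaces the placeholder in CommonText_Cash.
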WidgetBlueprint /Compete/TournamentDetails/WBP_TournamentDetailsCashPrize
graphs [
 EventGraph {links [], nodes []}
 SetData {nodes [
   SetData!() -> (then, Header Data:Tournament Payout Threshold Data Structure) {is=title K2Node_FunctionEntry}
   n1 = Break Tournament Payout Item Data!(TournamentPayoutItemData:"Tournament Payout Item Data Structure (by ref)"=Get HeaderPayoutItemData!(self:Tournament View VM Object Reference=Get FortTournamentViewVM() -> (FortTournamentViewVM:Tournament View VM Object Reference)) -> (HeaderPayoutItemData:Tournament Payout Item Data Structure)) -> (RewardType:EPayoutRewardType Enum, DisplayText:Text) {is=title K2Node_BreakStruct}
   n2 = "Equal (Enum)"(A:EPayoutRewardType Enum=@n1, B:EPayoutRewardType Enum=Commerce)
   > Set bIsFocusable(bIsFocusable:bool=@n2) -> (Output_Get:bool)
   > SetVisibility(InVisibility=Select(Option 0=Collapsed, Option 1=Visible, Index=@n2))
   > Branch(Condition=@n2)
   > "SetText (Text)"(self=Get CommonText_Cash() -> (CommonText_Cash:Common Text Object Reference), InText=@n1.DisplayText)
  ]}
]
variables []
widgets {
 info {inheritedWidgetCount=0, parentClass=&UserWidget, rootWidgetClass=&GridPanel, widgetCount=2}
 widgets [
  Root {widgetClassPath=&GridPanel, widget=Root, properties {columnFill [0,1], rowFill [1]}}
  CommonText_Cash {
   widgetClassPath=&CommonTextBlock
   bIsVariable=true
   parent=Root
   widget=CommonText_Cash
   properties {
    scrollStyle=&TextScrollStyle-Base_C
    text="$-,---"
    font {
     bForceMonospaced=false
     bMaterialIsStencil=false
     fontMaterial=null
     fontObject=&NotoSans
     letterSpacing=0
     monospacedWidth=1
     size=24
     skewAmount=0
     typefaceFontName=NotoSans-Regular
     outlineSettings {
      bApplyOutlineToDropShadows=false
      bMiteredCorners=false
      bSeparateFillAlpha=false
      outlineMaterial=null
      outlineSize=0
      outlineColor=<rgba 0 0 0 1>
     }
    }
    shadowOffset=<0 0>
   }
   slot {class=GridSlot, properties {horizontalAlignment=HAlign_Left, verticalAlignment=VAlign_Center, padding=<ltrb 0 0 12 0>}}
  }
 ]
}
