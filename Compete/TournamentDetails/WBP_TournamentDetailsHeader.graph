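# Tournament details header widget showing a Live highlight, round, date and time, tournament name, and a tag for game mode, team size and type icon.
# Setter functions fill each element, hiding empty rounds and mapping enums to localized tag text or icon materials.
# PreConstruct makes the name either wrap at a fixed width or scroll, based on Wrap Name.
# Two functions forward setup and window updates to the header view model.
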WidgetBlueprint /Compete/TournamentDetails/WBP_TournamentDetailsHeader
graphs [
 EventGraph {nodes [
   Event PreConstruct()
   > SetTextTransformPolicy!(
    execute
    self:Text Object Reference=Get Text_Label!(self:WBP UIKit Tag Base Object Reference=Get TournamentTag() -> (TournamentTag:WBP UIKit Tag Object Reference)) -> (Text_Label:Common Text Object Reference)
    InTransformPolicy:ETextTransformPolicy Enum=None
   ) -> (then)
   > n1 = Branch(Condition=Get Wrap Name() -> (Wrap Name:bool))
   > SetWidthOverride(self=Get SizeBox_Name() -> (SizeBox_Name:Size Box Object Reference), InWidthOverride=Get Name Width() -> (Name Width:float))
   > SetScrollingEnabled!(
    execute
    self:Common Text Object Reference=Get TournamentName() -> (TournamentName:Common Text Object Reference)
    bInIsScrollingEnabled:bool=false
   ) -> (then)
   > Set AutoWrapText!(execute, AutoWrapText:bool=true, self:Text Layout Widget Object Reference=Get TournamentName()) -> (then, Output_Get:bool)
   ClearWidthOverride(execute=@n1.else, self:Size Box Object Reference=Get SizeBox_Name())
   > SetScrollingEnabled(self=Get TournamentName(), bInIsScrollingEnabled=true)
   > Set AutoWrapText(AutoWrapText=false, self=Get TournamentName())
  ]}
 SetLiveStatus {nodes [
   n1 = SetLiveStatus!() -> (then, IsLive:bool=~) {is=title K2Node_FunctionEntry}
   > SetVisibility(
    self=Get Highlight() -> (Highlight:WBP UIKit Highlight Object Reference)
    InVisibility=Select(Option 0=Collapsed, Option 1=SelfHitTestInvisible, Index=@n1)
   )
  ]}
 SetRoundDateTimeInfo {nodes [
   n1 = SetRoundDateTimeInfo!() -> (then, DateTimeInfo:Text) {is=title K2Node_FunctionEntry}
   > "SetText (Text)"(self=Get DateAndTime() -> (DateAndTime:Common Text Object Reference), InText=@n1)
  ]}
 SetRoundInfo {nodes [
   n1 = SetRoundInfo!() -> (then, RoundName:Text) {is=title K2Node_FunctionEntry}
   > n2 = Branch(Condition=TextIsEmpty(InText=@n1))
   > SetVisibility(self=Get Round() -> (Round:Common Text Object Reference), InVisibility=Collapsed)
   "SetText (Text)"(execute=@n2.else, self=Get Round(), InText=Get RoundName() -> (RoundName:Text))
   > SetVisibility(self=Get Round(), InVisibility=SelfHitTestInvisible)
  ]}
 SetTagExperienceType {nodes [
   n1 = SetTagExperienceType!() -> (then, GameType:EFortCompeteExperienceType Enum=Unknown) {is=title K2Node_FunctionEntry}
   > SetTagText!(
    execute
    self:WBP UIKit Tag Base Object Reference=Get TournamentTag()
    Text:Text=Select!(
     Unknown:Text
     Other:Text
     BattleRoyale:Text="LOCTABLE(\"/Compete/StringTables/GameModeStringTable\", \"BRKey\")"
     ZeroBuild:Text="LOCTABLE(\"&GameModeStringTable\", \"ZBKey\")"
     BlastBerry:Text="LOCTABLE(\"&GameModeStringTable\", \"ReloadKey\")"
     Figment:Text="LOCTABLE(\"&GameModeStringTable\", \"FigmentKey\")"
     ForbiddenFruit:Text="LOCTEXT(\"Blitz\")"
     Feral:Text="LOCTABLE(\"&GameModeStringTable\", \"FeralKey\")"
     Index:EFortCompeteExperienceType Enum=@n1
    ) -> (ReturnValue:Text)
   ) -> (then)
  ]}
 SetTagTeamType {nodes [
   n1 = SetTagTeamType!() -> (then, TeamType:EFortCompeteMatchType Enum=Unknown) {is=title K2Node_FunctionEntry}
   > SetLabelText!(
    execute
    self:WBP UIKit Tag Base Object Reference=Get TournamentTag()
    Text:Text=Select!(
     Unknown
     Solo:Text="LOCTEXT(\"Solo\")"
     Duos:Text="LOCTEXT(\"Duos\")"
     Trios:Text="LOCTEXT(\"Trios\")"
     Squads:Text="LOCTEXT(\"Squads\")"
     Index:EFortCompeteMatchType Enum=@n1
    ) -> (ReturnValue)
   ) -> (then)
  ]}
 SetTournamentName {nodes [
   n1 = SetTournamentName!() -> (then, TournamentName:Text) {is=title K2Node_FunctionEntry}
   > "SetText (Text)"(self=Get TournamentName(), InText=@n1)
  ]}
 SetTournamentTypeIcon {nodes [
   n1 = SetTournamentTypeIcon!() -> (then, TournamentType:ETournamentType Enum=None) {is=title K2Node_FunctionEntry}
   n2 = Find(
    TargetMap:Map of ETournamentType Enums to Material Instance Object References=Get TournamentTypeIcons() -> (TournamentTypeIcons:Map of ETournamentType Enums to Material Instance Object References=None)
    Key:"ETournamentType Enum (by ref)"=@n1
   ) -> (Value:Material Instance Object Reference)
   > Branch(Condition=@n2.ReturnValue)
   > SetIconMaterial!(execute, self:WBP UIKit Tag Base Object Reference=Get TournamentTag(), IconMaterial:Material Instance Object Reference=@n2) -> (then)
  ]}
 Setup Tournament View Model {nodes [
   Setup Tournament View Model() {is=title K2Node_FunctionEntry}
   > SetupViewModel(self:Tournament Header VM Object Reference=Get FortTournamentHeaderVM() -> (FortTournamentHeaderVM:Tournament Header VM Object Reference))
  ]}
 Update Window View Model {nodes [
   Update Window View Model() {is=title K2Node_FunctionEntry}
   > UpdateTournamentWindow(self:Tournament Header VM Object Reference=Get FortTournamentHeaderVM())
  ]}
]
variables [
 Name Width {type=number, default=600}
 TournamentTypeIcons {default {
   cashCup=/FortUILibrary/Textures/MaterialInstances/MI_UIKit_Icon_CashCup
   fNCS=/FortUILibrary/Textures/MaterialInstances/MI_UIKit_Icon_FNCS
   none=/FortUILibrary/Textures/MaterialInstances/MI_UIKit_Icon_OtherTournaments
   other=&MI_UIKit_Icon_OtherTournaments
   rankedCup=/FortUILibrary/Textures/MaterialInstances/MI_UIKit_Icon_RankedCup
   shopCup=/FortUILibrary/Textures/MaterialInstances/MI_UIKit_Icon_ItemShopCup
   victoryCup=/FortUILibrary/Textures/MaterialInstances/MI_UIKit_Icon_VictoryCup
  }}
 Wrap Name {type=boolean}
]
widgets {
 info {inheritedWidgetCount=0, parentClass=&CommonUserWidget, rootWidgetClass=&GridPanel, widgetCount=8}
 widgets [
  Root {widgetClassPath=&GridPanel, widget=Root, properties {columnFill [1], rowFill [0,1,0]}}
  DateTimeInfo {widgetClassPath=&GridPanel, parent=Root, widget=DateTimeInfo, slot {class=GridSlot}}
  Highlight {
   widgetClassPath=/FortUILibrary/UIKit/Highlight/WBP_UIKit_Highlight
   bIsVariable=true
   parent=DateTimeInfo
   widget=Highlight
   properties {highlightText=Live, highlightType=Social, widthOverride=103}
   slot {class=GridSlot, properties {verticalAlignment=VAlign_Top, padding=<ltrb 0 0 17 0>}}
  }
  Round {
   widgetClassPath=&CommonTextBlock
   bIsVariable=true
   parent=DateTimeInfo
   widget=Round
   properties {
    text=Final Round
    font {
     bForceMonospaced=false
     bMaterialIsStencil=false
     fontMaterial=null
     fontObject=&NotoSans
     letterSpacing=0
     monospacedWidth=1
     size=24
     skewAmount=0
     typefaceFontName=NotoSans-Regular
     outlineSettings {
      bApplyOutlineToDropShadows=false
      bMiteredCorners=false
      bSeparateFillAlpha=false
      outlineMaterial=null
      outlineSize=0
      outlineColor=<rgba 0 0 0 1>
     }
    }
    shadowOffset=<0 0>
   }
   slot {class=GridSlot, properties {column=1, verticalAlignment=VAlign_Bottom, padding=<ltrb 0 0 32 0>}}
  }
  DateAndTime {
   widgetClassPath=&CommonTextBlock
   bIsVariable=true
   parent=DateTimeInfo
   widget=DateAndTime
   properties {
    text="2/5/24  6:00 PM - 8:00 PM GMT-3"
    font {
     bForceMonospaced=false
     bMaterialIsStencil=false
     fontMaterial=null
     fontObject=&HeadingNow
     letterSpacing=0
     monospacedWidth=1
     size=18
     skewAmount=0
     typefaceFontName=74 Regular
     outlineSettings {
      bApplyOutlineToDropShadows=false
      bMiteredCorners=false
      bSeparateFillAlpha=false
      outlineMaterial=null
      outlineSize=0
      outlineColor=<rgba 0 0 0 1>
     }
    }
    shadowOffset=<0 0>
   }
   slot {class=GridSlot, properties {column=2, verticalAlignment=VAlign_Bottom}}
  }
  SizeBox_Name {
   widgetClassPath=&SizeBox
   bIsVariable=true
   parent=Root
   widget=SizeBox_Name
   slot {class=GridSlot, properties {horizontalAlignment=HAlign_Left, row=1, padding=<ltrb 0 4 0 0>}}
  }
  TournamentName {
   widgetClassPath=&CommonTextBlock
   bIsVariable=true
   parent=SizeBox_Name
   widget=TournamentName
   properties {
    scrollStyle=&TextScrollStyle-Base_C
    text=Silver scallywag cash cup
    textTransformPolicy=ToUpper
    font {
     bForceMonospaced=false
     bMaterialIsStencil=false
     fontMaterial=null
     fontObject=&HeadingNow
     letterSpacing=-60
     monospacedWidth=1
     size=32.25
     skewAmount=0
     typefaceFontName=86 Bold Italic
     outlineSettings {
      bApplyOutlineToDropShadows=false
      bMiteredCorners=false
      bSeparateFillAlpha=false
      outlineMaterial=null
      outlineSize=0
      outlineColor=<rgba 0 0 0 1>
     }
    }
    shadowOffset=<0 0>
   }
   slot {class=SizeBoxSlot}
  }
  TournamentTag {
   widgetClassPath=/FortUILibrary/UIKit/Tag/WBP_UIKit_Tag
   bIsVariable=true
   parent=Root
   widget=TournamentTag
   properties {
    iconLeftAligned=true
    iconMaterial=/FortUILibrary/Textures/MaterialInstances/MI_UIKit_Icon_Like
    labelText=Squads
    tagText=Battle Royale
    tagType=Primary Tag
    useIcon=true
    iconPadding=<ltrb 0 0 6 0>
    iconSizeCompact=<0 0>
    primaryFontCompact {
     bForceMonospaced=false
     bMaterialIsStencil=false
     fontMaterial=null
     fontObject=null
     letterSpacing=0
     monospacedWidth=1
     size=24
     skewAmount=0
     typefaceFontName=None
     outlineSettings {
      bApplyOutlineToDropShadows=false
      bMiteredCorners=false
      bSeparateFillAlpha=false
      outlineMaterial=null
      outlineSize=0
      outlineColor=<rgba 0 0 0 1>
     }
    }
    rarityFontCompact {
     bForceMonospaced=false
     bMaterialIsStencil=false
     fontMaterial=null
     fontObject=null
     letterSpacing=0
     monospacedWidth=1
     size=24
     skewAmount=0
     typefaceFontName=None
     outlineSettings {
      bApplyOutlineToDropShadows=false
      bMiteredCorners=false
      bSeparateFillAlpha=false
      outlineMaterial=null
      outlineSize=0
      outlineColor=<rgba 0 0 0 1>
     }
    }
    tagCompactPadding=<ltrb 0 0 0 0>
   }
   slot {class=GridSlot, properties {row=2, padding=<ltrb 0 12 0 0>}}
  }
 ]
}
