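# Eligibility button for Compete tournament details, built on the UIKit generic button, that shows whether eligibility is met.
# IsEligible takes a bool and swaps the button icon between the eligibility material and a check icon.
# It also drives the opacity of an overlaid Pulse material image from the inverted bool, so the pulse, hidden by default, appears only when ineligible.
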WidgetBlueprint /Compete/TournamentDetails/WBP_TournamentDetails_EligibilityButton
graphs [
 EventGraph {links [], nodes []}
 IsEligible {nodes [
   n1 = IsEligible!() -> (then, InEligible:bool=~) {is=title K2Node_FunctionEntry}
   > SetIcon!(
    execute
    self:WBP UIKit Button Generic Base Object Reference
    Icon:Object Reference=Select(
     Option 0:Object Reference=/Compete/MaterialInstances/MI_UI_Compete_TournamentDetails_Eligibility
     Option 1:Object Reference=/Compete/MaterialInstances/MI_UI_Compete_CheckIcon
     Index=@n1
    ) -> (ReturnValue:Object Reference)
   ) -> (then)
   > SetRenderOpacity(self=Get Pulse() -> (Pulse:Image Object Reference), InOpacity="To Float (Boolean)"(InBool=NOT Boolean(A=@n1)))
  ]}
]
variables []
widgets {
 info {
  inheritedWidgetCount=0
  namedSlotCount=1
  parentClass=/FortUILibrary/UIKit/Buttons/WBP_UIKit_Button_Generic.WBP_UIKit_Button_Generic_C
  rootWidgetClass=null
  widgetCount=2
 }
 widgets [
  Overlay_0 {widgetClassPath=&Overlay, widget=Overlay_0}
  Pulse {
   widgetClassPath=&Image
   bIsVariable=true
   parent=Overlay_0
   widget=Pulse
   properties {
    renderOpacity=0
    brush {
     bIsDynamicallyLoaded=false
     drawAs=Image
     imageType=FullColor
     mirroring=NoMirror
     resourceName=None
     resourceObject=/Compete/Materials/M_UI_Compete_TournamentDetails_Eligibility_Pulse
     tiling=NoTile
     imageSize=<32 32>
     margin=<ltrb 0 0 0 0>
     outlineSettings {
      bUseBrushTransparency=false
      roundingType=HalfHeightRadius
      width=0
      color {colorUseRule=UseColor_Specified, specifiedColor=<rgba 0 0 0 0>}
      cornerRadii=<0 0 0 0>
     }
     tintColor {colorUseRule=UseColor_Specified, specifiedColor=<rgba 1 1 1 1>}
     uVRegion {bIsValid=false, max=<0 0>, min=<0 0>}
    }
   }
   slot {class=OverlaySlot, properties {horizontalAlignment=HAlign_Fill, verticalAlignment=VAlign_Fill, padding=<ltrb -14 -14 -14 -14>}}
  }
 ]
}
