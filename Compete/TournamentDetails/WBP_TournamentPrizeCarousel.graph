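# Tournament details prize carousel: a horizontal list of cosmetic reward tiles with payout-threshold filter tabs.
# SetLeaderboardsPrizesData flattens payouts into reward viewmodels, mapping each tab to its index range.
# With multiple thresholds it adds an "All" tab, otherwise a text header; selecting a tab scrolls to its range and greys out other items.
# Clicking a tile opens a preview modal, hiding the details screen; arrow visibility follows input type.
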WidgetBlueprint /Compete/TournamentDetails/WBP_TournamentPrizeCarousel
animations [Anim_ShowArrowButtons {
  binding_count=4
  length_seconds=0.1991
  object=&self:Anim_ShowArrowButtons
  sequence {bindings [
    GridSlot_1 {id=#1, parent_id=#2, possessed_class=GridSlot, tracks []}
    NextButton {
     id=#3
     possessed_class=WBP_UIKit_Button_Generic_C
     tracks [
      {
       type=MovieScene2DTransformTrack
       sections [{
         type=MovieScene2DTransformSection
         end_frame=3
         start_frame=0
         channels [
          Translation.X {type=MovieSceneScriptingFloatChannel, keys [[0,-82],[3.9813,0]]}
          Translation.Y {type=MovieSceneScriptingFloatChannel, const [0,2]}
          Angle {type=MovieSceneScriptingFloatChannel}
          Scale.X {type=MovieSceneScriptingFloatChannel}
          Scale.Y {type=MovieSceneScriptingFloatChannel}
          Shear.X {type=MovieSceneScriptingFloatChannel}
          Shear.Y {type=MovieSceneScriptingFloatChannel}
         ]
        }]
      }
      {
       type=MovieSceneFloatTrack
       display_name=Render Opacity
       sections [{type=MovieSceneFloatSection, end_frame=3, start_frame=0, channels [None {type=MovieSceneScriptingFloatChannel, keys [[0,0],[3.9813,1]]}]}]
      }
     ]
    }
    Tabs_Subcategory {id=#2, possessed_class=WBP_UIKit_TabGroup_C, tracks []}
    PrevButton {
     id=#4
     possessed_class=WBP_UIKit_Button_Generic_C
     tracks [
      {
       type=MovieScene2DTransformTrack
       sections [{
         type=MovieScene2DTransformSection
         end_frame=3
         start_frame=0
         channels [
          Translation.X {type=MovieSceneScriptingFloatChannel, keys [[0,82],[3.9817,0]]}
          Translation.Y {type=MovieSceneScriptingFloatChannel, const [0,1]}
          Angle {type=MovieSceneScriptingFloatChannel}
          Scale.X {type=MovieSceneScriptingFloatChannel}
          Scale.Y {type=MovieSceneScriptingFloatChannel}
          Shear.X {type=MovieSceneScriptingFloatChannel}
          Shear.Y {type=MovieSceneScriptingFloatChannel}
         ]
        }]
      }
      {
       type=MovieSceneFloatTrack
       display_name=Render Opacity
       sections [{type=MovieSceneFloatSection, end_frame=3, start_frame=0, channels [None {type=MovieSceneScriptingFloatChannel, keys [[0,0],[3.9817,1]]}]}]
      }
     ]
    }
   ]}
 }]
graphs [
 EventGraph {nodes [
   n1 = "On Item Selection Changed (ListViewPrizes)"!() -> (OutputDelegate, then, Item:Object Reference, bIsSelected:bool) {is=title K2Node_ComponentBoundEvent}
   > Branch(Condition=@n1.bIsSelected)
   > Scroll Item Into View!(
    execute
    self:List View Object Reference=Get ListViewPrizes() -> (ListViewPrizes:Common List View Object Reference)
    Item:Object Reference=@n1.Item
   ) -> (then)
   n2 = "On Clicked (NextButton)"!() -> (OutputDelegate, then, Button:Common Button Base Object Reference) {is=title K2Node_ComponentBoundEvent}
   > Fire Event Element Interacted(Element=@n2.Button)
   n3 = Get Tabs_Subcategory() -> (Tabs_Subcategory:WBP UIKit Tab Group Object Reference)
   > SelectNextTab!(execute, self:UIKit Tab Group Object Reference=@n3, bSuppressClickFeedback:bool=true) -> (then)
   n4 = "On Clicked (PrevButton)"!() -> (OutputDelegate, then, Button:Common Button Base Object Reference) {is=title K2Node_ComponentBoundEvent}
   > Fire Event Element Interacted(Element=@n4.Button)
   > SelectPreviousTab!(execute, self:UIKit Tab Group Object Reference=@n3, bSuppressClickFeedback:bool=true) -> (then)
   n5 = "On UIKit Tab Selected (Tabs_Subcategory)"!() -> (OutputDelegate, then, TabID:Name=None, TabVM:MVVM Base Viewmodel Object Reference) {is=title K2Node_ComponentBoundEvent}
   > n6 = Cast To FortPlayerController(Object=GetOwningPlayer())
   > n7 = "Fire Event Element Interacted (Non-Localized)"!(
    execute
    FortPC:Fort Player Controller Object Reference=@n6
    ElementName:String=TabSubCategory
    AdditionalInfo:Flow Analytics Additional Info Structure="Add String (Non-Localized)"!(
     Info:"Flow Analytics Additional Info Structure (by ref)"=Make Flow Analytics Additional Info!() -> (FlowAnalyticsAdditionalInfo:Flow Analytics Additional Info Structure) {is=title K2Node_MakeStruct}
     Key:String=TabID
     Value:String="To String (Name)"(InName=@n5.TabID)
    ) -> (ReturnValue:Flow Analytics Additional Info Structure)
   ) -> (then)
   n8 = Branch(execute=[@n7.then, @n6.CastFailed], Condition="Equal (Name)"(A=@n5.TabID, B=All))
   > Set Greyed Out Rewards Items!(execute, self, In Min:int=0, In Max:int=9999) -> (then) {is=title K2Node_CallFunction}
   > SetSelectedIndex!(execute, self:List View Object Reference=Get ListViewPrizes(), Index:int=0) -> (then)
   n9 = Cast To UIKitSelectableButtonViewModel!(execute=@n8.else, Object=@n5.TabVM) -> (then, CastFailed, AsUI Kit Selectable Button View Model:UI Kit Selectable Button View Model Object Reference) {is=title K2Node_DynamicCast}
   n10 = Find(
    TargetMap:Map of UI Kit Selectable Button View Model Object References to Integers=Get Tab Buttons to First Item Index() -> (Tab Buttons to First Item Index:Map of UI Kit Selectable Button View Model Object References to Integers)
    Key:"UI Kit Selectable Button View Model Object Reference (by ref)"=@n9
   ) -> (Value:int)
   n11 = Find(
    TargetMap=Get Tab Buttons to Last Item Index() -> (Tab Buttons to Last Item Index:Map of UI Kit Selectable Button View Model Object References to Integers)
    Key=@n9
   )
   > Branch(Condition=AND Boolean(A=@n10.ReturnValue, B=@n11.ReturnValue))
   > Set Greyed Out Rewards Items(In Min=@n10, In Max=@n11)
   > SetSelectedIndex(self=Get ListViewPrizes(), Index=@n10)
   "On Clicked (PreviousRewardButton)"!() -> (OutputDelegate, then, Button:Common Button Base Object Reference) {is=title K2Node_ComponentBoundEvent}
   n12 = Get ListViewPrizes()
   n13 = GetIndexForItem!(
    self:List View Object Reference=@n12
    Item:Object Reference=Get Selected Item!(self:List View Object Reference=@n12) -> (ReturnValue:Object Reference) {is=title K2Node_CallFunction}
   ) -> (ReturnValue:int=0)
   > Branch(Condition="integer > integer"(A=@n13))
   > SetSelectedIndex(self=Get ListViewPrizes(), Index=int - int(A=@n13))
   "On Clicked (NextRewardButton)"!() -> (OutputDelegate, then, Button:Common Button Base Object Reference) {is=title K2Node_ComponentBoundEvent}
   n14 = GetNumItems!(self:List View Object Reference=@n12) -> (ReturnValue:int=0)
   > Branch(Condition="integer > integer"(A=int - int(A=@n14), B=@n13))
   > SetSelectedIndex(self=Get ListViewPrizes(), Index=int + int(A=@n13))
   n15 = "On Item Clicked (ListViewPrizes)"!() -> (OutputDelegate, then, Item:Object Reference) {is=title K2Node_ComponentBoundEvent}
   > n16 = Cast To FortTournamentRewardItemVM!(execute, Object=@n15.Item) -> (then, CastFailed, AsTournament Reward Item VM:Tournament Reward Item VM Object Reference) {is=title K2Node_DynamicCast}
   > n17 = Create Widget!(
    execute:Exec
    Class:User Widget Class Reference=/Compete/TournamentDetails/ItemPreview/WBP_ItemPreview_Modal.WBP_ItemPreview_Modal_C
    OwningPlayer:Player Controller Object Reference=GetOwningPlayer()
    FortTournamentViewVM:Tournament View VM Object Reference=Get FortTournamentViewVM() -> (FortTournamentViewVM:Tournament View VM Object Reference)
    InstantSelectedItem:Fort Item VM Object Reference=Get ItemVM!(self:Tournament Reward Item VM Object Reference=@n16) -> (ItemVM:Fort Item VM Object Reference)
   ) -> (then:Exec, ReturnValue:WBP Item Preview Modal Object Reference)
   > TryPushWidgetAsModalPanel!(execute, PushWidget:User Widget Object Reference=@n17) -> (then)
   > n18 = ActivateWidget(self=@n17)
   n19 = RestoreEnabled() {is=title K2Node_CustomEvent}
   Bind Event to On Widget Deactivated(execute=@n18.then, self:Common Activatable Widget Object Reference=@n17, Delegate=@n19) {is=title K2Node_AddDelegate}
   > Set bIsEnabled(bIsEnabled:bool=false) -> (Output_Get:bool)
   n20 = Self-Reference()
   > n21 = FindParentWidgetOfType!(
    execute
    StartingWidget:Widget Object Reference=@n20
    Type:Widget Class Reference=&/WBP_TournamentDetails.WBP_TournamentDetails_C
   ) -> (then, ReturnValue:WBP Tournament Details Object Reference)
   > SetVisibility(self=@n21, InVisibility=Hidden)
   DelayedRefreshButtons() {is=title K2Node_CustomEvent}
   > ForceLayoutPrepass(self:Widget Object Reference=Get ListViewPrizes())
   > DelayUntilNextTick!(execute) -> (then)
   > VisibilityTabsNavButtonsByInputType!(execute, self, InputType:ECommonInputType Enum=GetCurrentInputType(self=Get CommonInputSubsystem())) -> (then)
   Event PreConstruct()
   > SetMobileStyle() {is=title K2Node_CallFunction}
   > SetColorAndOpacity(
    self:User Widget Object Reference=[
     Get Block_Icon!(self:WBP UIKit Button Generic Base Object Reference=Get PreviousRewardButton() -> (PreviousRewardButton:WBP UIKit Button Generic Object Reference)) -> (Block_Icon:WBP UIKit Block Image Object Reference)
     Get Block_Icon(self=Get NextRewardButton() -> (NextRewardButton:WBP UIKit Button Generic Object Reference))
    ]
    InColorAndOpacity:Linear Color Structure=(R=0,G=0,B=0,A=1)
   )
   Event Construct()
   n22 = Get CommonInputSubsystem()
   > Bind Event to On Input Method Changed(self=@n22, Delegate=Create Event())
   > VisibilityTabsNavButtonsByInputType(InputType=GetCurrentInputType(self=@n22))
   Event Destruct()
   > Unbind Event from On Input Method Changed(self=Get CommonInputSubsystem(), Delegate=Create Event())
   SetFocus(execute=@n24.then, self=Get ListViewPrizes())
   Set!(
    execute
    WidgetStyle_BarThickness:float=9
    WidgetStyle_TopShadowBrush:Slate Brush Structure=(TintColor=(SpecifiedColor=(R=1,G=1,B=1,A=1),ColorUseRule=UseColor_Specified),DrawAs=Image,Tiling=NoTile,Mirroring=NoMirror,ImageType=NoImage,ImageSize=(X=32,Y=32),Margin=(Left=0,Top=0,Right=0,Bottom=0),ResourceObject=None,OutlineSettings=(CornerRadii=(X=0,Y=0,Z=0,W=0),Color=(SpecifiedColor=(R=0,G=0,B=0,A=0),ColorUseRule=UseColor_Specified),Width=0,RoundingType=HalfHeightRadius,bUseBrushTransparency=False),UVRegion=(Min=(X=0,Y=0),Max=(X=0,Y=0),bIsValid=False),bIsDynamicallyLoaded=False,ResourceName="")
    WidgetStyle_BottomShadowBrush:Slate Brush Structure=(TintColor=(SpecifiedColor=(R=1,G=1,B=1,A=1),ColorUseRule=UseColor_Specified),DrawAs=Image,Tiling=NoTile,Mirroring=NoMirror,ImageType=NoImage,ImageSize=(X=32,Y=32),Margin=(Left=0,Top=0,Right=0,Bottom=0),ResourceObject=None,OutlineSettings=(CornerRadii=(X=0,Y=0,Z=0,W=0),Color=(SpecifiedColor=(R=0,G=0,B=0,A=0),ColorUseRule=UseColor_Specified),Width=0,RoundingType=HalfHeightRadius,bUseBrushTransparency=False),UVRegion=(Min=(X=0,Y=0),Max=(X=0,Y=0),bIsValid=False),bIsDynamicallyLoaded=False,ResourceName="")
    WidgetStyle_LeftShadowBrush_TintColor:Slate Color Structure=(SpecifiedColor=(R=1,G=1,B=1,A=1),ColorUseRule=UseColor_Specified)
    WidgetStyle_LeftShadowBrush_DrawAs:ESlateBrushDrawType Enum=Image
    WidgetStyle_LeftShadowBrush_Tiling:ESlateBrushTileType Enum=NoTile
    WidgetStyle_LeftShadowBrush_Mirroring:ESlateBrushMirrorType Enum=NoMirror
    WidgetStyle_LeftShadowBrush_ImageSize:Vector 2D Structure=(X=32,Y=32)
    WidgetStyle_LeftShadowBrush_Margin:Margin Structure=(Left=0,Top=0,Right=0,Bottom=0)
    WidgetStyle_LeftShadowBrush_ResourceObject:Object Reference=/Locker/MPL/Materials/M_UI_ItemsFade_Left
    WidgetStyle_LeftShadowBrush_OutlineSettings:Slate Brush Outline Settings Structure=(CornerRadii=(X=0,Y=0,Z=0,W=0),Color=(SpecifiedColor=(R=0,G=0,B=0,A=0),ColorUseRule=UseColor_Specified),Width=0,RoundingType=HalfHeightRadius,bUseBrushTransparency=False)
    WidgetStyle_RightShadowBrush_TintColor:Slate Color Structure=(SpecifiedColor=(R=1,G=1,B=1,A=1),ColorUseRule=UseColor_Specified)
    WidgetStyle_RightShadowBrush_DrawAs:ESlateBrushDrawType Enum=Image
    WidgetStyle_RightShadowBrush_Tiling:ESlateBrushTileType Enum=NoTile
    WidgetStyle_RightShadowBrush_Mirroring:ESlateBrushMirrorType Enum=NoMirror
    WidgetStyle_RightShadowBrush_ImageSize:Vector 2D Structure=(X=32,Y=32)
    WidgetStyle_RightShadowBrush_Margin:Margin Structure=(Left=0,Top=0,Right=0,Bottom=0)
    WidgetStyle_RightShadowBrush_ResourceObject:Object Reference=/Locker/MPL/Materials/M_UI_ItemsFade_Right
    WidgetStyle_RightShadowBrush_OutlineSettings:Slate Brush Outline Settings Structure=(CornerRadii=(X=0,Y=0,Z=0,W=0),Color=(SpecifiedColor=(R=0,G=0,B=0,A=0),ColorUseRule=UseColor_Specified),Width=0,RoundingType=HalfHeightRadius,bUseBrushTransparency=False)
    WidgetStyle_HorizontalScrolledContentPadding:Margin Structure=(Left=0,Top=0,Right=1,Bottom=0)
    WidgetStyle_VerticalScrolledContentPadding:Margin Structure=(Left=0,Top=0,Right=0,Bottom=1)
    self:Scroll Box Object Reference=Get ScrollBox_Tabs!(self:WBP UIKit Tab Group Base Object Reference=Get Tabs_Subcategory()) -> (ScrollBox_Tabs:Scroll Box Object Reference)
   ) -> (then, Output_Get:Scroll Box Style Structure)
   n23 = FindParentWidgetOfType(execute=@n25.then, StartingWidget=@n20, Type=&WBP_TournamentDetails_C)
   > n24 = SetVisibility(self=@n23, InVisibility=SelfHitTestInvisible)
   n25 = Set bIsEnabled(execute=@n19.then, bIsEnabled=true)
  ]}
 IsShowTabs {nodes [
   IsShowTabs() {is=title K2Node_FunctionEntry}
   n1 = Get TotalPayout Thresholds Data() -> (TotalPayout Thresholds Data:Array of Tournament Payout Threshold Data Structures)
   > n2 = Branch(Condition="integer > integer"(A=Length(TargetArray:Array of Tournament Payout Threshold Data Structures=@n1), B=1))
   > n3 = Construct Object from Class(Class=/Script/FortniteUI.UIKitSelectableButtonViewModel) -> (ReturnValue:UI Kit Selectable Button View Model Object Reference)
   > Set Label!(execute, Label:Text="LOCTEXT(\"All\")", self:UI Kit Selectable Button View Model Object Reference=@n3) -> (then, Output_Get:Text)
   > AddTab!(
    execute
    self:WBP UIKit Tab Group Base Object Reference=Get Tabs_Subcategory()
    Tab VM:MVVM Base Viewmodel Object Reference=@n3
    Tab ID:Name=All
    Index:int=-1
   ) -> (then)
   > SetActiveWidget(
    self:Widget Switcher Object Reference=Get HeaderWidgetSwitcher() -> (HeaderWidgetSwitcher:Common Activatable Widget Switcher Object Reference)
    Widget=Get SubCategoryCarousel_Root() -> (SubCategoryCarousel_Root:Grid Panel Object Reference)
   )
   n4 = Get TotalPayout Thresholds Data()
   "SetText (Text)"(
    execute=@n2.else
    self=Get Text_Header() -> (Text_Header:Common Text Object Reference)
    InText=Break Tournament Payout Threshold Data!(TournamentPayoutThresholdData:"Tournament Payout Threshold Data Structure (by ref)"="Get (a copy)"(Array:Array of Tournament Payout Threshold Data Structures=@n4) -> (Output:Tournament Payout Threshold Data Structure)) -> (ThresholdDisplayText:Text, ThresholdDisplayTextHeader:Text, SourceLeaderboardDisplayName:Text) {is=title K2Node_BreakStruct}
   )
   n5 = Length(TargetArray:Array of Tournament Reward Item VM Object References=Get Prizes Items VM() -> (Prizes Items VM:Array of Tournament Reward Item VM Object References))
   > "SetText (Text)"(
    self=Get Text_SubHeader() -> (Text_SubHeader:Common Text Object Reference)
    InText=Select(Option 0:Text="LOCTEXT(\"prize\")", Option 1:Text="LOCTEXT(\"prizes\")", Index="integer > integer"(A=@n5, B=1)) -> (ReturnValue:Text)
   )
   > SetActiveWidget(self=Get HeaderWidgetSwitcher(), Widget=Get Grid_HeaderText() -> (Grid_HeaderText:Grid Panel Object Reference))
  ]}
 RefreshArrowsVisibility {nodes [
   RefreshArrowsVisibility() {is=title K2Node_FunctionEntry}
   n1 = "Equal (Enum)"(A=GetCurrentInputType(self=Get CommonInputSubsystem()), B=MouseAndKeyboard)
   n2 = Get ListViewPrizes()
   n3 = "integer > integer"(A=GetNumItems(self=@n2) -> (ReturnValue=0), B=8)
   > SetVisibility(
    self=Get NavigationButtons() -> (NavigationButtons:Grid Panel Object Reference)
    InVisibility=Select(Option 0:ESlateVisibility Enum=Collapsed, Option 1:ESlateVisibility Enum=SelfHitTestInvisible, Index=AND Boolean(A=@n3, B=@n1)) -> (ReturnValue:ESlateVisibility Enum)
   )
   "integer >= integer"(
    A=110
    B=Length(TargetArray:Array of User Widget Object References=GetDisplayedEntryWidgets!(self:List View Base Object Reference=@n2) -> (ReturnValue:Array of User Widget Object References))
   )
  ]}
 Set Greyed Out Rewards Items {nodes [
   n1 = Set Greyed Out Rewards Items!() -> (then, In Min:int, In Max:int) {is=title K2Node_FunctionEntry}
   > n2 = For Each Loop(Array:Array of Object References=GetListItems!(self:List View Object Reference=Get ListViewPrizes()) -> (ReturnValue:Array of Object References)) -> (Array Element:Object Reference)
   > n3 = Cast To FortTournamentRewardItemVM(Object=@n2)
   > SetbIsGreyedOut!(
    execute
    self:Tournament Reward Item VM Object Reference=@n3
    bInIsGreyedOut:bool=NOT Boolean(A="In Range (Integer)"!(Value:int=@n2.Array Index, Min:int=@n1, Max:int=@n1.In Max, InclusiveMin:bool=true, InclusiveMax:bool=false) -> (ReturnValue:bool))
   ) -> (then)
  ]}
 SetLeaderboardsPrizesData {nodes [
   n1 = SetLeaderboardsPrizesData!() -> (then, In Payouts:Array of Tournament Leaderboard Payout Table Data Structures) {is=title K2Node_FunctionEntry}
   > Clear(TargetArray:Array of Tournament Payout Threshold Data Structures=Get TotalPayout Thresholds Data())
   > n2 = Sequence!(execute) -> (then_0, then_1, then_2:Exec, then_3:Exec)
   > Clear(TargetArray:Array of Tournament Reward Item VM Object References=Get Prizes Items VM())
   > ClearListItems(self=Get ListViewPrizes())
   > Clear!(execute, TargetMap:Map of UI Kit Selectable Button View Model Object References to Integers=Get Tab Buttons to First Item Index()) -> (then) {is=title K2Node_CallFunction}
   > Clear(TargetMap=Get Tab Buttons to Last Item Index())
   > RemoveTabs(self:WBP UIKit Tab Group Base Object Reference=Get Tabs_Subcategory())
   n3 = For Each Loop(Exec=@n2.then_1, Array:Array of Tournament Leaderboard Payout Table Data Structures=@n1) -> (Array Element:Tournament Leaderboard Payout Table Data Structure)
   n4 = For Each Loop(Exec=@n3.Completed, Array:Array of Tournament Payout Threshold Data Structures=Get TotalPayout Thresholds Data()) -> (Array Element:Tournament Payout Threshold Data Structure)
   IsShowTabs(execute=@n4.Completed) {is=title K2Node_CallFunction}
   n5 = Is Empty!(TargetArray:Array of Tournament Reward Item VM Object References=Get Prizes Items VM()) -> (ReturnValue:bool)
   > SetVisibility(InVisibility=Select(Option 0=SelfHitTestInvisible, Option 1=Collapsed, Index=@n5))
   > n6 = Branch(Condition=@n5)
   Set List Items!(execute=@n6.else, self:List View Object Reference=Get ListViewPrizes(), InListItems:Array of Object References=Get Prizes Items VM()) -> (then) {is=title K2Node_CallFunction}
   n7 = Get Tabs_Subcategory()
   > SelectTabByIndex!(execute, self:UIKit Tab Group Object Reference=@n7, Index:int=0, bSuppressClickFeedback:bool=true) -> (then, ReturnValue:bool)
   n8 = "integer >= integer"(
    A=Length!(TargetMap:Map of UI Kit Selectable Button View Model Object References to Integers=Get Tab Buttons to First Item Index()) -> (ReturnValue:int=0) {is=title K2Node_CallFunction}
    B=2
   )
   > Set HidePrompts!(execute, HidePrompts:bool=NOT Boolean(A=@n8), self:WBP UIKit Tab Group Base Object Reference=@n7) -> (then, Output_Get:bool)
   > SetVisibility(
    self=[
     Get PrevButton() -> (PrevButton:WBP UIKit Button Generic Object Reference)
     Get NextButton() -> (NextButton:WBP UIKit Button Generic Object Reference)
    ]
    InVisibility=Select(Option 0=Hidden, Option 1=SelfHitTestInvisible, Index=@n8)
   )
   n9 = Break Tournament Payout Threshold Data!(TournamentPayoutThresholdData=@n4) -> (
    ScoringType:EPayoutScoringType Enum
    Threshold:double=~
    ThresholdDisplayText
    ThresholdDisplayTextHeader
    SourceLeaderboardDisplayName
    PayoutItemsData:Array of Tournament Payout Item Data Structures
    CosmeticPayoutItemsVM:Array of Fort Item VM Object References
   )
   SetPayoutTab!(
    execute=@n4.LoopBody
    self
    In Tab Label:Text=@n9.ThresholdDisplayText
    In Payouts VM:Array of Fort Item VM Object References=@n9.CosmeticPayoutItemsVM
   ) -> (then)
   Append Array!(
    execute=@n3.LoopBody
    TargetArray:Array of Tournament Payout Threshold Data Structures=Get TotalPayout Thresholds Data()
    SourceArray:Array of Tournament Payout Threshold Data Structures=Break Tournament Leaderboard Payout Table Data!(TournamentLeaderboardPayoutTableData:"Tournament Leaderboard Payout Table Data Structure (by ref)"=@n3) -> (bIsMainWindowLeaderboard:bool=~, PayoutThresholdsData:Array of Tournament Payout Threshold Data Structures) {is=title K2Node_BreakStruct}.PayoutThresholdsData
   ) -> (then)
   DelayedRefreshButtons(execute=@n2.then_2) {is=title K2Node_CallFunction}
  ]}
 SetMobileStyle {nodes [
   SetMobileStyle() {is=title K2Node_FunctionEntry}
   > Branch(Condition=IsMobileGame())
   > SetIconSize!(
    execute
    self:WBP UIKit Button Generic Base Object Reference=[Get PrevButton(), Get NextButton(), Get PreviousRewardButton(), Get NextRewardButton()]
    Icon:Vector 2D Structure=(X=80,Y=80)
   ) -> (then)
   n1 = Get Tabs_Subcategory()
   > Set!(execute, MinButtonSize_X:int=120, MinButtonSize_Y:int=80, self:WBP UIKit Tab Group Base Object Reference=@n1) -> (then, Output_Get:Int Point Structure)
   n2 = SlotAsGridSlot(Widget=@n1)
   n3 = Get!(self:Grid Slot Object Reference=@n2) -> (Padding_Left:float=~, Padding_Top:float=~, Padding_Right:float=~, Padding_Bottom:float=~)
   > SetPadding!(
    execute
    self:Grid Slot Object Reference=@n2
    InPadding_Left:float=100
    InPadding_Top:float=@n3.Padding_Top
    InPadding_Right:float=100
    InPadding_Bottom:float=@n3.Padding_Bottom
   ) -> (then)
  ]}
 SetPayoutTab {nodes [
   n1 = SetPayoutTab!() -> (then, In Tab Label:Text, In Payouts VM:Array of Fort Item VM Object References) {is=title K2Node_FunctionEntry}
   > Branch(Condition=Is Not Empty!(TargetArray:Array of Fort Item VM Object References=@n1.In Payouts VM) -> (ReturnValue:bool))
   > n2 = Sequence!(execute) -> (then_0, then_1, then_2)
   > n3 = Construct Object from Class(Class=&UIKitSelectableButtonViewModel)
   > n4 = Set New Payout Tab(New Payout Tab:UI Kit Selectable Button View Model Object Reference=@n3) -> (Output_Get:UI Kit Selectable Button View Model Object Reference)
   > Set Label(Label=Get In Tab Label() -> (In Tab Label:Text), self=@n4)
   > Add!(
    execute
    TargetMap:Map of UI Kit Selectable Button View Model Object References to Integers=Get Tab Buttons to First Item Index()
    Key:"UI Kit Selectable Button View Model Object Reference (by ref)"=@n4
    Value:"Integer (by ref)"=Length(TargetArray:Array of Tournament Reward Item VM Object References=Get Prizes Items VM()) {is=title K2Node_CallArrayFunction}
   ) -> (then) {is=title K2Node_CallFunction}
   > AddTab(self=Get Tabs_Subcategory(), Tab VM=@n4, Index=-1)
   n5 = For Each Loop(
    Exec=@n2.then_1
    Array:Array of Fort Item VM Object References=Get In Payouts VM() -> (In Payouts VM:Array of Fort Item VM Object References)
   ) -> (Array Element:Fort Item VM Object Reference)
   > n6 = Construct Object from Class(Class=/Script/CompeteUI.FortTournamentRewardItemVM) -> (ReturnValue:Tournament Reward Item VM Object Reference)
   > SetItemVM!(execute, self:Tournament Reward Item VM Object Reference=@n6, InItemVM:Fort Item VM Object Reference=@n5) -> (then)
   > Add(
    TargetArray:Array of Tournament Reward Item VM Object References=Get Prizes Items VM()
    NewItem:"Tournament Reward Item VM Object Reference (by ref)"=@n6
   ) {is=title K2Node_CallArrayFunction}
   Add(
    execute=@n2.then_2
    TargetMap=Get Tab Buttons to Last Item Index()
    Key=Get New Payout Tab() -> (New Payout Tab:UI Kit Selectable Button View Model Object Reference)
    Value=Length(TargetArray=Get Prizes Items VM())
   ) {is=title K2Node_CallFunction}
  ]}
 VisibilityTabsNavButtonsByInputType {nodes [
   n1 = VisibilityTabsNavButtonsByInputType!() -> (then, InputType:ECommonInputType Enum) {is=title K2Node_FunctionEntry}
   > RefreshArrowsVisibility() {is=title K2Node_CallFunction}
   > n2 = Branch(Condition=Select!(MouseAndKeyboard:bool=true, Gamepad:bool=false, Touch:bool=true, Count:bool=false, Index:ECommonInputType Enum=@n1) -> (ReturnValue:bool))
   n3 = Get Anim_ShowArrowButtons() -> (Anim_ShowArrowButtons:Widget Animation Object Reference)
   > PlayAnimationForward(InAnimation=@n3)
   PlayAnimationReverse(execute=@n2.else, InAnimation=@n3)
  ]}
]
variables [
 Prizes Items VM {type="array<FortTournamentRewardItemVM>", default []}
 Tab Buttons to First Item Index {default {}}
 Tab Buttons to Last Item Index {default {}}
 Target {type=WBP_CompeteRewardTile_C, default=null}
 TotalPayout Thresholds Data {type="array<TournamentPayoutThresholdData>", default []}
]
widgets {
 info {inheritedWidgetCount=0, parentClass=&CommonUserWidget, rootWidgetClass=&GridPanel, widgetCount=14}
 widgets [
  Root {widgetClassPath=&GridPanel, widget=Root, properties {columnFill [1]}}
  SizeBox_0 {
   widgetClassPath=&SizeBox
   parent=Root
   widget=SizeBox_0
   properties {bOverride_WidthOverride=true, clipping=ClipToBounds, widthOverride=1200}
   slot {class=GridSlot, properties {horizontalAlignment=HAlign_Left, row=1, padding=<ltrb 0 1 0 0>}}
  }
  ListViewPrizes {
   widgetClassPath=&CommonListView
   bIsVariable=true
   parent=SizeBox_0
   widget=ListViewPrizes
   properties {
    bAllowDragging=false
    bEnableRightClickScrolling=false
    clipping=Inherit
    consumeMouseWheel=Never
    entryWidgetClass=&/WBP_CompeteRewardTile.WBP_CompeteRewardTile_C
    horizontalEntrySpacing=24
    orientation=Orient_Horizontal
    scrollIntoViewAlignment=TopOrLeft
    wheelScrollMultiplier=0
    scrollBarPadding=<ltrb 0 22 0 0>
    scrollBarStyle {
     thickness=12
     draggedThumbImage {
      bIsDynamicallyLoaded=false
      drawAs=RoundedBox
      imageType=FullColor
      mirroring=NoMirror
      resourceName=../../../Engine/Content/Slate/Common/Scrollbar_Thumb.png
      resourceObject=null
      tiling=NoTile
      imageSize=<12 12>
      margin=<ltrb 0.25 0.25 0.25 0.25>
      outlineSettings {
       bUseBrushTransparency=false
       roundingType=FixedRadius
       width=2
       color {colorUseRule=UseColor_Specified, specifiedColor=<rgba 0 0 0 0>}
       cornerRadii=<6 6 6 6>
      }
      tintColor {colorUseRule=UseColor_Specified, specifiedColor=<rgba 1 1 1 1>}
      uVRegion {bIsValid=false, max=<0 0>, min=<0 0>}
     }
     horizontalBackgroundImage {
      bIsDynamicallyLoaded=false
      drawAs=Image
      imageType=FullColor
      mirroring=NoMirror
      resourceName=None
      resourceObject=/FortUILibrary/UIKit/ScrollBar/MaterialInstances/MI_UIKit_ScrollbarBackground
      tiling=NoTile
      imageSize=<0 0>
      margin=<ltrb 0 0 0 0>
      outlineSettings {
       bUseBrushTransparency=false
       roundingType=HalfHeightRadius
       width=0
       color {colorUseRule=UseColor_Specified, specifiedColor=<rgba 0 0 0 0>}
       cornerRadii=<0 0 0 0>
      }
      tintColor {colorUseRule=UseColor_Specified, specifiedColor=<rgba 1 1 1 1>}
      uVRegion {bIsValid=false, max=<0 0>, min=<0 0>}
     }
     horizontalBottomSlotImage {
      bIsDynamicallyLoaded=false
      drawAs=NoDrawType
      imageType=FullColor
      mirroring=NoMirror
      resourceName=../../../Engine/Content/Slate/Common/Scrollbar_Background_Horizontal.png
      resourceObject=null
      tiling=NoTile
      imageSize=<8 8>
      margin=<ltrb 0 0 0 0>
      outlineSettings {
       bUseBrushTransparency=false
       roundingType=HalfHeightRadius
       width=0
       color {colorUseRule=UseColor_Specified, specifiedColor=<rgba 0 0 0 0>}
       cornerRadii=<0 0 0 0>
      }
      tintColor {colorUseRule=UseColor_Specified, specifiedColor=<rgba 1 1 1 1>}
      uVRegion {bIsValid=false, max=<0 0>, min=<0 0>}
     }
     horizontalTopSlotImage {
      bIsDynamicallyLoaded=false
      drawAs=NoDrawType
      imageType=FullColor
      mirroring=NoMirror
      resourceName=../../../Engine/Content/Slate/Common/Scrollbar_Background_Horizontal.png
      resourceObject=null
      tiling=NoTile
      imageSize=<8 8>
      margin=<ltrb 0 0 0 0>
      outlineSettings {
       bUseBrushTransparency=false
       roundingType=HalfHeightRadius
       width=0
       color {colorUseRule=UseColor_Specified, specifiedColor=<rgba 0 0 0 0>}
       cornerRadii=<0 0 0 0>
      }
      tintColor {colorUseRule=UseColor_Specified, specifiedColor=<rgba 1 1 1 1>}
      uVRegion {bIsValid=false, max=<0 0>, min=<0 0>}
     }
     hoveredThumbImage {
      bIsDynamicallyLoaded=false
      drawAs=RoundedBox
      imageType=FullColor
      mirroring=NoMirror
      resourceName=../../../Engine/Content/Slate/Common/Scrollbar_Thumb.png
      resourceObject=null
      tiling=NoTile
      imageSize=<12 12>
      margin=<ltrb 0.25 0.25 0.25 0.25>
      outlineSettings {
       bUseBrushTransparency=false
       roundingType=FixedRadius
       width=2
       color {colorUseRule=UseColor_Specified, specifiedColor=<rgba 0 0 0 0>}
       cornerRadii=<6 6 6 6>
      }
      tintColor {colorUseRule=UseColor_Specified, specifiedColor=<rgba 1 1 1 1>}
      uVRegion {bIsValid=false, max=<0 0>, min=<0 0>}
     }
     normalThumbImage {
      bIsDynamicallyLoaded=false
      drawAs=RoundedBox
      imageType=FullColor
      mirroring=NoMirror
      resourceName=../../../Engine/Content/Slate/Common/Scrollbar_Thumb.png
      resourceObject=null
      tiling=NoTile
      imageSize=<12 12>
      margin=<ltrb 0.25 0.25 0.25 0.25>
      outlineSettings {
       bUseBrushTransparency=false
       roundingType=FixedRadius
       width=2
       color {colorUseRule=UseColor_Specified, specifiedColor=<rgba 0 0 0 0>}
       cornerRadii=<6 6 6 6>
      }
      tintColor {colorUseRule=UseColor_Specified, specifiedColor=<rgba 1 1 1 0.75>}
      uVRegion {bIsValid=false, max=<0 0>, min=<0 0>}
     }
     verticalBackgroundImage {
      bIsDynamicallyLoaded=false
      drawAs=Box
      imageType=NoImage
      mirroring=NoMirror
      resourceName=None
      resourceObject=&MI_UIKit_ScrollbarBackground
      tiling=NoTile
      imageSize=<12 12>
      margin=<ltrb 0.5 0.5 0.5 0.5>
      outlineSettings {
       bUseBrushTransparency=false
       roundingType=HalfHeightRadius
       width=0
       color {colorUseRule=UseColor_Specified, specifiedColor=<rgba 0 0 0 0>}
       cornerRadii=<0 0 0 0>
      }
      tintColor {colorUseRule=UseColor_Specified, specifiedColor=<rgba 1 1 1 1>}
      uVRegion {bIsValid=false, max=<0 0>, min=<0 0>}
     }
     verticalBottomSlotImage {
      bIsDynamicallyLoaded=false
      drawAs=NoDrawType
      imageType=FullColor
      mirroring=NoMirror
      resourceName=../../../Engine/Content/Slate/Common/Scrollbar_Background_Vertical.png
      resourceObject=null
      tiling=NoTile
      imageSize=<8 8>
      margin=<ltrb 0 0 0 0>
      outlineSettings {
       bUseBrushTransparency=false
       roundingType=HalfHeightRadius
       width=0
       color {colorUseRule=UseColor_Specified, specifiedColor=<rgba 0 0 0 0>}
       cornerRadii=<0 0 0 0>
      }
      tintColor {colorUseRule=UseColor_Specified, specifiedColor=<rgba 1 1 1 1>}
      uVRegion {bIsValid=false, max=<0 0>, min=<0 0>}
     }
     verticalTopSlotImage {
      bIsDynamicallyLoaded=false
      drawAs=NoDrawType
      imageType=FullColor
      mirroring=NoMirror
      resourceName=../../../Engine/Content/Slate/Common/Scrollbar_Background_Vertical.png
      resourceObject=null
      tiling=NoTile
      imageSize=<8 8>
      margin=<ltrb 0 0 0 0>
      outlineSettings {
       bUseBrushTransparency=false
       roundingType=HalfHeightRadius
       width=0
       color {colorUseRule=UseColor_Specified, specifiedColor=<rgba 0 0 0 0>}
       cornerRadii=<0 0 0 0>
      }
      tintColor {colorUseRule=UseColor_Specified, specifiedColor=<rgba 1 1 1 1>}
      uVRegion {bIsValid=false, max=<0 0>, min=<0 0>}
     }
    }
   }
   slot {class=SizeBoxSlot, properties {padding=<ltrb 15 32 15 8>}}
  }
  NavigationButtons {
   widgetClassPath=&GridPanel
   bIsVariable=true
   parent=Root
   widget=NavigationButtons
   properties {columnFill [1], rowFill [1]}
   slot {class=GridSlot, properties {layer=10, row=1, padding=<ltrb 0 46 0 25>}}
  }
  PreviousRewardButton {
   widgetClassPath=&WBP_UIKit_Button_Generic
   bIsVariable=true
   parent=NavigationButtons
   widget=PreviousRewardButton
   properties {
    bHideInputActionWithKeyboard=true
    bIsFocusable=false
    backgroundMaterial=/FortUILibrary/UIKit/Buttons/MaterialInstances/CTAButtons/MI_UIKit_BgRegular
    enableFocusedSelectedState=false
    icon=/FortUILibrary/Textures/Icons/ToBeDeprecated/T_UI_Icon_Global_Chevron_Left
    iconMaterial=/FortUILibrary/Textures/Material/MI_UIKit_Icon_Base
    minHeight=64
    minWidth=64
    outlineMaterial=/FortUILibrary/UIKit/Buttons/MaterialInstances/CTAButtons/MI_UIKit_ButtonOutline
    priority=100
    promptHorizontalAlignment=HAlign_Center
    useTextBlock=false
    buttonPadding=<ltrb 0 0 0 0>
    font {
     bForceMonospaced=false
     bMaterialIsStencil=false
     fontMaterial=/FortUILibrary/UIKit/Buttons/MaterialInstances/CTAButtons/MI_UIKit_TextQuiet
     fontObject=&Burbank
     letterSpacing=48
     monospacedWidth=1
     size=1
     skewAmount=0
     typefaceFontName=BurbankRegular-Black
     outlineSettings {
      bApplyOutlineToDropShadows=false
      bMiteredCorners=false
      bSeparateFillAlpha=false
      outlineMaterial=null
      outlineSize=0
      outlineColor=<rgba 0 0 0 1>
     }
    }
    iconPadding=<ltrb 4 4 4 4>
    iconSize=<20 20>
    promptNudge=<0 0>
   }
   slot {class=GridSlot, properties {horizontalAlignment=HAlign_Left, layer=10, verticalAlignment=VAlign_Center, nudge=<-70 0>}}
  }
  NextRewardButton {
   widgetClassPath=&WBP_UIKit_Button_Generic
   bIsVariable=true
   parent=NavigationButtons
   widget=NextRewardButton
   properties {
    bHideInputActionWithKeyboard=true
    bIsFocusable=false
    bShouldUseFallbackDefaultInputAction=false
    backgroundMaterial=&MI_UIKit_BgRegular
    enableFocusedSelectedState=false
    icon=/FortUILibrary/Textures/Icons/T_UI_Icon_UIKit_Chevron
    iconMaterial=&MI_UIKit_Icon_Base
    minHeight=64
    minWidth=64
    outlineMaterial=&MI_UIKit_ButtonOutline
    priority=100
    promptHorizontalAlignment=HAlign_Center
    useTextBlock=false
    buttonPadding=<ltrb 0 0 0 0>
    iconPadding=<ltrb 4 4 4 4>
    iconSize=<20 20>
    promptNudge=<0 0>
   }
   slot {class=GridSlot, properties {horizontalAlignment=HAlign_Right, layer=10, verticalAlignment=VAlign_Center, nudge=<70 0>}}
  }
  HeaderWidgetSwitcher {
   widgetClassPath=&CommonActivatableWidgetSwitcher
   bIsVariable=true
   parent=Root
   widget=HeaderWidgetSwitcher
   slot {class=GridSlot, properties {horizontalAlignment=HAlign_Left}}
  }
  SubCategoryCarousel_Root {
   widgetClassPath=&GridPanel
   bIsVariable=true
   parent=HeaderWidgetSwitcher
   widget=SubCategoryCarousel_Root
   properties {columnFill [0,1,0], rowFill [1]}
   slot {class=WidgetSwitcherSlot, properties {horizontalAlignment=HAlign_Left}}
  }
  Tabs_Subcategory {
   widgetClassPath=/FortUILibrary/UIKit/Tabs/WBP_UIKit_TabGroup
   bIsVariable=true
   parent=SubCategoryCarousel_Root
   widget=Tabs_Subcategory
   properties {
    animateReservedPromptSpace=false
    buttonClass=&/WBP_RewardsFilterTab.WBP_RewardsFilterTab_C
    initialSelectedTabIndex=0
    inputActions=Triggers
    nextPromptHorizontalAlignment=HAlign_Right
    prevPromptHorizontalAlignment=HAlign_Left
    reservedPromptSpace=0
    tabGroupHorizontalAlignment=HAlign_Fill
    tabSpacing=0
    useBackplate=false
    usePromptBackplate=false
    initialTabsData [
     {
      bEnableTabId=false
      tabId=None
      tabViewModel=&self:WidgetTree.Tabs_Subcategory.UIKitSelectableButtonViewModel_0
     }
     {
      bEnableTabId=false
      tabId=None
      tabViewModel=&self:WidgetTree.Tabs_Subcategory.UIKitSelectableButtonViewModel_2
     }
    ]
    minButtonSize=<120 48>
    nextPromptPadding=<ltrb 0 0 -32 0>
    prevPromptPadding=<ltrb -32 0 0 0>
    tabGroupPadding=<ltrb 10 0 10 0>
   }
   slot {class=GridSlot, properties {columnSpan=3, row=1, padding=<ltrb 0 0 54 0>}}
  }
  PrevButton {
   widgetClassPath=&WBP_UIKit_Button_Generic
   bIsVariable=true
   parent=SubCategoryCarousel_Root
   widget=PrevButton
   properties {
    bIsFocusable=false
    contentHorizontalAlignment=HAlign_Left
    enableFocusedSelectedState=false
    icon=&T_UI_Icon_Global_Chevron_Left
    iconMaterial=/FortUILibrary/UIKit/Buttons/MaterialInstances/GenericButtons/MI_UIKit_GenericIcon_Quiet
    minHeight=64
    minWidth=90
    text=""
    useBackgroundBlock=false
    visibility=Visible
   }
   slot {class=GridSlot, properties {horizontalAlignment=HAlign_Left, layer=1, row=1, verticalAlignment=VAlign_Center, nudge=<-54 0>}}
  }
  NextButton {
   widgetClassPath=&WBP_UIKit_Button_Generic
   bIsVariable=true
   parent=SubCategoryCarousel_Root
   widget=NextButton
   properties {
    bIsFocusable=false
    enableFocusedSelectedState=false
    icon=/FortUILibrary/Textures/Icons/ToBeDeprecated/T_UI_Icon_Global_Chevron_Right
    iconMaterial=&MI_UIKit_GenericIcon_Quiet
    minHeight=64
    minWidth=90
    text=""
    useBackgroundBlock=false
    visibility=Visible
   }
   slot {class=GridSlot, properties {column=2, horizontalAlignment=HAlign_Right, layer=1, row=1, verticalAlignment=VAlign_Center}}
  }
  Grid_HeaderText {
   widgetClassPath=&GridPanel
   bIsVariable=true
   parent=HeaderWidgetSwitcher
   widget=Grid_HeaderText
   slot {class=WidgetSwitcherSlot}
  }
  Text_Header {
   widgetClassPath=&CommonTextBlock
   bIsVariable=true
   parent=Grid_HeaderText
   widget=Text_Header
   properties {
    bAutoCollapseWithEmptyText=true
    bIsScrollingEnabled=false
    mobileFontSizeMultiplier=0.75
    textTransformPolicy=ToUpper
    wrapTextAt=510
    colorAndOpacity {colorUseRule=UseColor_Specified, specifiedColor=<rgba 1 1 1 0>}
    font {
     bForceMonospaced=false
     bMaterialIsStencil=false
     fontMaterial=null
     fontObject=&HeadingNow
     letterSpacing=80
     monospacedWidth=1
     size=41.25
     skewAmount=0
     typefaceFontName=86 Bold Italic
     outlineSettings {
      bApplyOutlineToDropShadows=false
      bMiteredCorners=false
      bSeparateFillAlpha=false
      outlineMaterial=/Compete/MaterialInstances/MI_TournamentPrizeTextOutline
      outlineSize=1
      outlineColor=<rgba 1 1 1 1>
     }
    }
    shadowOffset=<0 0>
   }
   slot {class=GridSlot, properties {horizontalAlignment=HAlign_Left, layer=1, row=1, rowSpan=1, nudge=<0 30>, padding=<ltrb 0 -30 0 0>}}
  }
  Text_SubHeader {
   widgetClassPath=&CommonTextBlock
   bIsVariable=true
   parent=Grid_HeaderText
   widget=Text_SubHeader
   properties {
    bAutoCollapseWithEmptyText=true
    bIsScrollingEnabled=false
    mobileFontSizeMultiplier=0.75
    text=Prize
    textTransformPolicy=ToUpper
    wrapTextAt=510
    font {
     bForceMonospaced=false
     bMaterialIsStencil=false
     fontMaterial=null
     fontObject=&HeadingNow
     letterSpacing=-10
     monospacedWidth=1
     size=32.25
     skewAmount=0
     typefaceFontName=86 Bold Italic
     outlineSettings {
      bApplyOutlineToDropShadows=false
      bMiteredCorners=false
      bSeparateFillAlpha=false
      outlineMaterial=null
      outlineSize=0
      outlineColor=<rgba 0 0 0 1>
     }
    }
    shadowOffset=<0 0>
   }
   slot {class=GridSlot, properties {horizontalAlignment=HAlign_Left, layer=1, row=2, rowSpan=1}}
  }
 ]
}
