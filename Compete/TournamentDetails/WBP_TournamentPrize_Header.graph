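# Header widget for tournament prize details, showing an outlined placement title above a subheader.
# SetPayouts collapses it, then switches on the view model's header reward type, revealing it only for Commerce prizes or Token next-round advancement without division labels.
# Commerce threshold 1 shows "1st place", with any leaderboard name; other cases use threshold text.
# SetHeader enlarges a leading digit through rich-text markup unless told to ignore that adjustment.
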WidgetBlueprint /Compete/TournamentDetails/WBP_TournamentPrize_Header
graphs [
 EventGraph {nodes [
   n1 = Event PreConstruct()
   > Branch(Condition=@n1.IsDesignTime)
   > SetFirstPlayerReward!(execute, self, PayoutThreshold:"Tournament Payout Threshold Data Structure (by ref)") -> (then)
  ]}
 IsDisplayingTokenHeader {nodes [
   IsDisplayingTokenHeader() {is=title K2Node_FunctionEntry}
   > Return Node!(
    execute
    IsDisplayingTokenHeader:bool="Equal (Enum)"(
     A:EPayoutRewardType Enum=Get CurrentRewardDisplayType() -> (CurrentRewardDisplayType:EPayoutRewardType Enum=Invalid)
     B:EPayoutRewardType Enum=Token
    )
   )
  ]}
 OnFocused {nodes [OnFocused() {is=title K2Node_FunctionEntry}]}
 SetFirstPlayerReward {nodes [
   n1 = SetFirstPlayerReward!() -> (then, PayoutThreshold:"Tournament Payout Threshold Data Structure (by ref)") {is=title K2Node_FunctionEntry}
   n2 = Break Tournament Payout Threshold Data!(TournamentPayoutThresholdData:"Tournament Payout Threshold Data Structure (by ref)"=@n1) -> (
    ScoringType:EPayoutScoringType Enum
    Threshold:double=~
    ThresholdDisplayText:Text
    ThresholdDisplayTextHeader:Text
    SourceLeaderboardDisplayName:Text
    PayoutItemsData:Array of Tournament Payout Item Data Structures
    CosmeticPayoutItemsVM:Array of Fort Item VM Object References
   ) {is=title K2Node_BreakStruct}
   > SetHeader!(
    execute
    self
    In Text:Text=SelectText(
     A="LOCTEXT(\"1st place\")"
     B=Format Text!(Format="LOCTEXT(\"1st place <SmallerTextStyle>in {LbName}</>\")", LbName:Text=@n2.SourceLeaderboardDisplayName) -> (Result)
     bPickA=TextIsEmpty(InText=@n2.SourceLeaderboardDisplayName)
    )
    IgnoreHeaderAdjustment:bool=true
   ) -> (then)
   > SetSubHeader!(execute, self, Text:Text="LOCTEXT(\"prize\")") -> (then)
  ]}
 SetHeader {nodes [
   n1 = SetHeader!() -> (then, In Text:Text, IgnoreHeaderAdjustment:bool=~) {is=title K2Node_FunctionEntry}
   > n2 = Branch(Condition=TextIsEmpty(InText=@n1))
   > SetVisibility(self=Get CommonRichText_Header() -> (CommonRichText_Header:Common Rich Text Block Object Reference), InVisibility=Hidden)
   SetVisibility(execute=@n2.else, self=Get CommonRichText_Header(), InVisibility=SelfHitTestInvisible)
   n3 = "To String (Text)"(InText=@n1)
   n4 = GetSubstring!(SourceString:String=@n3, StartIndex:int=0, Length:int=1) -> (ReturnValue:String)
   n5 = Get IgnoreHeaderAdjustment() -> (IgnoreHeaderAdjustment:bool)
   > n6 = Branch(Condition=AND Boolean(A=IsNumeric!(SourceString:String=@n4) -> (ReturnValue:bool), B=NOT Boolean(A=@n5)))
   > SetText(
    self=Get CommonRichText_Header()
    InText=Format Text!(
     Format="LOCTEXT(\"<text size=\\\"70\\\">{Position}</>{Text}\")"
     Position:String=@n4
     Text:String=Split(SourceString=@n3, InStr=@n4, SearchDir=FromStart).RightS
    ) -> (Result)
   )
   SetText(execute=@n6.else, self=Get CommonRichText_Header(), InText=@n1)
  ]}
 SetPayouts {nodes [
   SetPayouts!() -> (then, Header Threshold:Tournament Payout Threshold Data Structure) {is=title K2Node_FunctionEntry}
   > SetVisibility(InVisibility=Collapsed)
   n1 = Get FortTournamentViewVM() -> (FortTournamentViewVM:Tournament View VM Object Reference)
   > n2 = Set CurrentRewardDisplayType(CurrentRewardDisplayType:EPayoutRewardType Enum=Break Tournament Payout Item Data!(TournamentPayoutItemData:"Tournament Payout Item Data Structure (by ref)"=Get HeaderPayoutItemData!(self:Tournament View VM Object Reference=@n1) -> (HeaderPayoutItemData:Tournament Payout Item Data Structure)) -> (RewardType:EPayoutRewardType Enum, DisplayText:Text) {is=title K2Node_BreakStruct}) -> (Output_Get:EPayoutRewardType Enum=Invalid)
   > n3 = Switch on EPayoutRewardType!(execute, Selection:EPayoutRewardType Enum=@n2) -> (Invalid:Exec, Commerce:Exec, GameRelated:Exec, Token:Exec, Score:Exec) {is=title K2Node_SwitchEnum}
   n4 = Get Header Threshold() -> (Header Threshold:Tournament Payout Threshold Data Structure)
   n5 = Break Tournament Payout Threshold Data(TournamentPayoutThresholdData=@n4) -> (Threshold=~)
   n6 = Branch(execute=@n3.Commerce, Condition="Equal (Float)"(A=@n5.Threshold, B=1))
   > n7 = SetFirstPlayerReward(PayoutThreshold=@n4) {is=title K2Node_CallFunction}
   SetVisibility(execute=[@n8.then, @n7.then], InVisibility=SelfHitTestInvisible)
   n8 = SetTopPlayers!(execute=@n6.else, self, ThresholdText:Text=@n5.ThresholdDisplayTextHeader, SubHeader:Text="LOCTEXT(\"prize\")") -> (then)
   n9 = Branch(
    execute=@n3.Token
    Condition=ShowNextRoundTokensByDivisionLabels!(self:Fort Compete UIManager Object Reference=Get FortCompeteUIManager() -> (ReturnValue:Fort Compete UIManager Object Reference) {is=title K2Node_GetSubsystem}) -> (ReturnValue:bool)
   )
   SetTopPlayers(execute=@n9.else, ThresholdText=@n5.ThresholdDisplayTextHeader, SubHeader="LOCTEXT(\"Advance to next Round\")")
   > SetVisibility(InVisibility=SelfHitTestInvisible)
  ]}
 SetSubHeader {nodes [
   n1 = SetSubHeader!() -> (then, Text:Text) {is=title K2Node_FunctionEntry}
   > Set Text(Text=@n1, self=Get Text_SubHeader() -> (Text_SubHeader:Common Text Object Reference))
  ]}
 SetTopPlayers {nodes [
   n1 = SetTopPlayers!() -> (then, ThresholdText:Text, SubHeader:Text) {is=title K2Node_FunctionEntry}
   > SetSubHeader(Text=@n1.SubHeader) {is=title K2Node_CallFunction}
   > SetHeader(In Text=@n1, IgnoreHeaderAdjustment=false) {is=title K2Node_CallFunction}
   > ShowTitle!(execute, self, Show:bool=true) -> (then)
  ]}
 ShowHeader {nodes [
   n1 = ShowHeader!() -> (then, Show:bool=~) {is=title K2Node_FunctionEntry}
   > SetVisibility(self=Get CommonRichText_Header(), InVisibility=Select(Option 0=Hidden, Option 1=SelfHitTestInvisible, Index=@n1))
   > SetVisibility(self=Get Text_SubHeader(), InVisibility=SelfHitTestInvisible)
  ]}
 ShowTitle {nodes [
   n1 = ShowTitle!() -> (then, Show:bool=~) {is=title K2Node_FunctionEntry}
   > SetVisibility(self=[Get CommonRichText_Header(), Get Text_SubHeader()], InVisibility=Select(Option 0=Collapsed, Option 1=SelfHitTestInvisible, Index=@n1))
  ]}
]
variables [
 CurrentRewardDisplayType {type=EPayoutRewardType, default=Invalid}
 OnFocused {}
]
widgets {
 info {inheritedWidgetCount=0, parentClass=&CommonUserWidget, rootWidgetClass=&GridPanel, widgetCount=3}
 widgets [
  Root {widgetClassPath=&GridPanel, widget=Root, properties {columnFill [1], rowFill [0,0,0,1]}}
  CommonRichText_Header {
   widgetClassPath=&CommonRichTextBlock
   bIsVariable=true
   parent=Root
   widget=CommonRichText_Header
   properties {
    applyLineHeightToBottomLine=false
    bAutoCollapseWithEmptyText=true
    bOverrideDefaultStyle=true
    lineHeightPercentage=0.7
    scrollStyle=/FortUILibrary/Styles/Typography/Foundation/TextScrollStyle-NoFade.TextScrollStyle-NoFade_C
    text="1st place\r\n<PayoutsLeaderboardNameTextStyle>test name</>"
    textStyleSet=/Compete/DataTables/DT_CompeteUIRichTextStyles
    textTransformPolicy=ToUpper
    defaultTextStyleOverride {
     overflowPolicy=Clip
     transformPolicy=ToUpper
     colorAndOpacity {colorUseRule=UseColor_Specified, specifiedColor=<rgba 0 0 0 0>}
     font {
      bForceMonospaced=false
      bMaterialIsStencil=false
      fontMaterial=null
      fontObject=&HeadingNow
      letterSpacing=80
      monospacedWidth=1
      size=41.25
      skewAmount=0
      typefaceFontName=86 Bold Italic
      outlineSettings {
       bApplyOutlineToDropShadows=false
       bMiteredCorners=true
       bSeparateFillAlpha=false
       outlineMaterial=/Compete/MaterialInstances/MI_TournamentPrizeTextOutline
       outlineSize=1
       outlineColor=<rgba 1 1 1 1>
      }
     }
     highlightColor {colorUseRule=UseColor_Specified, specifiedColor=<rgba 0 0 0 1>}
     highlightShape {
      bIsDynamicallyLoaded=false
      drawAs=Image
      imageType=NoImage
      mirroring=NoMirror
      resourceName=None
      resourceObject=null
      tiling=NoTile
      imageSize=<32 32>
      margin=<ltrb 0 0 0 0>
      outlineSettings {
       bUseBrushTransparency=false
       roundingType=HalfHeightRadius
       width=0
       color {colorUseRule=UseColor_Specified, specifiedColor=<rgba 0 0 0 0>}
       cornerRadii=<0 0 0 0>
      }
      tintColor {colorUseRule=UseColor_Specified, specifiedColor=<rgba 1 1 1 1>}
      uVRegion {bIsValid=false, max=<0 0>, min=<0 0>}
     }
     selectedBackgroundColor {colorUseRule=UseColor_Foreground, specifiedColor=<rgba 1 0 1 1>}
     shadowColorAndOpacity=<rgba 0 0 0 1>
     shadowOffset=<0 0>
     strikeBrush {
      bIsDynamicallyLoaded=false
      drawAs=Image
      imageType=NoImage
      mirroring=NoMirror
      resourceName=None
      resourceObject=null
      tiling=NoTile
      imageSize=<32 32>
      margin=<ltrb 0 0 0 0>
      outlineSettings {
       bUseBrushTransparency=false
       roundingType=HalfHeightRadius
       width=0
       color {colorUseRule=UseColor_Specified, specifiedColor=<rgba 0 0 0 0>}
       cornerRadii=<0 0 0 0>
      }
      tintColor {colorUseRule=UseColor_Specified, specifiedColor=<rgba 1 1 1 1>}
      uVRegion {bIsValid=false, max=<0 0>, min=<0 0>}
     }
     underlineBrush {
      bIsDynamicallyLoaded=false
      drawAs=Image
      imageType=NoImage
      mirroring=NoMirror
      resourceName=None
      resourceObject=null
      tiling=NoTile
      imageSize=<32 32>
      margin=<ltrb 0 0 0 0>
      outlineSettings {
       bUseBrushTransparency=false
       roundingType=HalfHeightRadius
       width=0
       color {colorUseRule=UseColor_Specified, specifiedColor=<rgba 0 0 0 0>}
       cornerRadii=<0 0 0 0>
      }
      tintColor {colorUseRule=UseColor_Specified, specifiedColor=<rgba 1 1 1 1>}
      uVRegion {bIsValid=false, max=<0 0>, min=<0 0>}
     }
    }
   }
   slot {class=GridSlot, properties {columnSpan=1, rowSpan=1}}
  }
  Text_SubHeader {
   widgetClassPath=&CommonTextBlock
   bIsVariable=true
   parent=Root
   widget=Text_SubHeader
   properties {
    bAutoCollapseWithEmptyText=true
    bIsScrollingEnabled=false
    lineHeightPercentage=0.75
    mobileFontSizeMultiplier=0.75
    text="[PH] Subheader"
    textTransformPolicy=ToUpper
    font {
     bForceMonospaced=false
     bMaterialIsStencil=false
     fontMaterial=null
     fontObject=&HeadingNow
     letterSpacing=-40
     monospacedWidth=1
     size=32.25
     skewAmount=0
     typefaceFontName=86 Bold Italic
     outlineSettings {
      bApplyOutlineToDropShadows=false
      bMiteredCorners=false
      bSeparateFillAlpha=false
      outlineMaterial=null
      outlineSize=0
      outlineColor=<rgba 0 0 0 1>
     }
    }
    shadowOffset=<0 0>
   }
   slot {class=GridSlot, properties {columnSpan=1, layer=1, row=1, rowSpan=1, padding=<ltrb 0 -30 0 0>}}
  }
 ]
}
