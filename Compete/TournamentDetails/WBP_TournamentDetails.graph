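# Tournament details screen in the Compete UI: shows header, prize, detail and scoring sections, plus play, bookmark, leaderboard, watch-live and eligibility buttons.
# Tournament view models feed the widgets; prize sections are created lazily into named slots when data refreshes.
# Buttons push region-warning, eligibility and watch-live modals or update bookmarks.
# Custom navigation skips hidden sections, while input-type handlers adjust focus and scrolling, with separate mobile layout tweaks.
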WidgetBlueprint /Compete/TournamentDetails/WBP_TournamentDetails
graphs [
 BP_GetDesiredFocusTarget {nodes [
   Get Desired Focus Target()
   > Return Node!(execute, ReturnValue:Widget Object Reference=GetFirstButton!(self) -> (Widget:Widget Object Reference))
  ]}
 DegugPassedRequirements {nodes [
   n1 = DegugPassedRequirements!() -> (then, Passed:bool=~, Index:EFortTournamentRequirementCheckType Enum=Invalid) {is=title K2Node_FunctionEntry}
   > Return Node(ReturnValue:bool=Select!(
     Invalid:bool=false
     SystemFeatureFlag_CryptoProcessor:bool=@n1
     SystemFeatureFlag_SecureBoot:bool=@n1
     SystemFeatureFlag_KernelCodeIntegrity:bool=@n1
     SystemFeatureFlag_IOMMU:bool=@n1
     SystemFeatureFlag_StableOSVersion:bool=@n1
     SystemFeatureFlag_StrictInjectionPolicy:bool=@n1
     SystemFeatureFlag_StrictVirtualizationPolicy:bool=@n1
     SystemFeatureFlag_AndroidSecureDevice:bool=false
     SystemFeatureFlag_IOSSecureDevice:bool=false
     SystemFeatureFlag_IOSMinimumOSVersion:bool=false
     AccountLevel:bool=@n1
     MultiFactorAuthentication:bool=@n1
     HabaneroRank:bool=@n1
     QualifyingRound:bool=@n1
     EULA:bool=true
     TotalNumberOfPlayedMatches:bool=@n1
     SeasonNumberOfPlayedMatches:bool=@n1
     Age:bool=@n1
     RegionLock:bool=@n1
     AccountLock:bool=@n1
     TeamLock:bool=@n1
     TokenRestrictions:bool=@n1
     SuspiciousBehavior:bool=@n1
     PermanentTournamentBan:bool=false
     InputType:bool=false
     TimeAlive:bool=false
     PowerRankings:bool=false
     Index:EFortTournamentRequirementCheckType Enum=@n1.Index
    ) -> (ReturnValue:bool))
  ]}
 DoCustomNavigation_AdvancedTokens {nodes [
   n1 = DoCustomNavigation_AdvancedTokens!() -> (then, Navigation:EUINavigation Enum) {is=title K2Node_FunctionEntry}
   > n2 = Switch on EUINavigation!(execute, Selection:EUINavigation Enum=@n1) -> (Left:Exec, Right:Exec, Up:Exec, Down:Exec, Next:Exec, Previous:Exec, Invalid:Exec) {is=title K2Node_SwitchEnum}
   n3 = Get Cash Prizes Section() -> (Cash Prizes Section:WBP Tournament Details Section Prizes Object Reference)
   n4 = Get Cosmetic Prizes Carousel() -> (Cosmetic Prizes Carousel:WBP Tournament Prize Carousel Object Reference)
   n5 = IsVisible(self=@n4)
   n6 = GetFirstButton()
   n7 = Get ListViewPrizes!(self:WBP Tournament Prize Carousel Object Reference=@n4) -> (ListViewPrizes:Common List View Object Reference)
   Return Node(
    execute=@n2.Up
    ReturnValue:Widget Object Reference=Select!(
     Option 0:Widget Object Reference=Select(Option 0=@n6, Option 1=@n7, Index=@n5)
     Option 1:Widget Object Reference=@n3
     Index:bool=IsVisible(self=@n3)
    ) -> (ReturnValue:Widget Object Reference)
   )
  ]}
 DoCustomNavigation_CashSection {nodes [
   n1 = DoCustomNavigation_CashSection!() -> (then, Navigation:EUINavigation Enum) {is=title K2Node_FunctionEntry}
   > n2 = Switch on EUINavigation(Selection=@n1)
   n3 = Get Advance Tokens Prize Section() -> (Advance Tokens Prize Section:WBP Tournament Details Section Advancement Object Reference)
   Return Node(
    execute=@n2.Down
    ReturnValue=Select(
     Option 0=Get Section_Details() -> (Section_Details:WBP Tournament Details Section Details Object Reference)
     Option 1=@n3
     Index=IsVisible(self=@n3)
    )
   )
   n4 = Get Cosmetic Prizes Carousel()
   Return Node(execute=@n2.Up, ReturnValue=Select(Option 0=GetFirstButton(), Option 1=Get ListViewPrizes(self=@n4), Index=IsVisible(self=@n4)))
  ]}
 DoCustomNavigation_Details {nodes [
   n1 = DoCustomNavigation_Details!() -> (then, Navigation:EUINavigation Enum) {is=title K2Node_FunctionEntry}
   > n2 = Switch on EUINavigation(Selection=@n1)
   n3 = Get Section_Scoring() -> (Section_Scoring:WBP Tournament Details Section Scoring Object Reference)
   Return Node(execute=@n2.Down, ReturnValue=Select(Option 0=Get Section_Details(), Option 1=@n3, Index=IsVisible(self=@n3)))
   n4 = Get Advance Tokens Prize Section()
   n5 = Get Cash Prizes Section()
   n6 = IsVisible(self=@n5)
   n7 = Get Cosmetic Prizes Carousel()
   n8 = Select(Option 0=GetFirstButton(), Option 1=Get ListViewPrizes(self=@n7), Index=IsVisible(self=@n7))
   Return Node(execute=@n2.Up, ReturnValue=Select(Option 0=Select(Option 0=@n8, Option 1=@n5, Index=@n6), Option 1=@n4, Index=IsVisible(self=@n4)))
  ]}
 DoCustomNavigation_ItemCarousel {nodes [
   n1 = DoCustomNavigation_ItemCarousel!() -> (then, Navigation:EUINavigation Enum) {is=title K2Node_FunctionEntry}
   > n2 = Switch on EUINavigation(Selection=@n1)
   n3 = Get Cash Prizes Section()
   n4 = Get Advance Tokens Prize Section()
   n5 = IsVisible(self=@n4)
   n6 = Get Section_Details()
   Return Node(execute=@n2.Down, ReturnValue=Select(Option 0=Select(Option 0=@n6, Option 1=@n4, Index=@n5), Option 1=@n3, Index=IsVisible(self=@n3)))
   n7 = GetFirstButton()
   ScrollWidgetIntoView!(
    execute=@n2.Up
    self:Scroll Box Object Reference=Get AthenaScrollBox_Content() -> (AthenaScrollBox_Content:Athena Scroll Box Object Reference)
    WidgetToFind:Widget Object Reference=@n7
    AnimateScroll:bool=true
    ScrollDestination:EDescendantScrollDestination Enum=IntoView
    Padding:float=0
   ) -> (then)
   > Return Node(ReturnValue=@n7)
  ]}
 EventGraph {nodes [
   "On Context Menu Closed (ContextMenu_Button)"() {is=title K2Node_ComponentBoundEvent}
   > SetScrollWhenFocusChanges!(
    execute
    self:Scroll Box Object Reference=Get AthenaScrollBox_Content()
    NewScrollWhenFocusChanges:EScrollWhenFocusChanges Enum=AnimatedScroll
   ) -> (then)
   n1 = Get AthenaScrollBox_Content()
   > Set ConsumeMouseWheel!(execute, ConsumeMouseWheel:EConsumeMouseWheel Enum=WhenScrollingPossible, self:Scroll Box Object Reference=@n1) -> (then, Output_Get:EConsumeMouseWheel Enum=WhenScrollingPossible)
   > Set bEnableTouchScrolling!(execute, bEnableTouchScrolling:bool=true, self:Scroll Box Object Reference=@n1) -> (then, Output_Get:bool)
   n2 = "On Hovered (Bookmark_Button)"!() -> (OutputDelegate, then, Button:Common Button Base Object Reference) {is=title K2Node_ComponentBoundEvent}
   Branch(execute=[@n2.then, @n8.then], Condition=IsGamepad!() -> (IsGamepad:bool))
   > ScrollToStart(self:Scroll Box Object Reference=Get AthenaScrollBox_Content())
   "Button Other Action Clicked (CoachMark)"() {is=title K2Node_ComponentBoundEvent}
   n3 = Get CoachMark() -> (CoachMark:WBP UIKit Coach Mark Object Reference)
   > DeactivateWidget(self=@n3)
   > RemoveFromParent(self=@n3)
   > Setup Initial Focus For Devices()
   n4 = "On Clicked (Back_Button)"!() -> (OutputDelegate, then, Button:Common Button Base Object Reference) {is=title K2Node_ComponentBoundEvent}
   > Fire Event Element Interacted(Element=@n4.Button)
   > DeactivateWidget()
   n5 = "On Clicked (SeriesLeaderboard_Button)"!() -> (OutputDelegate, then, Button:Common Button Base Object Reference) {is=title K2Node_ComponentBoundEvent}
   > Fire Event Element Interacted(Element=@n5.Button)
   > ShowTournamentLeaderboardScreen!(execute, self, IsTournamentSeriesPointLeaderboard:bool=true) -> (then)
   "On Context Menu Opened (ContextMenu_Button)"() {is=title K2Node_ComponentBoundEvent}
   > Fire Event Element Interacted(Element=Get ContextMenu_Button() -> (ContextMenu_Button:WBP UIKit Context Menu Anchor Button Object Reference))
   n6 = Get AthenaScrollBox_Content()
   > ScrollToStart(self=@n6)
   > SetScrollWhenFocusChanges(self=@n6, NewScrollWhenFocusChanges=NoScroll)
   n7 = Get AthenaScrollBox_Content()
   > Set ConsumeMouseWheel(ConsumeMouseWheel=Never, self=@n7) -> (Output_Get=WhenScrollingPossible)
   > Set bEnableTouchScrolling(bEnableTouchScrolling=false, self=@n7)
   n8 = "On Hovered (ContextMenu_Button)"() {is=title K2Node_ComponentBoundEvent}
   n9 = "On Clicked (Bookmark_Button)"!() -> (OutputDelegate, then, Button:Common Button Base Object Reference) {is=title K2Node_ComponentBoundEvent}
   > Fire Event Element Interacted(Element=@n9.Button)
   > SetCurrentWindowBookmark!(
    execute
    self:Tournament View VM Object Reference=Get FortTournamentViewVM() -> (FortTournamentViewVM:Tournament View VM Object Reference)
    bValue:bool=NOT Boolean(A=Get IsBookmarked() -> (IsBookmarked:bool))
   ) -> (then)
   "On Clicked (Section_Scoring)"!() -> (OutputDelegate, then, Button:Common Button Base Object Reference) {is=title K2Node_ComponentBoundEvent}
   > n10 = Branch(Condition=IsGamepad())
   ScrollToEnd(execute=[@n11.then, @n10.else], self:Scroll Box Object Reference=Get AthenaScrollBox_Content())
   "On Hovered (Section_Scoring)"!() -> (OutputDelegate, then, Button:Common Button Base Object Reference) {is=title K2Node_ComponentBoundEvent}
   > n11 = Branch(Condition=IsGamepad())
   n12 = "On Clicked (Play_Button)"!() -> (OutputDelegate, then, Button:Common Button Base Object Reference) {is=title K2Node_ComponentBoundEvent}
   > Fire Event Element Interacted(Element=@n12.Button)
   n13 = Get FortCompeteRegionSubsystem() -> (ReturnValue:Fort Compete Region Subsystem Object Reference) {is=title K2Node_GetSubsystem}
   > n14 = Branch(Condition=CheckEventWillChangeTheProfileRegion!(self:Fort Compete Region Subsystem Object Reference=@n13) -> (ReturnValue:bool))
   n15 = Get FortTournamentViewVM()
   > n16 = Create Widget!(
    execute:Exec
    Class:User Widget Class Reference=&/WBP_ChangeRegionWarningModal.WBP_ChangeRegionWarningModal_C
    OwningPlayer:Player Controller Object Reference=GetOwningPlayer()
    CurrentRegionName:Text=GetCurrentProfileRegionName!(self:Fort Compete Region Subsystem Object Reference=@n13) -> (ReturnValue:Text)
    NewRegionName:Text=Get RegionName!(self:Tournament View VM Object Reference=@n15) -> (RegionName:Text)
   ) -> (then:Exec, ReturnValue:WBP Change Region Warning Modal Object Reference)
   n17 = OnDialogAccepted() {is=title K2Node_CustomEvent}
   Bind Event to Dialog Accepted(execute=@n16.then, self:WBP Change Region Warning Modal Object Reference=@n16, Delegate=@n17) {is=title K2Node_AddDelegate}
   > TryPushWidgetAsModalPanel!(execute, PushWidget:User Widget Object Reference=@n16) -> (then)
   > ActivateWidget()
   n18 = GoToLobbyWithTournament!(execute=@n14.else, self:Tournament View VM Object Reference=@n15, bSwitchProfileRegion:bool=false) -> (then)
   DeactivateWidget(execute=[@n42.then, @n18.then])
   n19 = "On Clicked (Leaderboard_Button)"!() -> (OutputDelegate, then, Button:Common Button Base Object Reference) {is=title K2Node_ComponentBoundEvent}
   > Fire Event Element Interacted(Element=@n19.Button)
   > ShowTournamentLeaderboardScreen(IsTournamentSeriesPointLeaderboard=false)
   n20 = "On Clicked (WBP_TournamentDetails_EligibilityButton)"!() -> (OutputDelegate, then, Button:Common Button Base Object Reference) {is=title K2Node_ComponentBoundEvent}
   > Fire Event Element Interacted(Element=@n20.Button)
   n21 = Get EligibilityWidget() -> (EligibilityWidget:WBP Eligibility Check Modal Object Reference)
   > n22 = Is Valid(InputObject=@n21)
   n23 = Get EligibilityRequirementsModalVM!(self:Tournament View VM Object Reference=Get FortTournamentViewVM()) -> (EligibilityRequirementsModalVM:Tournament Requirements Modal VM Object Reference)
   n24 = Create Widget!(
    execute=@n22.Is Not Valid
    Class=/Compete/TournamentLeaderboard/ModalMessages/EligibilityCheck/WBP_EligibilityCheckModal.WBP_EligibilityCheckModal_C
    OwningPlayer=GetOwningPlayer()
    FortTournamentRequirementsModalVM:Tournament Requirements Modal VM Object Reference=@n23
   ) -> (then, ReturnValue:WBP Eligibility Check Modal Object Reference)
   > n25 = Set EligibilityWidget(EligibilityWidget:WBP Eligibility Check Modal Object Reference=@n24) -> (Output_Get:WBP Eligibility Check Modal Object Reference)
   n26 = Get EligibilityWidget()
   TryPushWidgetAsModalPanel(execute=[@n25.then, @n29.then], PushWidget=@n26)
   > n27 = ActivateWidget(self=@n26)
   n28 = RestoreEnabled() {is=title K2Node_CustomEvent}
   Bind Event to On Widget Deactivated(execute=@n27.then, self:Common Activatable Widget Object Reference=@n26, Delegate=@n28) {is=title K2Node_AddDelegate}
   n29 = SetFortTournamentRequirementsModalVM!(
    execute=@n22.Is Valid
    self:WBP Eligibility Check Modal Object Reference=@n21
    Viewmodel:Tournament Requirements Modal VM Object Reference=@n23
   ) -> (then)
   n30 = "On Clicked (WatchLive_Button)"!() -> (OutputDelegate, then, Button:Common Button Base Object Reference) {is=title K2Node_ComponentBoundEvent}
   > Fire Event Element Interacted(Element=@n30.Button)
   > n31 = Create Widget!(
    execute
    Class=/Compete/TournamentLeaderboard/ModalMessages/WatchLive/WBP_WatchLiveModal.WBP_WatchLiveModal_C
    OwningPlayer=GetOwningPlayer()
    FortTournamentWatchLiveVM:Watch Live VM Object Reference
   ) -> (then, ReturnValue:WBP Watch Live Modal Object Reference)
   > TryPushWidgetAsModalPanel(PushWidget=@n31)
   > Initialize!(
    execute
    self:WBP Watch Live Modal Object Reference=@n31
    WatchLiveVM:Watch Live VM Object Reference=Get WatchLiveVM!(self:Tournament View VM Object Reference=Get FortTournamentViewVM()) -> (WatchLiveVM:Watch Live VM Object Reference)
    LeaderboardCategory:"Fort Compete UILeaderboard Category Structure (by ref)"
    bShowFriendsOnly:bool=false
    AlternativeButtonVisibility:ESlateVisibility Enum=Collapsed
   ) -> (then)
   > ActivateWidget(self=@n31)
   n32 = "On Clicked (BookmarkSessions_Button)"!() -> (OutputDelegate, then, Button:Common Button Base Object Reference) {is=title K2Node_ComponentBoundEvent}
   > Fire Event Element Interacted(Element=@n32.Button)
   > SetAllWindowsBookmark!(
    execute
    self:Tournament View VM Object Reference=Get FortTournamentViewVM()
    bValue:bool=NOT Boolean(A=Get Are All Windows Bookmarked() -> (Are All Windows Bookmarked:bool))
   ) -> (then)
   n33 = OnInputMethodChanged!() -> (OutputDelegate, then, bNewInputType:ECommonInputType Enum) {is=title K2Node_CustomEvent}
   > n34 = Sequence()
   > SetScrollWhenFocusChanges(
    self=Get AthenaScrollBox_Content()
    NewScrollWhenFocusChanges=Select(
     Option 0:EScrollWhenFocusChanges Enum=AnimatedScroll
     Option 1:EScrollWhenFocusChanges Enum=NoScroll
     Index="Equal (Enum)"(A=@n33.bNewInputType, B=Touch)
    ) -> (ReturnValue:EScrollWhenFocusChanges Enum)
   )
   > n35 = Switch on ECommonInputType!(execute, Selection:ECommonInputType Enum=@n33.bNewInputType) -> (MouseAndKeyboard:Exec, Gamepad:Exec, Touch:Exec, Count:Exec) {is=title K2Node_SwitchEnum}
   n36 = GetDesiredFocusTarget!(self:Common Activatable Widget Object Reference) -> (ReturnValue:Widget Object Reference)
   ScrollWidgetIntoView(execute=@n35.Gamepad, self=Get AthenaScrollBox_Content(), WidgetToFind=@n36, AnimateScroll=false, ScrollDestination=IntoView, Padding=0)
   > SetFocus(self=@n36)
   n37 = Branch(execute=@n34.then_1, Condition=Get InitialInputChanged() -> (InitialInputChanged:bool))
   > RefreshTournamentWindowData()
   Set InitialInputChanged(execute=@n37.else, InitialInputChanged:bool=true) -> (Output_Get:bool)
   Setup Focus On Device Input Changed() {is=title K2Node_CustomEvent}
   > Assign On Input Method Changed!(execute:Exec, self:Common Input Object Reference=Get CommonInputSubsystem(), Delegate:"Delegate (by ref)"=@n33) -> (then:Exec) {is=title K2Node_AssignDelegate}
   Setup Initial Focus For Devices() {is=title K2Node_CustomEvent}
   > n38 = Switch on ECommonInputType(Selection=GetCurrentInputType(self=Get CommonInputSubsystem()))
   > ScrollToStart(self=Get AthenaScrollBox_Content())
   SetFocus(execute=@n38.Gamepad, self=GetFirstButton())
   Event PreConstruct()
   > SetSize!(
    execute
    self:WBP UIKit Block Image Object Reference=Get Block_Icon!(self:WBP UIKit Button CTA Base Object Reference=Get WatchLive_Button() -> (WatchLive_Button:WBP UIKit Button Quiet Object Reference)) -> (Block_Icon:WBP UIKit Block Image Object Reference)
    Desired Size:Vector 2D Structure=(X=12,Y=12)
   ) -> (then)
   > n39 = Is Valid(InputObject=Get ContextMenuActionContext() -> (ContextMenuActionContext:UIKit Context Menu Action Context Object Reference))
   n40 = Construct Object from Class(execute=@n39.Is Not Valid, Class=/Script/FortniteUI.UIKitContextMenuActionContext) -> (ReturnValue:UIKit Context Menu Action Context Object Reference)
   > Set ContextMenuActionContext(ContextMenuActionContext:UIKit Context Menu Action Context Object Reference=@n40) -> (Output_Get:UIKit Context Menu Action Context Object Reference)
   Event OnInitialized()
   > n41 = Construct Object from Class(Class=/Script/CompeteUI.FortTournamentViewVM) -> (ReturnValue:Tournament View VM Object Reference)
   > SetFortTournamentViewVM!(execute, self, Viewmodel:Tournament View VM Object Reference=@n41) -> (then)
   > Setup Initial Focus For Devices() {is=title K2Node_CallFunction}
   > Setup Focus On Device Input Changed() {is=title K2Node_CallFunction}
   > SetMobileStyle() {is=title K2Node_CallFunction}
   Event On Activated()
   > On Activated Internal()
   Event On Deactivated()
   > Clear(TargetArray:Array of Tournament Requirement Item Structures=Get RequirementsList!(self:Tournament Requirements Modal VM Object Reference=@n23) -> (RequirementsList:Array of Tournament Requirement Item Structures))
   > ClearLockRefreshTimer(self:Tournament Requirements Modal VM Object Reference=@n23)
   SetFocus(
    execute=@n28.then
    self=Get WBP_TournamentDetails_EligibilityButton() -> (WBP_TournamentDetails_EligibilityButton:WBP Tournament Details Eligibility Button Object Reference)
   )
   n42 = GoToLobbyWithTournament(execute=@n17.then, self=@n15, bSwitchProfileRegion=true)
  ]}
 GenerateContextMenu {nodes [
   n1 = GenerateContextMenu!() -> (then, TournamentPages:Array of Fort Event Windows List Info Structures) {is=title K2Node_FunctionEntry}
   > RemoveAllInteractions(self:UIKit Context Menu Action Context Object Reference=Get ContextMenuActionContext())
   > n2 = For Each Loop(Array:Array of Fort Event Windows List Info Structures=@n1) -> (Array Element:Fort Event Windows List Info Structure)
   SetContext!(
    execute=@n2.Completed
    self:WBP UIKit Context Menu Anchor Button Base Object Reference=Get ContextMenu_Button()
    InteractionContext_Data:UIKit Context Menu Action Context Object Reference=Get ContextMenuActionContext()
   ) -> (then)
   n3 = Break Fort Event Windows List Info!(FortEventWindowsListInfo:"Fort Event Windows List Info Structure (by ref)"=@n2) -> (DisplayText:Text, EventWindowId:String) {is=title K2Node_BreakStruct}
   AddInteraction!(
    execute=@n2.LoopBody
    self:UIKit Context Menu Action Context Object Reference=Get ContextMenuActionContext()
    InActionInfo_DisplayName:Text=@n3
    InActionInfo_IconTexture:Texture 2D Soft Object Reference
    InActionInfo_IconMaterial:Material Instance Soft Object Reference
    InActionInfo_SortPriority:int=1200
    InActionInfo_bIsAvailable:bool=True
    InActionInfo_bIsEnabled:bool=True
    InActionInfo_bIsSelected:bool=False
    InActionInfo_bAddDivider:bool=False
    InActionInfo_bPositionDividerAfter:bool=False
    InActionInfo_bCloseMenuOnSelect:bool=False
    InActionInfo_InteractionNameID:Name=String To Name(InString=@n3.EventWindowId)
    InActionInfo_Interaction:Data Table Row Handle Structure=(DataTable=None,RowName="")
    InActionInfo_PressedDelegate:Delegate="(null).None"
    InActionInfo_ReleasedDelegate:Delegate="(null).None"
    InInteractionDelegate:"Delegate (by ref)"=Create Event()
   ) -> (then)
  ]}
 GetFirstButton {nodes [
   GetFirstButton() {is=title K2Node_FunctionEntry}
   > n1 = For Each Loop(Array:Array of Widget Object References=GetAllChildren(self=Get ButtonsContainer() -> (ButtonsContainer:Stack Box Object Reference))) -> (Array Element:Widget Object Reference)
   > Branch(Condition=AND Boolean(A=IsVisible(self=@n1), B=Get bIsEnabled!(self:Widget Object Reference=@n1) -> (bIsEnabled:bool)))
   > Return Node!(execute, Widget:Widget Object Reference=@n1)
  ]}
 GetPrimaryContentSetup {nodes [
   GetPrimaryContentSetup() {is=title K2Node_FunctionEntry}
   > Return Node!(
    execute
    ReturnValue:Primary Content Setup Structure=Make Primary Content Setup!(
     bShowBottomBar:bool=false
     bShowChatWidget:bool=false
     bEnableNewTextChat:bool=False
     bShowTopBar:bool=false
     bTabSupportsInterface:bool=False
    ) -> (PrimaryContentSetup:Primary Content Setup Structure) {is=title K2Node_MakeStruct}
   )
  ]}
 GetRankRequirementStats {nodes [
   GetRankRequirementStats() {is=title K2Node_FunctionEntry}
   n1 = Get!(self:Tournament Requirements Modal VM Object Reference=Get EligibilityRequirementsModalVM(self=Get FortTournamentViewVM())) -> (RankInfo_ConditionalModes:Array of Tournament Conditional Mode Requirement Structures)
   > n2 = Branch(Condition=Is Empty!(TargetArray:Array of Tournament Conditional Mode Requirement Structures=@n1) -> (ReturnValue:bool))
   > Return Node!(execute, Level:int=0, ReturnValue:String)
   n3 = "Get (a ref)"!(Array:Array of Tournament Conditional Mode Requirement Structures=@n1, Dimension 1=0) -> (Output_Type:String, Output_Mode:String, Output_Level:int)
   Return Node(execute=@n2.else, Level=@n3.Output_Level, ReturnValue=@n3.Output_Mode)
  ]}
 Go Navigation to Buttons Container {nodes [
   Go Navigation to Buttons Container!() -> (then, Navigation:EUINavigation Enum) {is=title K2Node_FunctionEntry}
   > ScrollWidgetIntoView(self=Get AthenaScrollBox_Content(), WidgetToFind=Get ButtonsContainer(), AnimateScroll=true, ScrollDestination=BottomOrRight, Padding=0)
   > Return Node!(execute, ReturnValue:Widget Object Reference=GetFirstButton() {is=title K2Node_CallFunction})
  ]}
 Handle Background Focus {nodes [
   Handle Background Focus!() -> (then, MyGeometry:Geometry Structure, MouseEvent:"Pointer Event Structure (by ref)") {is=title K2Node_FunctionEntry}
   > SetFocus(self=GetFirstButton())
   > Return Node(ReturnValue:Event Reply Structure=Handled())
  ]}
 IsGamepad {nodes [
   IsGamepad() {is=title K2Node_FunctionEntry}
   n1 = Get CommonInputSubsystem()
   > Return Node!(execute, IsGamepad:bool="Equal (Enum)"(A=GetCurrentInputType(self=@n1), B=Gamepad))
  ]}
 On Activated Internal {nodes [
   On Activated Internal() {is=title K2Node_FunctionEntry}
   > ResetEligibilityButtonStatus()
   > OnInputMethodChanged!(execute, self, bNewInputType:ECommonInputType Enum=GetCurrentInputType(self=Get CommonInputSubsystem())) -> (then) {is=title K2Node_CallFunction}
   n1 = Get TournamentHeaderVM!(self:Tournament View VM Object Reference=Get FortTournamentViewVM()) -> (TournamentHeaderVM:Tournament Header VM Object Reference)
   > SetupViewModel(self:Tournament Header VM Object Reference=@n1)
   > SetFortTournamentHeaderVM!(
    execute
    self:WBP Tournament Details Header Object Reference=Get Header() -> (Header:WBP Tournament Details Header Object Reference)
    Viewmodel:Tournament Header VM Object Reference=@n1
   ) -> (then)
   > RefreshTournamentWindowData()
   > n2 = Construct Object from Class(Class=/Script/CompeteUI.FortTournamentBackgroundVM) -> (ReturnValue:Tournament Background VM Object Reference)
   > SetupViewModel!(execute, self:Tournament Background VM Object Reference=@n2, InType:EFortTournamentBackgroundImageType Enum=TournamentViewBackground) -> (then)
   > SetTournament Background VM!(
    execute
    self:WBP Tournament Background Object Reference=Get Background() -> (Background:WBP Tournament Background Object Reference)
    Viewmodel:Tournament Background VM Object Reference=@n2
   ) -> (then)
   > SetFocus(self=GetFirstButton())
   > SetIsSeriesPointLeaderboardAvailable!(
    execute
    self
    Avaliable:bool=IsTournamentSeriesPointLeaderboardAvailable!(self:Fort Compete UIManager Object Reference=Get FortCompeteUIManager() -> (ReturnValue:Fort Compete UIManager Object Reference) {is=title K2Node_GetSubsystem}) -> (ReturnValue:bool)
   ) -> (then)
   n3 = Get TournamentName!(self:Tournament Header VM Object Reference=Get TournamentHeaderVM(self=Get FortTournamentViewVM())) -> (TournamentName:Text)
   > Fire Event Sub View Opened!(
    execute
    View:Widget Object Reference=Self-Reference()
    AdditionalInfo:Flow Analytics Additional Info Structure="Add String (Non-Localized)"!(
     Info:"Flow Analytics Additional Info Structure (by ref)"=Make Flow Analytics Additional Info!() -> (FlowAnalyticsAdditionalInfo:Flow Analytics Additional Info Structure) {is=title K2Node_MakeStruct}
     Key:String=TournamentName
     Value:String="To String (Text)"(InText=@n3)
    ) -> (ReturnValue:Flow Analytics Additional Info Structure)
    bIsOverlay:bool=false
   ) -> (then)
  ]}
 OnCarouselPrizesVisibilityChanged {nodes [
   n1 = OnCarouselPrizesVisibilityChanged!() -> (then, InVisibility:ESlateVisibility Enum=Visible) {is=title K2Node_FunctionEntry}
   > SetVisibility(self=Get NamedSlot_2_ItemsCarousel() -> (NamedSlot_2_ItemsCarousel:Named Slot Object Reference), InVisibility=@n1)
  ]}
 RefreshTournamentWindowData {nodes [
   RefreshTournamentWindowData() {is=title K2Node_FunctionEntry}
   > UpdateTournamentWindow(self:Tournament View VM Object Reference=Get FortTournamentViewVM())
   > Update Window View Model(self:WBP Tournament Details Header Object Reference=Get Header())
   > SetWatchLiveAvailable!(
    execute
    self
    Avaliable:bool=CanWatchLiveGames!(self:Fort Compete UIManager Object Reference=Get FortCompeteUIManager()) -> (ReturnValue:bool)
   ) -> (then)
   > SetCashAndItemPrizeData()
   > SetCosmeticPrizesData()
   > SetCashPrizesSectionData()
   > SetAdvanceTokensSectionData()
   > UpdatePrizeSection()
  ]}
 ResetEligibilityButtonStatus {nodes [
   ResetEligibilityButtonStatus() {is=title K2Node_FunctionEntry}
   n1 = Get WBP_TournamentDetails_EligibilityButton()
   > SetIsEnabled(self=@n1, bInIsEnabled=false)
   > SetIcon!(
    execute
    self:WBP UIKit Button Generic Base Object Reference=@n1
    Icon:Object Reference=/FortUILibrary/UIKit/Throbber/MI_UIKit_Monochrome_Throbber
   ) -> (then)
  ]}
 Set Is Leaderboard Available {nodes [
   n1 = Set Is Leaderboard Available!() -> (then, IsLeaderboardAvailable:bool=~) {is=title K2Node_FunctionEntry}
   > SetVisibility(
    self=Get Leaderboard_Button() -> (Leaderboard_Button:WBP UIKit Button Regular Object Reference)
    InVisibility=Select(Option 0:ESlateVisibility Enum=Collapsed, Option 1:ESlateVisibility Enum=SelfHitTestInvisible, Index=@n1) -> (ReturnValue:ESlateVisibility Enum)
   )
  ]}
 Set Is Live Window {nodes [
   n1 = Set Is Live Window!() -> (then, IsLiveWindow:bool=~) {is=title K2Node_FunctionEntry}
   > n2 = Set Is Live(Is Live:bool=@n1) -> (Output_Get:bool)
   > SetButtonsPool!(execute, self, IsLive:bool=@n2) -> (then)
  ]}
 Set Player Eligible for Window {nodes [
   n1 = Set Player Eligible for Window!() -> (then, New Is Eligible:bool=~) {is=title K2Node_FunctionEntry}
   > Set Is Player Eligible for WIndow(Is Player Eligible for WIndow:bool=@n1) -> (Output_Get:bool)
  ]}
 SetAdvanceTokensSectionData {nodes [
   SetAdvanceTokensSectionData() {is=title K2Node_FunctionEntry}
   n1 = Get FortTournamentViewVM()
   > n2 = Branch(Condition=Is Not Empty!(TargetArray:Array of Tournament Leaderboard Payout Table Data Structures=Get Payouts!(self:Tournament View VM Object Reference=@n1) -> (Payouts:Array of Tournament Leaderboard Payout Table Data Structures)) -> (ReturnValue:bool))
   > n3 = Sequence()
   > n4 = Is Valid(InputObject=Get Advance Tokens Prize Section())
   n5 = Create Widget!(
    execute=@n4.Is Not Valid
    Class=/Compete/TournamentDetails/Sections/WBP_TournamentDetailsSection_Advancement.WBP_TournamentDetailsSection_Advancement_C
    OwningPlayer
    TournamentViewVM:Tournament View VM Object Reference
    Style:Common Button Style Class Reference=/FortUILibrary/Foundation/Buttons/ButtonStyles/ButtonStyle-Empty.ButtonStyle-Empty_C
    bApplyAlphaOnDisable:bool=True
    bLocked:bool=False
    bSelectable:bool=False
    bShouldSelectUponReceivingFocus:bool=False
    bInteractableWhenSelected:bool=False
    bToggleable:bool=False
    bTriggerClickedAfterSelection:bool=False
    bDisplayInputActionWhenNotInteractable:bool=True
    bHideInputActionWithKeyboard:bool=False
    bShouldUseFallbackDefaultInputAction:bool=True
    bRequiresHold:bool=False
    HoldData:Common UIHold Data Class Reference
    ClickMethod:EButtonClickMethod Enum=DownAndUp
    InputPriority:int=0
    TriggeringInputAction:Data Table Row Handle Structure
   ) -> (then, ReturnValue:WBP Tournament Details Section Advancement Object Reference)
   > n6 = Set Advance Tokens Prize Section(Advance Tokens Prize Section:WBP Tournament Details Section Advancement Object Reference=@n5) -> (Output_Get:WBP Tournament Details Section Advancement Object Reference)
   n7 = Get NamedSlot_4_AdvanceTokensSection() -> (NamedSlot_4_AdvanceTokensSection:Named Slot Object Reference)
   > SetVisibility(self=@n7, InVisibility=SelfHitTestInvisible)
   > AddChild(self=@n7, Content=@n6)
   n8 = Get Advance Tokens Prize Section()
   SetTournamentViewVM!(
    execute=@n3.then_1
    self:WBP Tournament Details Section Advancement Object Reference=@n8
    Viewmodel:Tournament View VM Object Reference=Get FortTournamentViewVM()
   ) -> (then)
   > SetVisibility(
    self=Get NamedSlot_4_AdvanceTokensSection()
    InVisibility=GetVisibility!(self:Widget Object Reference=@n8) -> (ReturnValue:ESlateVisibility Enum=Visible)
   )
   SetVisibility(execute=@n2.else, self=Get NamedSlot_4_AdvanceTokensSection(), InVisibility=Collapsed)
  ]}
 SetBookmarkState {nodes [
   n1 = SetBookmarkState!() -> (then, IsBookmarked:bool=~) {is=title K2Node_FunctionEntry}
   > n2 = Set IsBookmarked(IsBookmarked:bool=@n1) -> (Output_Get:bool)
   > n3 = Branch(Condition=@n2)
   > SetText!(
    execute
    self:Fort CTAButton Object Reference=Get Bookmark_Button() -> (Bookmark_Button:WBP UIKit Button Loud Object Reference)
    ButtonText:"Text (by ref)"=Get RemoveBookmarkText() -> (RemoveBookmarkText:Text)
   ) -> (then)
   SetText(execute=@n3.else, self=Get Bookmark_Button(), ButtonText=Get AddBookmarkText() -> (AddBookmarkText:Text))
  ]}
 SetBookmarksAmount {nodes [
   n1 = SetBookmarksAmount!() -> (then, InText:Text) {is=title K2Node_FunctionEntry}
   > "SetText (Text)"(self=Get CommonText_BookmarksAmount() -> (CommonText_BookmarksAmount:Common Text Object Reference), InText=@n1)
  ]}
 SetButtonsPool {nodes [
   n1 = SetButtonsPool!() -> (then, IsLive:bool=~) {is=title K2Node_FunctionEntry}
   > SetVisibility(
    self=Get Play_Button() -> (Play_Button:WBP UIKit Button Loud Object Reference)
    InVisibility=Select(Option 0=Collapsed, Option 1=SelfHitTestInvisible, Index=@n1)
   )
   n2 = NOT Boolean(A=@n1)
   n3 = NOT Boolean(A=Get Is Tournament Window Ended() -> (Is Tournament Window Ended:bool))
   > SetVisibility(self=Get Bookmark_Button(), InVisibility=Select(Option 0=Collapsed, Option 1=SelfHitTestInvisible, Index=AND Boolean(A=@n2, B=@n3)))
  ]}
 SetCanPlayNow {nodes [
   n1 = SetCanPlayNow!() -> (then, Can Play Now:bool=~) {is=title K2Node_FunctionEntry}
   > n2 = Set Can Play Now(Can Play Now:bool=@n1) -> (Output_Get:bool)
   > SetIsEnabled(self=Get Play_Button(), bInIsEnabled=@n2)
  ]}
 SetCashAndItemPrizeData {nodes [
   SetCashAndItemPrizeData() {is=title K2Node_FunctionEntry}
   n1 = Get FortTournamentViewVM()
   > n2 = Branch(Condition=Is Not Empty(TargetArray=Get Payouts(self=@n1)))
   > n3 = Sequence()
   > n4 = Is Valid(InputObject=Get Cash And Item Prize Section() -> (Cash And Item Prize Section:WBP Tournament Details Cash Prize Object Reference))
   n5 = Construct Object from Class!(
    execute=@n4.Is Not Valid
    Class=&/WBP_TournamentDetailsCashPrize.WBP_TournamentDetailsCashPrize_C
    self
    FortTournamentViewVM:Tournament View VM Object Reference
   ) -> (then, ReturnValue:WBP Tournament Details Cash Prize Object Reference)
   > n6 = Set Cash And Item Prize Section(Cash And Item Prize Section:WBP Tournament Details Cash Prize Object Reference=@n5) -> (Output_Get:WBP Tournament Details Cash Prize Object Reference)
   n7 = Get NamedSlot_1_CashItem() -> (NamedSlot_1_CashItem:Named Slot Object Reference)
   > SetVisibility(self=@n7, InVisibility=SelfHitTestInvisible)
   > AddChild(self=@n7, Content=@n6)
   > Set bIsEnabled!(execute, bIsEnabled:bool=false, self:Widget Object Reference=@n6) -> (then, Output_Get:bool)
   n8 = Get Cash And Item Prize Section()
   SetFortTournamentViewVM(execute=@n3.then_1, self:WBP Tournament Details Cash Prize Object Reference=@n8, Viewmodel=Get FortTournamentViewVM())
   > SetVisibility(self=Get NamedSlot_1_CashItem(), InVisibility=GetVisibility(self=@n8) -> (ReturnValue=Visible))
   SetVisibility(execute=@n2.else, self=Get NamedSlot_1_CashItem(), InVisibility=Collapsed)
  ]}
 SetCashPrizesSectionData {nodes [
   SetCashPrizesSectionData() {is=title K2Node_FunctionEntry}
   n1 = Get FortTournamentViewVM()
   > n2 = Branch(Condition=Is Not Empty(TargetArray=Get Payouts(self=@n1)))
   > n3 = Sequence()
   > n4 = Is Valid(InputObject=Get Cash Prizes Section())
   n5 = Create Widget!(
    execute=@n4.Is Not Valid
    Class=/Compete/TournamentDetails/Sections/WBP_TournamentDetailsSection_Prizes.WBP_TournamentDetailsSection_Prizes_C
    OwningPlayer
    Tournament View VM:Tournament View VM Object Reference
    Style=&ButtonStyle-Empty_C
    bApplyAlphaOnDisable=True
    bLocked=False
    bSelectable=False
    bShouldSelectUponReceivingFocus=False
    bInteractableWhenSelected=False
    bToggleable=False
    bTriggerClickedAfterSelection=False
    bDisplayInputActionWhenNotInteractable=True
    bHideInputActionWithKeyboard=False
    bShouldUseFallbackDefaultInputAction=True
    bRequiresHold=False
    HoldData
    ClickMethod=DownAndUp
    InputPriority=0
    TriggeringInputAction
   ) -> (then, ReturnValue:WBP Tournament Details Section Prizes Object Reference)
   > n6 = Set Cash Prizes Section(Cash Prizes Section:WBP Tournament Details Section Prizes Object Reference=@n5) -> (Output_Get:WBP Tournament Details Section Prizes Object Reference)
   > SetVisibility(self=Get NamedSlot_3_CashSection() -> (NamedSlot_3_CashSection:Named Slot Object Reference), InVisibility=SelfHitTestInvisible)
   > SetTitle!(execute, self:WBP Tournament Details Section Prizes Object Reference=@n6, InText:Text="LOCTEXT(\"[PH] Final round prizes\")") -> (then)
   > AddChild(self=Get NamedSlot_3_CashSection(), Content=@n6)
   n7 = Get Cash Prizes Section()
   SetTournament View VM!(
    execute=@n3.then_1
    self:WBP Tournament Details Section Prizes Object Reference=@n7
    Viewmodel:Tournament View VM Object Reference=Get FortTournamentViewVM()
   ) -> (then)
   > SetVisibility(self=Get NamedSlot_3_CashSection(), InVisibility=GetVisibility(self=@n7) -> (ReturnValue=Visible))
   SetVisibility(execute=@n2.else, self=Get NamedSlot_3_CashSection(), InVisibility=Collapsed)
  ]}
 SetCosmeticPrizesData {nodes [
   SetCosmeticPrizesData() {is=title K2Node_FunctionEntry}
   > n1 = Sequence()
   > n2 = Get!(execute:Exec) -> (then:Exec, else:Exec, Cosmetic Prizes Carousel:WBP Tournament Prize Carousel Object Reference)
   n3 = Construct Object from Class(execute=@n2.else, Class=&/WBP_TournamentPrizeCarousel.WBP_TournamentPrizeCarousel_C) -> (ReturnValue:WBP Tournament Prize Carousel Object Reference)
   > n4 = Set Cosmetic Prizes Carousel(Cosmetic Prizes Carousel:WBP Tournament Prize Carousel Object Reference=@n3) -> (Output_Get:WBP Tournament Prize Carousel Object Reference)
   > SetVisibility(self=Get NamedSlot_2_ItemsCarousel(), InVisibility=SelfHitTestInvisible)
   > Bind Event to On Visibility Changed(self:User Widget Object Reference=@n4, Delegate=Create Event()) {is=title K2Node_AddDelegate}
   > AddChild(self=Get NamedSlot_2_ItemsCarousel(), Content=@n4)
   SetFortTournamentViewVM(execute=@n1.then_1, self:WBP Tournament Prize Carousel Object Reference=Get Cosmetic Prizes Carousel(), Viewmodel=Get FortTournamentViewVM())
  ]}
 SetGamepadFocus {nodes [
   SetGamepadFocus() {is=title K2Node_FunctionEntry}
   > Set ContentHeight(ContentHeight:double=0) -> (Output_Get:double)
   n1 = Get AthenaScrollBox_Content()
   > n2 = Branch(Condition="float <= float"(A=GetScrollOffset!(self:Scroll Box Object Reference=@n1) -> (ReturnValue:float), B=200) -> (ReturnValue=~))
   > SetFocus(self=GetFirstButton())
   > ScrollToStart(self=Get AthenaScrollBox_Content())
   n3 = For Each Loop with Break!(
    Exec:Exec=@n2.else
    Array:Array of Widget Object References=GetAllChildren(self=Get ScrolledContent() -> (ScrolledContent:Grid Panel Object Reference))
    Break:Exec=@n8.then
   ) -> (LoopBody:Exec, Array Element:Widget Object Reference, Array Index:int, Completed:Exec)
   > n4 = Cast To UserWidget!(execute, Object=@n3) -> (then, CastFailed, AsUser Widget:User Widget Object Reference) {is=title K2Node_DynamicCast}
   n5 = float + float(
    A=Get ContentHeight() -> (ContentHeight:double)
    B=GetDesiredSize!(self:Widget Object Reference=@n3) -> (ReturnValue_X:double, ReturnValue_Y:double).ReturnValue_Y
   )
   > Set ContentHeight(ContentHeight=@n5)
   n6 = GetScrollOffset(self=Get AthenaScrollBox_Content())
   n7 = float - float(A=@n5, B=200)
   > Branch(Condition=AND Boolean(A=Get bIsFocusable!(self:User Widget Object Reference=@n4) -> (bIsFocusable:bool), B="float > float"(A=@n7, B=@n6)))
   > n8 = SetFocus(self=@n3)
  ]}
 SetIsSeriesPointLeaderboardAvailable {nodes [
   n1 = SetIsSeriesPointLeaderboardAvailable!() -> (then, Avaliable:bool=~) {is=title K2Node_FunctionEntry}
   > SetVisibility(
    self=Get SeriesLeaderboard_Button() -> (SeriesLeaderboard_Button:WBP UIKit Button Regular Object Reference)
    InVisibility=Select(Option 0=Collapsed, Option 1=SelfHitTestInvisible, Index=@n1)
   )
  ]}
 SetMobileStyle {nodes [
   SetMobileStyle() {is=title K2Node_FunctionEntry}
   > Branch(Condition=IsMobileGame())
   > SetWidthOverride(self=Get SizeBox_ScrollBox() -> (SizeBox_ScrollBox:Size Box Object Reference), InWidthOverride=1400)
   n1 = Get SizeBox_Content() -> (SizeBox_Content:Size Box Object Reference)
   > ClearWidthOverride(self:Size Box Object Reference=@n1)
   n2 = SlotAsSafeBoxSlot!(Widget:Widget Object Reference=@n1) -> (ReturnValue:Safe Zone Slot Object Reference)
   > Set HAlign!(execute, HAlign:EHorizontalAlignment Enum=HAlign_Fill, self:Safe Zone Slot Object Reference=@n2) -> (then, Output_Get:EHorizontalAlignment Enum=HAlign_Fill)
   n3 = Get!(self:Safe Zone Slot Object Reference=@n2) -> (Padding_Left:float=~, Padding_Top:float=~, Padding_Right:float=~, Padding_Bottom:float=~)
   > Set!(
    execute
    Padding_Left:float=48
    Padding_Top:float=@n3.Padding_Top
    Padding_Right:float=48
    Padding_Bottom:float=@n3.Padding_Bottom
    self:Safe Zone Slot Object Reference=@n2
   ) -> (then, Output_Get:Margin Structure)
   > SetDesiredSizeOverride!(
    execute
    self:Image Object Reference=Get Image_RankIcon() -> (Image_RankIcon:Image Object Reference)
    DesiredSize:Vector 2D Structure=(X=32,Y=32)
   ) -> (then)
   > SetDesiredSizeOverride(self=Get Icon_BookmarkAmount() -> (Icon_BookmarkAmount:Image Object Reference), DesiredSize=(X=23,Y=32))
   n4 = SlotAsGridSlot(Widget=Get CommonVisualAttachment_Coachmark() -> (CommonVisualAttachment_Coachmark:Common Visual Attachment Object Reference))
   > SetNudge!(
    execute
    self:Grid Slot Object Reference=@n4
    InNudge_X:double=400
    InNudge_Y:double=Get!(self:Grid Slot Object Reference=@n4) -> (Nudge_X:double, Nudge_Y:double).Nudge_Y
   ) -> (then)
  ]}
 SetPassed {nodes [
   SetPassed!() -> (then, Passed?:bool=~) {is=title K2Node_FunctionEntry}
   n1 = Get WBP_TournamentDetails_EligibilityButton()
   > SetIsEnabled(self=@n1, bInIsEnabled=true)
   > IsEligible!(execute, self:WBP Tournament Details Eligibility Button Object Reference=@n1, InEligible:bool=Get Passed?() -> (Passed?:bool)) -> (then)
   n2 = Get FortTournamentViewVM()
   n3 = IsMatchmakingFlowActive!(self:Fort Matchmaking Context Object Reference=Get FortMatchmakingContext!() -> (ReturnValue:Fort Matchmaking Context Object Reference) {is=title K2Node_GetBlueprintContextEx}) -> (ReturnValue:bool)
   > SetIsEnabled(
    self=Get Play_Button()
    bInIsEnabled=AND Boolean(A=NOT Boolean(A=@n3), B=IsPlayNowEnabled!(self:Tournament View VM Object Reference=@n2) -> (ReturnValue:bool))
   )
  ]}
 SetPlayerMeetsRankRequirement {nodes [
   n1 = SetPlayerMeetsRankRequirement!() -> (then, MeetsRequirement:bool=~) {is=title K2Node_FunctionEntry}
   > Set PlayerMeetsRankRequirement(PlayerMeetsRankRequirement:bool=@n1) -> (Output_Get:bool)
  ]}
 SetRankIcon {nodes [
   SetRankIcon() {is=title K2Node_FunctionEntry}
   n1 = GetRankRequirementStats!(self) -> (Level:int=0, ReturnValue:String) {is=title K2Node_CallFunction}
   n2 = Get HabaneroRanked() -> (HabaneroRanked:Fort Habanero Display Data Object Reference)
   n3 = Break Fort Habanero Mode List!(FortHabaneroModeList:"Fort Habanero Mode List Structure (by ref)"=Find(
     TargetMap:Map of Strings to Fort Habanero Mode List Structures=Get EntryData!(self:Fort Habanero Display Data Object Reference=@n2) -> (EntryData:Map of Strings to Fort Habanero Mode List Structures)
     Key=@n1.ReturnValue
    ) -> (Value:Fort Habanero Mode List Structure)) -> (TierList:Array of Fort Habanero Tier Structures, bShowTopLeaderboardIcon:bool=~, TopTierTopLeaderboardPositions:Fort Habanero Tier Structure) {is=title K2Node_BreakStruct}
   n4 = Is Valid Index!(TargetArray:Array of Fort Habanero Tier Structures=@n3, IndexToTest:int=@n1) -> (ReturnValue:bool)
   n5 = Select(Option 0=Collapsed, Option 1=SelfHitTestInvisible, Index=@n4)
   > SetVisibility(self=Get Image_RankIcon(), InVisibility=@n5)
   > SetVisibility(self=Get CommonText_RankEligibilityStatus() -> (CommonText_RankEligibilityStatus:Common Text Object Reference), InVisibility=@n5)
   > n6 = Branch(Condition=@n4)
   > SetBrushFromSoftTexture!(
    execute
    self:Image Object Reference=Get Image_RankIcon()
    SoftTexture:Texture 2D Soft Object Reference=Break Fort Habanero Tier!(FortHabaneroTier:"Fort Habanero Tier Structure (by ref)"="Get (a copy)"(Array:Array of Fort Habanero Tier Structures=@n3, Dimension 1=@n1) -> (Output:Fort Habanero Tier Structure)) -> (
     ForegroundIcon:Material Instance Soft Object Reference
     ForegroundIconSmall:Material Instance Soft Object Reference
     BackgroundIcon:Material Instance Soft Object Reference
     BackgroundIconSmall:Material Instance Soft Object Reference
     WhiteIcon:Texture 2D Soft Object Reference
     SimpleIconSmall:Texture 2D Soft Object Reference
     Name:Text
     TierType:Fort Habanero Tier Type Data Object Reference
     AdditionalLeftIcon:Material Instance Soft Object Reference
     AdditionalRightIcon:Material Instance Soft Object Reference
    ) {is=title K2Node_BreakStruct}.SimpleIconSmall
    bMatchSize:bool=false
   ) -> (then)
   n7 = Branch(execute=@n6.else, Condition=IsEmpty(InString=@n1.ReturnValue))
   PrintString!(
    execute=@n7.else
    InString=Append(A="RankIcon error: Invalid track: ", B=@n1.ReturnValue)
    bPrintToScreen
    bPrintToLog
    TextColor=(R=1,G=0.0033,B=0,A=1)
    Duration=5
    Key
   ) -> (then)
  ]}
 SetRegion {nodes [
   n1 = SetRegion!() -> (then, InText:Text) {is=title K2Node_FunctionEntry}
   > "SetText (Text)"(self=Get CommonText_Region() -> (CommonText_Region:Common Text Object Reference), InText=@n1)
  ]}
 SetRequirements {nodes [
   n1 = SetRequirements!() -> (then, RequirementList:Array of Tournament Requirement Item Structures) {is=title K2Node_FunctionEntry}
   > Branch(Condition=Is Not Empty(TargetArray:Array of Tournament Requirement Item Structures=@n1))
   > SetRankIcon() {is=title K2Node_CallFunction}
   n2 = Get LastCheckRankResult!(self:Tournament Requirements Modal VM Object Reference=Get EligibilityRequirementsModalVM(self=Get FortTournamentViewVM())) -> (LastCheckRankResult:EFortCompeteHabaneroRankRequirementCheckResult Enum=Invalid)
   > SetVisibility(
    self=Get RankEligibilitySection() -> (RankEligibilitySection:Grid Panel Object Reference)
    InVisibility=Select(
     Option 0=HitTestInvisible
     Option 1=Hidden
     Index="Equal (Enum)"(A:EFortCompeteHabaneroRankRequirementCheckResult Enum=@n2, B:EFortCompeteHabaneroRankRequirementCheckResult Enum=Invalid)
    )
   )
   > "SetText (Text)"(
    self=Get CommonText_RankEligibilityStatus()
    InText=Select!(
     Invalid:Text="LOCTEXT(\"Invalid\")"
     Valid:Text="LOCTEXT(\"Rank Is Valid\")"
     PlayerIsUnhabaneroed:Text="LOCTABLE(\"/Compete/TournamentLeaderboard/ModalMessages/EligibilityCheck/ST_EligibilityTitle\", \"RankUnhabaneroed\")"
     PlayerCurrentRankBelowMinimumRequirement:Text="LOCTABLE(\"&ST_EligibilityTitle\", \"RankBelowMin\")"
     PlayerHighestRankBelowMinimumRequirement:Text="LOCTABLE(\"&ST_EligibilityTitle\", \"RankBelowMin\")"
     PlayerCurrentRankAboveMaximumRequirement:Text="LOCTABLE(\"&ST_EligibilityTitle\", \"RankAboveMax\")"
     PlayerIsIneligible:Text="LOCTABLE(\"&ST_EligibilityTitle\", \"RankBelowMin\")"
     PlayerTrackRankBelowMinimumRequirement:Text="LOCTABLE(\"&ST_EligibilityTitle\", \"RankBelowMin\")"
     PlayerTrackRankAboveMaximumRequirement:Text="LOCTABLE(\"&ST_EligibilityTitle\", \"RankAboveMax\")"
     Index:EFortCompeteHabaneroRankRequirementCheckResult Enum=@n2
    ) -> (ReturnValue:Text)
   )
   > n3 = Switch on EFortCompeteHabaneroRankRequirementCheckResult!(execute, Selection:EFortCompeteHabaneroRankRequirementCheckResult Enum=@n2) -> (
    Invalid:Exec
    Valid:Exec
    PlayerIsUnhabaneroed:Exec
    PlayerCurrentRankBelowMinimumRequirement:Exec
    PlayerHighestRankBelowMinimumRequirement:Exec
    PlayerCurrentRankAboveMaximumRequirement:Exec
    PlayerIsIneligible:Exec
    PlayerTrackRankBelowMinimumRequirement:Exec
    PlayerTrackRankAboveMaximumRequirement:Exec
   ) {is=title K2Node_SwitchEnum}
   SetPassed!(
    execute=[
     @n3.PlayerCurrentRankAboveMaximumRequirement
     @n3.PlayerCurrentRankBelowMinimumRequirement
     @n3.PlayerHighestRankBelowMinimumRequirement
     @n3.PlayerIsIneligible
     @n3.PlayerIsUnhabaneroed
     @n3.PlayerTrackRankAboveMaximumRequirement
     @n3.PlayerTrackRankBelowMinimumRequirement
    ]
    self
    Passed?:bool=false
   ) -> (then) {is=title K2Node_CallFunction}
   > Return Node!(execute)
   n4 = For Each Loop with Break(
    Exec=[@n3.Invalid, @n3.Valid]
    Array:Array of Tournament Requirement Item Structures=Get RequirementList() -> (RequirementList:Array of Tournament Requirement Item Structures)
    Break=@n8.else
   ) -> (Array Element:Tournament Requirement Item Structure)
   SetPassed(execute=@n4.Completed, Passed?=Get Passed?())
   > Return Node()
   n5 = Break Tournament Requirement Item!(TournamentRequirementItem:"Tournament Requirement Item Structure (by ref)"=@n4) -> (RequirementType:EFortTournamentRequirementCheckType Enum, bHasPassed:bool=~, Arguments:Tournament Requirement Item Arguments Structure) {is=title K2Node_BreakStruct}
   n6 = Branch(
    execute=@n4.LoopBody
    Condition="Equal (Enum)"(A:EFortTournamentRequirementCheckType Enum=@n5, B:EFortTournamentRequirementCheckType Enum=EULA)
   )
   > Set EULA_Passed?(EULA_Passed?:bool=@n5.bHasPassed) -> (Output_Get:bool)
   n7 = Set Passed?(execute=@n6.else, Passed?:bool=@n5.bHasPassed) -> (Output_Get:bool)
   > n8 = Branch(Condition=@n7)
  ]}
 SetSeriesLeaderboardName {nodes [
   n1 = SetSeriesLeaderboardName!() -> (then, LeaderboardName:Text) {is=title K2Node_FunctionEntry}
   > Set Text(
    Text=SelectText(A="LOCTEXT(\"Series Leaderboard\")", B=@n1, bPickA=TextIsEmpty(InText=@n1))
    self:Fort CTAButton Object Reference=Get SeriesLeaderboard_Button()
   )
  ]}
 SetTournamentWindow {nodes [
   n1 = SetTournamentWindow!() -> (then, InteractionNameID:Name) {is=title K2Node_FunctionEntry}
   > SetEventWindow!(execute, self:Fort Compete UIManager Object Reference=Get FortCompeteUIManager(), InTournamentWindowId:String="To String (Name)"(InName=@n1)) -> (then)
   > RefreshTournamentWindowData() {is=title K2Node_CallFunction}
  ]}
 SetTournamentWindowEnded {nodes [
   n1 = SetTournamentWindowEnded!() -> (then, bIsTournamentWindowEnded:bool=~) {is=title K2Node_FunctionEntry}
   > Set Is Tournament Window Ended(Is Tournament Window Ended:bool=@n1) -> (Output_Get:bool)
   > SetButtonsPool(IsLive=Get Is Live() -> (Is Live:bool)) {is=title K2Node_CallFunction}
  ]}
 SetWatchLiveAvailable {nodes [
   n1 = SetWatchLiveAvailable!() -> (then, Avaliable:bool=~) {is=title K2Node_FunctionEntry}
   > SetVisibility(
    self=Get WatchLive_Button()
    InVisibility=Select!(Option 0:ESlateVisibility Enum=Collapsed, Option 1:ESlateVisibility Enum=SelfHitTestInvisible, Index:bool=@n1) -> (ReturnValue:ESlateVisibility Enum)
   )
  ]}
 Show Coach Mark {nodes [
   n1 = Show Coach Mark!() -> (then, IsVisible:bool=~) {is=title K2Node_FunctionEntry}
   > Branch(Condition=IsValid(Object=Get CoachMark()))
   > n2 = Branch(Condition=@n1)
   n3 = Get CoachMark()
   > ActivateWidget(self=@n3)
   > SetVisibility(self=@n3, InVisibility=SelfHitTestInvisible)
   RemoveFromParent(execute=@n2.else, self=Get CoachMark())
  ]}
 ShowTournamentLeaderboardScreen {nodes [
   n1 = ShowTournamentLeaderboardScreen!() -> (then, IsTournamentSeriesPointLeaderboard:bool=~) {is=title K2Node_FunctionEntry}
   n2 = Get FortTournamentViewVM()
   > ShowTournamentLeaderboardScreen!(
    execute
    self:Fort Compete UIManager Object Reference=Get FortCompeteUIManager()
    Params:"Tournament Leaderboard Screen Params Structure (by ref)"=Make Tournament Leaderboard Screen Params!(
     ContextPC:Player Controller Object Reference=GetPlayerController!(PlayerIndex:int=0) -> (ReturnValue:Player Controller Object Reference)
     TournamentId:String
     TournamentWindowId:String
     bIsTournamentSeriesPointLeaderboard:bool=@n1
     TournamentWatchLiveEntryVM:Watch Live VM Object Reference=Get WatchLiveVM(self=@n2)
     TournamentHeaderVM:Tournament Header VM Object Reference=Get TournamentHeaderVM(self=@n2)
     PayoutDisplayData:Array of Tournament Leaderboard Payout Table Data Structures=Get Payouts(self=@n2)
    ) -> (TournamentLeaderboardScreenParams:Tournament Leaderboard Screen Params Structure) {is=title K2Node_MakeStruct}
   ) -> (then) {is=title K2Node_CallFunction}
   > DeactivateWidget()
  ]}
 UpdateBookmarkAllButton {nodes [
   n1 = UpdateBookmarkAllButton!() -> (then, AreAllWindowsBookmarked:bool=~) {is=title K2Node_FunctionEntry}
   > Set Are All Windows Bookmarked(Are All Windows Bookmarked:bool=@n1) -> (Output_Get:bool)
   > SetText(
    self=Get BookmarkSessions_Button() -> (BookmarkSessions_Button:WBP UIKit Button Quiet Object Reference)
    ButtonText=Select(
     Option 0:"Text (by ref)"=Get BookmarkAllText() -> (BookmarkAllText:Text)
     Option 1:"Text (by ref)"=Get RemoveAllBookmarksText() -> (RemoveAllBookmarksText:Text)
     Index=@n1
    ) -> (ReturnValue:"Text (by ref)")
   )
  ]}
 UpdateBookmarkButtonVisibility {nodes [
   n1 = UpdateBookmarkButtonVisibility!() -> (then, InAnyBookmarkables:bool=~) {is=title K2Node_FunctionEntry}
   n2 = Get BookmarkSessions_Button()
   > SetTriggeringInputAction!(
    execute
    self:Common Button Base Object Reference=@n2
    InputActionRow:"Data Table Row Handle Structure (by ref)"=Select(
     Option 0:"Data Table Row Handle Structure (by ref)"=Make Data Table Row Handle!(DataTable:Data Table Object Reference, RowName:Name=None) -> (DataTableRowHandle:Data Table Row Handle Structure) {is=title K2Node_MakeStruct}
     Option 1:"Data Table Row Handle Structure (by ref)"=Get Bookmark All Input Action() -> (Bookmark All Input Action:Data Table Row Handle Structure)
     Index=@n1
    ) -> (ReturnValue:"Data Table Row Handle Structure (by ref)")
   ) -> (then)
   > SetVisibility(
    self=@n2
    InVisibility=Select(Option 0:ESlateVisibility Enum=Collapsed, Option 1:ESlateVisibility Enum=Visible, Index=@n1) -> (ReturnValue:ESlateVisibility Enum)
   )
  ]}
 UpdatePrizeSection {nodes [
   UpdatePrizeSection() {is=title K2Node_FunctionEntry}
   > SetFortTournamentViewVM(
    self:WBP Tournament Prize Header Object Reference=Get Section_Prizes() -> (Section_Prizes:WBP Tournament Prize Header Object Reference)
    Viewmodel=Get FortTournamentViewVM()
   )
   > n1 = Sequence!(execute) -> (then_0, then_1, then_2:Exec)
   n2 = Get Cash Prizes Section()
   > Is Valid(InputObject=@n2)
   > Branch(Condition="Not Equal (Enum)"!(A:ESlateVisibility Enum=GetVisibility(self=@n2) -> (ReturnValue=Visible), B:ESlateVisibility Enum=Collapsed) -> (ReturnValue:bool))
   > Set LocalAnyPrizesVisible(LocalAnyPrizesVisible:bool=true) -> (Output_Get:bool)
   n3 = Get Cosmetic Prizes Carousel()
   Is Valid(exec=@n1.then_1, InputObject=@n3)
   > Branch(Condition="Not Equal (Enum)"(A=GetVisibility(self=@n3) -> (ReturnValue=Visible), B=Collapsed))
   > Set LocalAnyPrizesVisible(LocalAnyPrizesVisible=true)
   SetSize!(
    execute=@n1.then_2
    self:Spacer Object Reference=Get FinalRoundPrizes_Spacer() -> (FinalRoundPrizes_Spacer:Spacer Object Reference)
    InSize_X:double=0
    InSize_Y:double=Select(Option 0:double=0, Option 1:double=89, Index=Get LocalAnyPrizesVisible() -> (LocalAnyPrizesVisible:bool)) -> (ReturnValue:double)
   ) -> (then)
   > n4 = Branch(Condition=Get LocalAnyPrizesVisible())
   > n5 = IsDisplayingTokenHeader!(execute, self:WBP Tournament Prize Header Object Reference=Get Section_Prizes()) -> (then, IsDisplayingTokenHeader:bool)
   > SetVisibility(
    self=Get Text_FinalRoundHeader() -> (Text_FinalRoundHeader:Common Text Object Reference)
    InVisibility=Select(Option 0:ESlateVisibility Enum=Collapsed, Option 1:ESlateVisibility Enum=SelfHitTestInvisible, Index=@n5) -> (ReturnValue:ESlateVisibility Enum)
   )
   SetVisibility(execute=@n4.else, self=Get Text_FinalRoundHeader(), InVisibility=Collapsed)
  ]}
]
variables [
 AddBookmarkText {type=string, default=Add Bookmark}
 Advance Tokens Prize Section {type=WBP_TournamentDetailsSection_Advancement_C, default=null}
 Are All Windows Bookmarked {type=boolean}
 BackgroundColor {type=LinearColor, default=<rgba 0.0704 0.003 0.0222 1>}
 BackgroundTexture {type=Texture, default=/Compete/Textures/T_UI_Compete_Placeholder}
 BookmarkAllText {type=string, default=Bookmark all sessions}
 Can Play Now {type=boolean}
 Cash And Item Prize Section {type=WBP_TournamentDetailsCashPrize_C, default=null}
 Cash Prizes Section {type=WBP_TournamentDetailsSection_Prizes_C, default=null}
 ContextMenuActionContext {type=UIKitContextMenuActionContext, default=null}
 Cosmetic Prizes Carousel {type=WBP_TournamentPrizeCarousel_C, default=null}
 EligibilityWidget {type=WBP_EligibilityCheckModal_C, default=null}
 EULA_Passed? {type=boolean}
 Focused Section Index {type=integer}
 HabaneroRanked {type=FortHabaneroDisplayData, default=/Game/UI/Habanero/HabaneroDisplayData}
 InitialInputChanged {type=boolean}
 Is Live {type=boolean}
 Is Player Eligible for WIndow {type=boolean, default=true}
 Is Tournament Window Ended {type=boolean}
 IsBookmarked {type=boolean}
 PlayerMeetsRankRequirement {type=boolean, default=true}
 PlayerRank {type=E_RankType, default=Platinum}
 QualificationRequirement {type=string, default=Not Eligible - Qualifying round has passed}
 RemoveAllBookmarksText {type=string, default=Remove all session bookmarks}
 RemoveBookmarkText {type=string, default=Remove Bookmark}
 Sections {type="array<WBP_TournamentDetailsSection_Base_C>", default []}
]
widgets {
 info {
  inheritedWidgetCount=0
  namedSlotCount=1
  parentClass=&CommonActivatableWidget
  rootWidgetClass=&Overlay
  widgetCount=46
 }
 widgets [
  Root {widgetClassPath=&Overlay, widget=Root}
  Content {
   widgetClassPath=&GridPanel
   bIsVariable=true
   parent=Root
   widget=Content
   properties {columnFill [1], rowFill [1]}
   slot {class=OverlaySlot, properties {horizontalAlignment=HAlign_Fill, verticalAlignment=VAlign_Fill}}
  }
  SafeZone_1 {
   widgetClassPath=&SafeZone
   bIsVariable=true
   parent=Content
   widget=SafeZone_1
   properties {padBottom=false, padLeft=false}
   slot {class=GridSlot, properties {horizontalAlignment=HAlign_Right, layer=2, verticalAlignment=VAlign_Top, padding=<ltrb 0 90 80 0>}}
  }
  CommonText_Region {
   widgetClassPath=&CommonTextBlock
   bIsVariable=true
   parent=SafeZone_1
   widget=CommonText_Region
   properties {
    scrollStyle=&TextScrollStyle-Base_C
    text=NAE
    font {
     bForceMonospaced=false
     bMaterialIsStencil=false
     fontMaterial=null
     fontObject=&HeadingNow
     letterSpacing=0
     monospacedWidth=1
     size=18
     skewAmount=0
     typefaceFontName=75 Medium
     outlineSettings {
      bApplyOutlineToDropShadows=false
      bMiteredCorners=false
      bSeparateFillAlpha=false
      outlineMaterial=null
      outlineSize=2
      outlineColor=<rgba 0 0 0 0.2>
     }
    }
    shadowOffset=<0 0>
   }
   slot {class=SafeZoneSlot}
  }
  AthenaScrollBox_Content {
   widgetClassPath=/Script/FortniteUI.AthenaScrollBox
   bIsVariable=true
   parent=Content
   widget=AthenaScrollBox_Content
   properties {
    allowOverscroll=false
    bAllowRightClickDragScrolling=false
    navigationDestination=Center
    scrollBarVisibility=Collapsed
    scrollWhenFocusChanges=AnimatedScroll
    visibility=SelfHitTestInvisible
    scrollbarPadding=<ltrb 0 0 0 0>
    scrollbarThickness=<0 0>
    widgetBarStyle {
     thickness=16
     draggedThumbImage {
      bIsDynamicallyLoaded=false
      drawAs=NoDrawType
      imageType=FullColor
      mirroring=NoMirror
      resourceName=../../../Engine/Content/Slate/Common/Scrollbar_Thumb.png
      resourceObject=null
      tiling=NoTile
      imageSize=<0 0>
      margin=<ltrb 0.25 0.25 0.25 0.25>
      outlineSettings {
       bUseBrushTransparency=false
       roundingType=HalfHeightRadius
       width=0
       color {colorUseRule=UseColor_Specified, specifiedColor=<rgba 0 0 0 0>}
       cornerRadii=<0 0 0 0>
      }
      tintColor {colorUseRule=UseColor_Specified, specifiedColor=<rgba 1 1 1 1>}
      uVRegion {bIsValid=false, max=<0 0>, min=<0 0>}
     }
     horizontalBackgroundImage {
      bIsDynamicallyLoaded=false
      drawAs=NoDrawType
      imageType=NoImage
      mirroring=NoMirror
      resourceName=None
      resourceObject=null
      tiling=NoTile
      imageSize=<0 0>
      margin=<ltrb 0 0 0 0>
      outlineSettings {
       bUseBrushTransparency=false
       roundingType=HalfHeightRadius
       width=0
       color {colorUseRule=UseColor_Specified, specifiedColor=<rgba 0 0 0 0>}
       cornerRadii=<0 0 0 0>
      }
      tintColor {colorUseRule=UseColor_Specified, specifiedColor=<rgba 1 1 1 1>}
      uVRegion {bIsValid=false, max=<0 0>, min=<0 0>}
     }
     horizontalBottomSlotImage {
      bIsDynamicallyLoaded=false
      drawAs=NoDrawType
      imageType=FullColor
      mirroring=NoMirror
      resourceName=../../../Engine/Content/Slate/Common/Scrollbar_Background_Horizontal.png
      resourceObject=null
      tiling=NoTile
      imageSize=<8 8>
      margin=<ltrb 0 0 0 0>
      outlineSettings {
       bUseBrushTransparency=false
       roundingType=HalfHeightRadius
       width=0
       color {colorUseRule=UseColor_Specified, specifiedColor=<rgba 0 0 0 0>}
       cornerRadii=<0 0 0 0>
      }
      tintColor {colorUseRule=UseColor_Specified, specifiedColor=<rgba 1 1 1 1>}
      uVRegion {bIsValid=false, max=<0 0>, min=<0 0>}
     }
     horizontalTopSlotImage {
      bIsDynamicallyLoaded=false
      drawAs=NoDrawType
      imageType=FullColor
      mirroring=NoMirror
      resourceName=../../../Engine/Content/Slate/Common/Scrollbar_Background_Horizontal.png
      resourceObject=null
      tiling=NoTile
      imageSize=<8 8>
      margin=<ltrb 0 0 0 0>
      outlineSettings {
       bUseBrushTransparency=false
       roundingType=HalfHeightRadius
       width=0
       color {colorUseRule=UseColor_Specified, specifiedColor=<rgba 0 0 0 0>}
       cornerRadii=<0 0 0 0>
      }
      tintColor {colorUseRule=UseColor_Specified, specifiedColor=<rgba 1 1 1 1>}
      uVRegion {bIsValid=false, max=<0 0>, min=<0 0>}
     }
     hoveredThumbImage {
      bIsDynamicallyLoaded=false
      drawAs=NoDrawType
      imageType=FullColor
      mirroring=NoMirror
      resourceName=../../../Engine/Content/Slate/Common/Scrollbar_Thumb.png
      resourceObject=null
      tiling=NoTile
      imageSize=<0 0>
      margin=<ltrb 0.25 0.25 0.25 0.25>
      outlineSettings {
       bUseBrushTransparency=false
       roundingType=HalfHeightRadius
       width=0
       color {colorUseRule=UseColor_Specified, specifiedColor=<rgba 0 0 0 0>}
       cornerRadii=<0 0 0 0>
      }
      tintColor {colorUseRule=UseColor_Specified, specifiedColor=<rgba 1 1 1 1>}
      uVRegion {bIsValid=false, max=<0 0>, min=<0 0>}
     }
     normalThumbImage {
      bIsDynamicallyLoaded=false
      drawAs=NoDrawType
      imageType=FullColor
      mirroring=NoMirror
      resourceName=../../../Engine/Content/Slate/Common/Scrollbar_Thumb.png
      resourceObject=null
      tiling=NoTile
      imageSize=<0 0>
      margin=<ltrb 0.25 0.25 0.25 0.25>
      outlineSettings {
       bUseBrushTransparency=false
       roundingType=HalfHeightRadius
       width=0
       color {colorUseRule=UseColor_Specified, specifiedColor=<rgba 0 0 0 0>}
       cornerRadii=<0 0 0 0>
      }
      tintColor {colorUseRule=UseColor_Specified, specifiedColor=<rgba 1 1 1 1>}
      uVRegion {bIsValid=false, max=<0 0>, min=<0 0>}
     }
     verticalBackgroundImage {
      bIsDynamicallyLoaded=false
      drawAs=NoDrawType
      imageType=NoImage
      mirroring=NoMirror
      resourceName=None
      resourceObject=null
      tiling=NoTile
      imageSize=<0 0>
      margin=<ltrb 0 0 0 0>
      outlineSettings {
       bUseBrushTransparency=false
       roundingType=HalfHeightRadius
       width=0
       color {colorUseRule=UseColor_Specified, specifiedColor=<rgba 0 0 0 0>}
       cornerRadii=<0 0 0 0>
      }
      tintColor {colorUseRule=UseColor_Specified, specifiedColor=<rgba 1 1 1 1>}
      uVRegion {bIsValid=false, max=<0 0>, min=<0 0>}
     }
     verticalBottomSlotImage {
      bIsDynamicallyLoaded=false
      drawAs=NoDrawType
      imageType=FullColor
      mirroring=NoMirror
      resourceName=../../../Engine/Content/Slate/Common/Scrollbar_Background_Vertical.png
      resourceObject=null
      tiling=NoTile
      imageSize=<8 8>
      margin=<ltrb 0 0 0 0>
      outlineSettings {
       bUseBrushTransparency=false
       roundingType=HalfHeightRadius
       width=0
       color {colorUseRule=UseColor_Specified, specifiedColor=<rgba 0 0 0 0>}
       cornerRadii=<0 0 0 0>
      }
      tintColor {colorUseRule=UseColor_Specified, specifiedColor=<rgba 1 1 1 1>}
      uVRegion {bIsValid=false, max=<0 0>, min=<0 0>}
     }
     verticalTopSlotImage {
      bIsDynamicallyLoaded=false
      drawAs=NoDrawType
      imageType=FullColor
      mirroring=NoMirror
      resourceName=../../../Engine/Content/Slate/Common/Scrollbar_Background_Vertical.png
      resourceObject=null
      tiling=NoTile
      imageSize=<8 8>
      margin=<ltrb 0 0 0 0>
      outlineSettings {
       bUseBrushTransparency=false
       roundingType=HalfHeightRadius
       width=0
       color {colorUseRule=UseColor_Specified, specifiedColor=<rgba 0 0 0 0>}
       cornerRadii=<0 0 0 0>
      }
      tintColor {colorUseRule=UseColor_Specified, specifiedColor=<rgba 1 1 1 1>}
      uVRegion {bIsValid=false, max=<0 0>, min=<0 0>}
     }
    }
    widgetStyle {
     barThickness=9
     bottomShadowBrush {
      bIsDynamicallyLoaded=false
      drawAs=NoDrawType
      imageType=FullColor
      mirroring=NoMirror
      resourceName=../../../Engine/Content/Slate/Common/ScrollBoxShadowBottom.png
      resourceObject=null
      tiling=NoTile
      imageSize=<16 8>
      margin=<ltrb 0.5 0 0.5 1>
      outlineSettings {
       bUseBrushTransparency=false
       roundingType=HalfHeightRadius
       width=0
       color {colorUseRule=UseColor_Specified, specifiedColor=<rgba 0 0 0 0>}
       cornerRadii=<0 0 0 0>
      }
      tintColor {colorUseRule=UseColor_Specified, specifiedColor=<rgba 1 1 1 1>}
      uVRegion {bIsValid=false, max=<0 0>, min=<0 0>}
     }
     horizontalScrolledContentPadding=<ltrb 0 0 0 0>
     leftShadowBrush {
      bIsDynamicallyLoaded=false
      drawAs=NoDrawType
      imageType=FullColor
      mirroring=NoMirror
      resourceName=../../../Engine/Content/Slate/Common/ScrollBoxShadowLeft.png
      resourceObject=null
      tiling=NoTile
      imageSize=<8 16>
      margin=<ltrb 1 0.5 0 0.5>
      outlineSettings {
       bUseBrushTransparency=false
       roundingType=HalfHeightRadius
       width=0
       color {colorUseRule=UseColor_Specified, specifiedColor=<rgba 0 0 0 0>}
       cornerRadii=<0 0 0 0>
      }
      tintColor {colorUseRule=UseColor_Specified, specifiedColor=<rgba 1 1 1 1>}
      uVRegion {bIsValid=false, max=<0 0>, min=<0 0>}
     }
     rightShadowBrush {
      bIsDynamicallyLoaded=false
      drawAs=NoDrawType
      imageType=FullColor
      mirroring=NoMirror
      resourceName=../../../Engine/Content/Slate/Common/ScrollBoxShadowRight.png
      resourceObject=null
      tiling=NoTile
      imageSize=<8 16>
      margin=<ltrb 0 0.5 1 0.5>
      outlineSettings {
       bUseBrushTransparency=false
       roundingType=HalfHeightRadius
       width=0
       color {colorUseRule=UseColor_Specified, specifiedColor=<rgba 0 0 0 0>}
       cornerRadii=<0 0 0 0>
      }
      tintColor {colorUseRule=UseColor_Specified, specifiedColor=<rgba 1 1 1 1>}
      uVRegion {bIsValid=false, max=<0 0>, min=<0 0>}
     }
     topShadowBrush {
      bIsDynamicallyLoaded=false
      drawAs=NoDrawType
      imageType=FullColor
      mirroring=NoMirror
      resourceName=../../../Engine/Content/Slate/Common/ScrollBoxShadowTop.png
      resourceObject=null
      tiling=NoTile
      imageSize=<16 8>
      margin=<ltrb 0.5 1 0.5 0>
      outlineSettings {
       bUseBrushTransparency=false
       roundingType=HalfHeightRadius
       width=0
       color {colorUseRule=UseColor_Specified, specifiedColor=<rgba 0 0 0 0>}
       cornerRadii=<0 0 0 0>
      }
      tintColor {colorUseRule=UseColor_Specified, specifiedColor=<rgba 1 1 1 1>}
      uVRegion {bIsValid=false, max=<0 0>, min=<0 0>}
     }
     verticalScrolledContentPadding=<ltrb 0 0 0 0>
    }
   }
   slot {class=GridSlot, properties {layer=1}}
  }
  Overlay_0 {
   widgetClassPath=&Overlay
   parent=AthenaScrollBox_Content
   widget=Overlay_0
   slot {class=ScrollBoxSlot, properties {size {sizeRule=Fill, value=1}}}
  }
  Background {
   widgetClassPath=&/WBP_TournamentBackground
   bIsVariable=true
   parent=Overlay_0
   widget=Background
   slot {class=OverlaySlot, properties {horizontalAlignment=HAlign_Fill, verticalAlignment=VAlign_Fill}}
  }
  CaptureForPostBufferUpdate_1 {
   widgetClassPath=&WBP_CaptureForPostBufferUpdate
   bIsVariable=true
   parent=Overlay_0
   widget=CaptureForPostBufferUpdate_1
   slot {class=OverlaySlot, properties {horizontalAlignment=HAlign_Fill, verticalAlignment=VAlign_Fill}}
  }
  SafeZone_Content {
   widgetClassPath=&SafeZone
   bIsVariable=true
   parent=Overlay_0
   widget=SafeZone_Content
   slot {class=OverlaySlot, properties {horizontalAlignment=HAlign_Fill, verticalAlignment=VAlign_Fill}}
  }
  SizeBox_Content {
   widgetClassPath=&SizeBox
   bIsVariable=true
   parent=SafeZone_Content
   widget=SizeBox_Content
   properties {bOverride_WidthOverride=true, widthOverride=1743}
   slot {class=SafeZoneSlot, properties {hAlign=HAlign_Center}}
  }
  SizeBox_ScrollBox {
   widgetClassPath=&SizeBox
   bIsVariable=true
   parent=SizeBox_Content
   widget=SizeBox_ScrollBox
   properties {bOverride_WidthOverride=true, widthOverride=1152}
   slot {class=SizeBoxSlot, properties {horizontalAlignment=HAlign_Left, padding=<ltrb 0 69 0 0>}}
  }
  ScrolledContent {
   widgetClassPath=&GridPanel
   bIsVariable=true
   parent=SizeBox_ScrollBox
   widget=ScrolledContent
   properties {columnFill [1], rowFill [0]}
   slot {class=SizeBoxSlot, properties {horizontalAlignment=HAlign_Left}}
  }
  Header {
   widgetClassPath=&/WBP_TournamentDetailsHeader
   bIsVariable=true
   parent=ScrolledContent
   widget=Header
   slot {class=GridSlot, properties {padding=<ltrb 0 0 0 50>}}
  }
  RankEligibilitySection {
   widgetClassPath=&GridPanel
   bIsVariable=true
   parent=ScrolledContent
   widget=RankEligibilitySection
   properties {columnFill [0,1], rowFill [0]}
   slot {class=GridSlot, properties {layer=1, row=1, verticalAlignment=VAlign_Top, padding=<ltrb 0 10 0 0>}}
  }
  Image_RankIcon {
   widgetClassPath=&Image
   bIsVariable=true
   parent=RankEligibilitySection
   widget=Image_RankIcon
   properties {brush {
     bIsDynamicallyLoaded=false
     drawAs=Image
     imageType=FullColor
     mirroring=NoMirror
     resourceName=None
     resourceObject=&MI_UIKit_Monochrome_Throbber
     tiling=NoTile
     imageSize=<32 32>
     margin=<ltrb 0 0 0 0>
     outlineSettings {
      bUseBrushTransparency=false
      roundingType=HalfHeightRadius
      width=0
      color {colorUseRule=UseColor_Specified, specifiedColor=<rgba 0 0 0 0>}
      cornerRadii=<0 0 0 0>
     }
     tintColor {colorUseRule=UseColor_Specified, specifiedColor=<rgba 1 1 1 1>}
     uVRegion {bIsValid=false, max=<0 0>, min=<0 0>}
    }}
   slot {class=GridSlot, properties {horizontalAlignment=HAlign_Left, verticalAlignment=VAlign_Center, padding=<ltrb 0 0 6 0>}}
  }
  CommonText_RankEligibilityStatus {
   widgetClassPath=&CommonTextBlock
   bIsVariable=true
   parent=RankEligibilitySection
   widget=CommonText_RankEligibilityStatus
   properties {
    mobileFontSizeMultiplier=1.2
    scrollStyle=&TextScrollStyle-Base_C
    text="Rank up to join!"
    colorAndOpacity {colorUseRule=UseColor_Specified, specifiedColor=<rgba 0.5776 0.5776 0.5776 1>}
    font {
     bForceMonospaced=false
     bMaterialIsStencil=false
     fontMaterial=null
     fontObject=&HeadingNow
     letterSpacing=0
     monospacedWidth=1
     size=15.75
     skewAmount=0
     typefaceFontName=75 Medium Italic
     outlineSettings {
      bApplyOutlineToDropShadows=false
      bMiteredCorners=false
      bSeparateFillAlpha=false
      outlineMaterial=null
      outlineSize=2
      outlineColor=<rgba 0 0 0 0.2>
     }
    }
    shadowOffset=<0 0>
   }
   slot {class=GridSlot, properties {column=1, verticalAlignment=VAlign_Center}}
  }
  GridPanel_2 {
   widgetClassPath=&GridPanel
   parent=ScrolledContent
   widget=GridPanel_2
   slot {class=GridSlot, properties {horizontalAlignment=HAlign_Left, row=2, verticalAlignment=VAlign_Center, padding=<ltrb 0 12 0 0>}}
  }
  ButtonsContainer {widgetClassPath=&StackBox, bIsVariable=true, parent=GridPanel_2, widget=ButtonsContainer, slot {class=GridSlot}}
  Bookmark_Button {
   widgetClassPath=/FortUILibrary/UIKit/Buttons/WBP_UIKit_Button_Loud
   bIsVariable=true
   parent=ButtonsContainer
   widget=Bookmark_Button
   properties {
    bHideInputAction=true
    bShouldUseFallbackDefaultInputAction=false
    buttonSize=Large
    clickMethod=MouseDown
    minHeight=88
    pressMethod=ButtonPress
    text=Add Bookmark
    touchMethod=Down
   }
   slot {class=StackBoxSlot, properties {padding=<ltrb 0 0 16 0>}}
  }
  Play_Button {
   widgetClassPath=&WBP_UIKit_Button_Loud
   bIsVariable=true
   parent=ButtonsContainer
   widget=Play_Button
   properties {
    bHideInputAction=true
    bShouldUseFallbackDefaultInputAction=false
    buttonSize=Large
    clickMethod=MouseDown
    minHeight=88
    pressMethod=ButtonPress
    text=Play Now
    touchMethod=Down
   }
   slot {class=StackBoxSlot, properties {padding=<ltrb 0 0 16 0>}}
  }
  WBP_TournamentDetails_EligibilityButton {
   widgetClassPath=&/WBP_TournamentDetails_EligibilityButton
   bIsVariable=true
   parent=ButtonsContainer
   widget=WBP_TournamentDetails_EligibilityButton
   slot {class=StackBoxSlot, properties {padding=<ltrb 0 0 16 0>}}
  }
  Leaderboard_Button {
   widgetClassPath=&WBP_UIKit_Button_Regular
   bIsVariable=true
   parent=ButtonsContainer
   widget=Leaderboard_Button
   properties {
    bHideInputAction=true
    bShouldUseFallbackDefaultInputAction=false
    buttonSize=Large
    clickMethod=MouseDown
    minHeight=88
    pressMethod=ButtonPress
    text=Leaderboard
    touchMethod=Down
   }
   slot {class=StackBoxSlot, properties {padding=<ltrb 0 0 16 0>}}
  }
  SeriesLeaderboard_Button {
   widgetClassPath=&WBP_UIKit_Button_Regular
   bIsVariable=true
   parent=ButtonsContainer
   widget=SeriesLeaderboard_Button
   properties {
    bHideInputAction=true
    bShouldUseFallbackDefaultInputAction=false
    buttonSize=Large
    clickMethod=MouseDown
    minHeight=88
    pressMethod=ButtonPress
    text=Series Leaderboard
    touchMethod=Down
   }
   slot {class=StackBoxSlot, properties {padding=<ltrb 0 0 16 0>}}
  }
  ContextMenu_Button {
   widgetClassPath=/FortUILibrary/UIKit/ContextMenu/WBP_UIKit_ContextMenuAnchorButton
   bIsVariable=true
   parent=ButtonsContainer
   widget=ContextMenu_Button
   properties {
    bIsFocusable=true
    button Background Material=/FortUILibrary/UIKit/Buttons/MaterialInstances/CTAButtons/MI_UIKit_BgQuiet
    buttonClass=&/WBP_TournamentDetails_Button_Context.WBP_TournamentDetails_Button_Context_C
    buttonIcon=/FortUILibrary/Textures/Icons/T_UI_Icon_UIKit_ContextMenu
    buttonText=""
    closeMenuOnContextClicked=true
    min Width=88
    button Padding=<ltrb 8 0 8 0>
    button Text Font {
     bForceMonospaced=false
     bMaterialIsStencil=false
     fontMaterial=/FortUILibrary/UIKit/Buttons/MaterialInstances/GenericButtons/MI_UIKit_GenericText_Quiet
     fontObject=&Burbank
     letterSpacing=180
     monospacedWidth=1
     size=26.25
     skewAmount=0
     typefaceFontName=BurbankRegular-Black
     outlineSettings {
      bApplyOutlineToDropShadows=false
      bMiteredCorners=false
      bSeparateFillAlpha=false
      outlineMaterial=null
      outlineSize=0
      outlineColor=<rgba 0 0 0 1>
     }
    }
   }
   slot {class=StackBoxSlot}
  }
  CommonVisualAttachment_Coachmark {
   widgetClassPath=&CommonVisualAttachment
   bIsVariable=true
   parent=GridPanel_2
   widget=CommonVisualAttachment_Coachmark
   properties {contentAnchor=<0 1>}
   slot {class=GridSlot, properties {horizontalAlignment=HAlign_Left, layer=1, verticalAlignment=VAlign_Bottom, nudge=<345 0>}}
  }
  CoachMark {
   widgetClassPath=/FortUILibrary/UIKit/CoachMark/WBP_UIKit_CoachMark
   bIsVariable=true
   parent=CommonVisualAttachment_Coachmark
   widget=CoachMark
   properties {
    arrowPosition=Left_Down
    dismissButtonClass=null
    shouldShowDismissButton=false
    textDescription="You will be notified with the system message when the tournament you’ve bookmarked will start."
    textDismissButton=""
    textHeader=Tournament Bookmarks
    textOtherActionButton=Got it
   }
   slot {class=SizeBoxSlot, properties {verticalAlignment=VAlign_Top, padding=<ltrb 20 0 0 -20>}}
  }
  BookmarksAmount {
   widgetClassPath=&GridPanel
   parent=ScrolledContent
   widget=BookmarksAmount
   properties {visibility=Collapsed, columnFill [0,1], rowFill [0]}
   slot {class=GridSlot, properties {layer=1, row=3, padding=<ltrb 0 16 0 0>}}
  }
  Icon_BookmarkAmount {
   widgetClassPath=&Image
   bIsVariable=true
   parent=BookmarksAmount
   widget=Icon_BookmarkAmount
   properties {brush {
     bIsDynamicallyLoaded=false
     drawAs=Image
     imageType=FullColor
     mirroring=NoMirror
     resourceName=None
     resourceObject=/Compete/MaterialInstances/MI_UI_Compete_Bookmark
     tiling=NoTile
     imageSize=<15 21>
     margin=<ltrb 0 0 0 0>
     outlineSettings {
      bUseBrushTransparency=false
      roundingType=HalfHeightRadius
      width=0
      color {colorUseRule=UseColor_Specified, specifiedColor=<rgba 0 0 0 0>}
      cornerRadii=<0 0 0 0>
     }
     tintColor {colorUseRule=UseColor_Specified, specifiedColor=<rgba 0.5776 0.5776 0.5776 1>}
     uVRegion {bIsValid=false, max=<0 0>, min=<0 0>}
    }}
   slot {class=GridSlot, properties {horizontalAlignment=HAlign_Left, verticalAlignment=VAlign_Center}}
  }
  CommonText_BookmarksAmount {
   widgetClassPath=&CommonTextBlock
   bIsVariable=true
   parent=BookmarksAmount
   widget=CommonText_BookmarksAmount
   properties {
    mobileFontSizeMultiplier=1.2
    scrollStyle=&TextScrollStyle-Base_C
    text=25.2K
    colorAndOpacity {colorUseRule=UseColor_Specified, specifiedColor=<rgba 0.5776 0.5776 0.5776 1>}
    font {
     bForceMonospaced=false
     bMaterialIsStencil=false
     fontMaterial=null
     fontObject=&HeadingNow
     letterSpacing=10
     monospacedWidth=1
     size=15.75
     skewAmount=0
     typefaceFontName=74 Regular
     outlineSettings {
      bApplyOutlineToDropShadows=false
      bMiteredCorners=false
      bSeparateFillAlpha=false
      outlineMaterial=null
      outlineSize=2
      outlineColor=<rgba 0 0 0 0.2>
     }
    }
    shadowOffset=<0 0>
   }
   slot {class=GridSlot, properties {column=1, verticalAlignment=VAlign_Center, padding=<ltrb 6 0 0 0>}}
  }
  Section_Prizes {
   widgetClassPath=&/WBP_TournamentPrize_Header
   bIsVariable=true
   parent=ScrolledContent
   widget=Section_Prizes
   properties {bIsFocusable=false, visibility=Collapsed}
   slot {class=GridSlot, properties {row=5, padding=<ltrb 0 58 0 0>}}
  }
  Section_Details {
   widgetClassPath=/Compete/TournamentDetails/Sections/WBP_TournamentDetailsSection_Details
   bIsVariable=true
   parent=ScrolledContent
   widget=Section_Details
   properties {navigation=&self:WidgetTree.Section_Details.WidgetNavigation_0}
   slot {class=GridSlot, properties {row=11, padding=<ltrb 0 70 0 0>}}
  }
  Section_Scoring {
   widgetClassPath=/Compete/TournamentDetails/Sections/WBP_TournamentDetailsSection_Scoring
   bIsVariable=true
   parent=ScrolledContent
   widget=Section_Scoring
   properties {navigation=&self:WidgetTree.Section_Scoring.WidgetNavigation_0}
   slot {class=GridSlot, properties {row=12, padding=<ltrb 0 59 0 0>}}
  }
  CommonText_Legal {
   widgetClassPath=&CommonTextBlock
   parent=ScrolledContent
   widget=CommonText_Legal
   properties {
    autoWrapText=true
    bIsScrollingEnabled=false
    mobileFontSizeMultiplier=1.4
    text="Players must be at least 13 years old (or such other age, if greater, as may be required in their country of residence) and MFA enabled. For details see: www.epicgames.com/fortnite/competitive/rules-library."
    font {
     bForceMonospaced=false
     bMaterialIsStencil=false
     fontMaterial=null
     fontObject=&NotoSans
     letterSpacing=0
     monospacedWidth=1
     size=24
     skewAmount=0
     typefaceFontName=NotoSans-Regular
     outlineSettings {
      bApplyOutlineToDropShadows=false
      bMiteredCorners=false
      bSeparateFillAlpha=false
      outlineMaterial=null
      outlineSize=0
      outlineColor=<rgba 0 0 0 1>
     }
    }
    shadowOffset=<0 0>
   }
   slot {class=GridSlot, properties {row=13, padding=<ltrb 0 56 0 180>}}
  }
  FinalRoundPrizes_Spacer {
   widgetClassPath=&Spacer
   bIsVariable=true
   parent=ScrolledContent
   widget=FinalRoundPrizes_Spacer
   properties {size=<10 0>}
   slot {class=GridSlot, properties {row=7}}
  }
  NamedSlot_1_CashItem {
   widgetClassPath=&NamedSlot
   bIsVariable=true
   parent=ScrolledContent
   widget=NamedSlot_1_CashItem
   properties {visibility=Collapsed}
   slot {class=GridSlot, properties {row=6, verticalAlignment=VAlign_Bottom, padding=<ltrb 0 31 0 0>}}
  }
  Text_FinalRoundHeader {
   widgetClassPath=&CommonTextBlock
   bIsVariable=true
   parent=ScrolledContent
   widget=Text_FinalRoundHeader
   properties {
    bAutoCollapseWithEmptyText=true
    bIsScrollingEnabled=false
    mobileFontSizeMultiplier=0.75
    text=Final Round Prizes
    textTransformPolicy=ToUpper
    wrapTextAt=510
    font {
     bForceMonospaced=false
     bMaterialIsStencil=false
     fontMaterial=null
     fontObject=&HeadingNow
     letterSpacing=0
     monospacedWidth=1
     size=24
     skewAmount=0
     typefaceFontName=86 Bold Italic
     outlineSettings {
      bApplyOutlineToDropShadows=false
      bMiteredCorners=false
      bSeparateFillAlpha=false
      outlineMaterial=null
      outlineSize=0
      outlineColor=<rgba 0 0 0 1>
     }
    }
    shadowOffset=<0 0>
   }
   slot {class=GridSlot, properties {row=7, verticalAlignment=VAlign_Bottom}}
  }
  NamedSlot_2_ItemsCarousel {
   widgetClassPath=&NamedSlot
   bIsVariable=true
   parent=ScrolledContent
   widget=NamedSlot_2_ItemsCarousel
   properties {navigation=&self:WidgetTree.NamedSlot_2_ItemsCarousel.WidgetNavigation_0, visibility=Collapsed}
   slot {class=GridSlot, properties {row=8, padding=<ltrb 0 6 0 0>}}
  }
  NamedSlot_3_CashSection {
   widgetClassPath=&NamedSlot
   bIsVariable=true
   parent=ScrolledContent
   widget=NamedSlot_3_CashSection
   properties {navigation=&self:WidgetTree.NamedSlot_3_CashSection.WidgetNavigation_0, visibility=Collapsed}
   slot {class=GridSlot, properties {row=9, padding=<ltrb 0 25 0 0>}}
  }
  NamedSlot_4_AdvanceTokensSection {
   widgetClassPath=&NamedSlot
   bIsVariable=true
   parent=ScrolledContent
   widget=NamedSlot_4_AdvanceTokensSection
   properties {
    navigation=&self:WidgetTree.NamedSlot_4_AdvanceTokensSection.WidgetNavigation_0
    visibility=Collapsed
   }
   slot {class=GridSlot, properties {row=10, padding=<ltrb 0 83 0 0>}}
  }
  CaptureForPostBufferUpdate {
   widgetClassPath=&WBP_CaptureForPostBufferUpdate
   bIsVariable=true
   parent=Root
   widget=CaptureForPostBufferUpdate
   slot {class=OverlaySlot, properties {horizontalAlignment=HAlign_Fill, verticalAlignment=VAlign_Fill}}
  }
  BottomBar {
   widgetClassPath=/Game/UI/Frontend/BottomBar/WBP_AthenaBottomBarWidget
   bIsVariable=true
   parent=Root
   widget=BottomBar
   slot {class=OverlaySlot, properties {horizontalAlignment=HAlign_Right, verticalAlignment=VAlign_Bottom}}
  }
  BottomButtons {
   widgetClassPath=&WrapBox
   namedSlotHost=BottomBar
   widget=BottomButtons
   properties {bExplicitWrapSize=true, horizontalAlignment=HAlign_Right, wrapSize=1500, innerSlotPadding=<24 0>}
  }
  WatchLive_Button {
   widgetClassPath=/FortUILibrary/UIKit/Buttons/WBP_UIKit_Button_Quiet
   bIsVariable=true
   parent=BottomButtons
   widget=WatchLive_Button
   properties {
    bIsFocusable=false
    iconMaterial=/Compete/MaterialInstances/MI_UI_Compete_LiveStatus
    minWidth=120
    mobileMinWidth=200
    text=Watch Live
    triggeringInputAction {dataTable=&InputActionDataTable, rowName=GenericLeftStickClick}
   }
   slot {class=WrapBoxSlot, properties {horizontalAlignment=HAlign_Right}}
  }
  BookmarkSessions_Button {
   widgetClassPath=&WBP_UIKit_Button_Quiet
   bIsVariable=true
   parent=BottomButtons
   widget=BookmarkSessions_Button
   properties {
    bIsFocusable=false
    bRequiresHold=true
    minWidth=120
    mobileMinWidth=200
    text=Bookmark all sessions
    triggeringInputAction {dataTable=&InputActionDataTable, rowName=GenericFaceButtonLeft}
   }
   slot {class=WrapBoxSlot, properties {horizontalAlignment=HAlign_Right}}
  }
  WBP_AthenaBottomBar_WithNotifications {
   widgetClassPath=/Game/UI/Frontend/BottomBar/WBP_AthenaBottomBar_WithNotifications
   bIsVariable=true
   parent=BottomButtons
   widget=WBP_AthenaBottomBar_WithNotifications
   slot {class=WrapBoxSlot}
  }
  Back_Button {
   widgetClassPath=&WBP_UIKit_Button_Quiet
   bIsVariable=true
   parent=BottomButtons
   widget=Back_Button
   properties {
    bIsFocusable=false
    minWidth=120
    mobileMinWidth=200
    text=Back
    triggeringInputAction {dataTable=&InputActionDataTable, rowName=Back}
   }
   slot {class=WrapBoxSlot, properties {horizontalAlignment=HAlign_Right}}
  }
 ]
}
